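# Niagara system for the MuffinLadle_Gas emote prop, producing purple electricity with lightning meshes, flares and glows.
# Emitters combine continuous spawning with timed bursts; BaseFlash002, Lightning001_Down, Lightning002 and Lightning_Up001 are disabled.
# Spawn groups and VisibilityTag let one emitter drive several effect layers with their own lifetime and color.
# Waveform flicker modulates color and light, with time-of-day and front-end brightness scaling in Lightning_Up001.
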NiagaraSystem /BRCosmetics/Animation/Game/MainPlayer/Emotes/MuffinLadle_Gas/FX/NS_Emote_MuffinLadle_Gas_PropElectricity
dependencies {
 usedDataInterfaces [
  &NiagaraDataInterfaceCurve
  &NiagaraDataInterfaceVector2DCurve
  &NiagaraDataInterfaceMeshRendererInfo
  &NiagaraDataInterfaceArrayInt32
  &NiagaraDataInterfaceArrayFloat3
  &NiagaraDataInterfaceSocketReader
 ]
 usedDynamicInputs [
  &MakeCustomVectorFromBool
  &SetBoolByFloatComparison
  &RandomRangeInt
 ]
 usedModules [
  &SystemState
  &self:SystemSpawnScript.SystemScriptSource.SystemScriptGraph.NiagaraNodeAssignment_0.SetVariables_#1
  &self:SystemSpawnScript.SystemScriptSource.SystemScriptGraph.NiagaraNodeAssignment_1.SetVariables_#2
 ]
}
emitters {
 BaseFlash002 {
  bEnabled=false
  emitterUpdateScript [
   &EmitterState(Life Cycle Mode=::NewEnumerator1, Loop Duration=6.4167, UseLoopDelay=false)
   &SpawnRate(SpawnRate=37.1552, Use Spawn Probability=false) {enabled=false}
   &SpawnBurst_Instantaneous(Spawn Time=4.8, Use Spawn Probability=false, UseLoopCountLimit=false)
   &SpawnBurst_Instantaneous(Spawn Time=8.0333, Use Spawn Probability=false, UseLoopCountLimit=false) {module=SpawnBurst_Instantaneous001}
  ]
  particleSpawnScript [
   &InitializeParticle(
    Color=<rgba 1 1 1 0.5>
    Lifetime=0.5
    Position=<0 0 15>
    Position Mode=::NewEnumerator1
    Position Offset=<0 0 -130>
    Sprite Size Mode=::NewEnumerator3
    Uniform Sprite Size=30
    UsePositionOffset=false
    Write Material Random=true
   )
   &CameraOffset(Camera Offset Amount=7)
   &Color(Color=<rgba 0.2582 0.0437 0.6867 1>, Scale Alpha=11.9833)
  ]
  particleUpdateScript [
   &ParticleState()
   &ScaleSpriteSize(Uniform Curve Sprite Scale=DataInterface() {propertyValues {Curve {keys [{tangentMode=RCTM_User, time=0, value=0}, {tangentMode=RCTM_User, time=0.3, value=5}]}}})
   &ScaleColor(
    Scale Alpha=&FloatFromCurve(FloatCurve=DataInterface() {propertyValues {Curve {
        defaultValue=0
        keys [{tangentMode=RCTM_User, time=0, value=0}, <at 0.1 0.9>, <at 0.3 0.15>, <at 0.4 1>, {tangentMode=RCTM_User, time=1, value=0}]
       }}})
    Scale RGB=&VectorFromFloat(Value=&FloatFromCurve(FloatCurve=DataInterface() {propertyValues {Curve {keys [<at 0 1>,<at 0.1 15>,<at 0.2 1>,<at 0.3 9.5>,<at 0.4 1>,<at 0.5 10.5>,<at 0.6 3.5>,<at 0.7 9>,<at 1 1>]}}}))
    ScaleA=true
    ScaleRGB=true
   ) {enabled=false}
   &SolveForcesAndVelocity(Clamp Velocity=false, Limit Acceleration=false)
   &ScaleColor(
    Scale Alpha=&FloatFromCurve(
     FloatCurve=DataInterface() {propertyValues {Curve {keys [
         {tangentMode=RCTM_User, time=0, value=1}
         {arriveTangent=-0.0405, interpMode=RCIM_Cubic, leaveTangent=-0.0405, tangentMode=RCTM_User, time=0.7009, value=1}
         {arriveTangent=-8.0089, interpMode=RCIM_Cubic, leaveTangent=-8.0089, tangentMode=RCTM_User, time=1, value=0}
        ]}}}
     Scale Curve=0.5
    )
    Scale RGB=&Lerp_Vector(
     A=&VectorFromFloat(Value=&FloatFromCurve(
       FloatCurve=DataInterface() {propertyValues {Curve {keys [<at 0 1>,<at 0.1 1>,<at 0.2 2>,<at 0.3 1>,<at 0.6 1>,<at 0.7 2>,<at 0.8 1>,<at 1 1>]}}}
       Scale Curve=&MaskFloatBySpawnGroup(Masked Float=1, Passthrough Float=1)
      ))
     B=&VectorFromFloat(Value=&RandomRangeFloat(Minimum=0.3, Maximum=3, Evaluation Type=::NewEnumerator1))
     Alpha=&FloatFromCurve(
      FloatCurve=DataInterface() {propertyValues {Curve {keys [
          {interpMode=RCIM_Cubic, leaveTangent=2.5294, tangentMode=RCTM_Break, time=0, value=0}
          {arriveTangent=-0.2166, interpMode=RCIM_Cubic, tangentMode=RCTM_Break, time=0.4, value=1}
         ]}}}
      Scale Curve=0.5
     )
    )
    ScaleA=true
    ScaleRGB=true
   ) {module=ScaleColor001}
  ]
  rendererClasses [&NiagaraSpriteRendererProperties]
  renderers [{rendererClass=&NiagaraSpriteRendererProperties}]
 }
 Lightning001_Down {
  bEnabled=false
  emitterUpdateScript [
   &EmitterState(
    Inactive Response=::NewEnumerator1
    Life Cycle Mode=::NewEnumerator1
    Loop Behavior=::NewEnumerator1
    Loop Duration=6.4167
    UseLoopDelay=false
   )
   &SpawnBurst_Instantaneous(Spawn Time=8.0333, Use Spawn Probability=false, UseLoopCountLimit=false) {module=SpawnBurst_Instantaneous002}
   &SpawnBurst_Instantaneous(Spawn Time=4.8, Use Spawn Probability=false, UseLoopCountLimit=false)
   &SpawnRate(SpawnGroup=1, SpawnRate=&RandomRangeFloat(Minimum=35, Maximum=60), Use Spawn Probability=false) {module=SpawnRate001}
   &SpawnBurst_Instantaneous(Spawn Group=2, Use Spawn Probability=false, UseLoopCountLimit=false) {module=SpawnBurst_Instantaneous001}
   /Niagara/Noise/Waveform(
    EnableExponent1=false
    GlobalClampMax=false
    GlobalClampMin=false
    UseInput1=false
    UsePhase1=false
    "Waveform [1]":ENiagara_Waveforms=::NewEnumerator12
    "[1] Frequency"=7
   )
   /SetVariables(Emitter.Flicker:NiagaraFloat=&Add_Float(A=&Multiply_Float(A=$Emitter.Waveform.Waveform, B=0.5), B=0.5)) {isSetParameters=true, module=SetVariables_#3}
  ]
  particleSpawnScript [
   /SetVariables(Particles.VisibilityTag=&SpawnGroup()) {isSetParameters=true, module=SetVariables_#4}
   &InitializeParticle(
    Color=&Multiply_LinearColorByFloat(Linear Color=<rgba 0.2582 0.0437 0.6867 1>, Float=0.4)
    Lifetime=&MaskFloat(
     Mask Source=0
     Mask Parameter=?
     Mask=0
     Masked Float=1
     Passthrough Float=&MaskFloat(
      Mask Source=0
      Mask Parameter=?
      Mask=1
      Masked Float=&RandomRangeFloat(Minimum=0.24, Maximum=0.4)
      Passthrough Float=6
     )
    )
    Mesh Scale Max=&MaskVectorBySpawnGroup(Masked Vector=<1.3 1.3 2.5>, Passthrough Vector=<2 2 3>)
    Mesh Scale Min=&MaskVectorBySpawnGroup(Masked Vector=<1 1 2.5>, Passthrough Vector=<1 1 2>)
    Mesh Scale Mode=::NewEnumerator2
    Position Offset=/Niagara/DynamicInputs/Helpers/MaskPosition(
     Mask Source:NiagaraInt32=0
     Mask Parameter:NiagaraInt32=?
     Mask:NiagaraInt32=0
     Masked Position:NiagaraPosition=<0 0 0>
     Passthrough Position:NiagaraPosition=&MaskPosition(
      Mask Source:NiagaraInt32=0
      Mask Parameter:NiagaraInt32=?
      Mask:NiagaraInt32=1
      Masked Position:NiagaraPosition=<0 0 -90>
      Passthrough Position:NiagaraPosition=<0 0 -108.8283>
     )
    )
    Sprite Size Mode=::NewEnumerator3
    Uniform Sprite Size=&MaskFloat(
     Mask Source=0
     Mask Parameter=?
     Mask=0
     Masked Float=0
     Passthrough Float=&MaskFloat(Mask Source=0, Mask Parameter=?, Mask=1, Masked Float=0, Passthrough Float=37)
    )
    UsePositionOffset=true
    Write Material Random=true
   )
   &InitialMeshOrientation(
    Mesh Axis To Orient=<0 0 1>
    Mesh Orientation Mode=::NewEnumerator2
    Orientation Vector=&Multiply_VectorByFloat(Vector=<0 0 -1>, Float=1)
    Rotation Coordinate Space=::NewEnumerator2
    Use Rotation Vector=true
   ) {module=InitialMeshOrientation001}
   &InitialMeshOrientation(
    Mesh Axis To Orient=&MaskVectorBySpawnGroup(Masked Vector=<0 0 0>, Passthrough Vector=<0 0 0>)
    Mesh Orientation Mode=::NewEnumerator2
    Orientation Vector=&MaskVectorBySpawnGroup(Masked Vector=<0 0 0>, Passthrough Vector=<0 0 0>)
    Rotation=&RandomRangeVector(Minimum=<-0.05 -0.05 0>, Maximum=<0.05 0.05 0>)
    Use Rotation Vector=true
   ) {enabled=false}
   &CameraOffset(Camera Offset Amount=&MaskFloatBySpawnGroup(Masked Float=0, Passthrough Float=2))
   /SetVariables(
    Particles.Flare CameraOffset=7
    Particles.Flare Color:LinearColor=<rgba 0.2582 0.0437 0.6867 1>
    Particles.Flare lifetime=0.5
    Particles.Flare size:Vector2f=<30 30>
    Particles.flare position:NiagaraPosition=&AddVectorToPosition(Position=<0 0 15>, Vector=<0 0 0>)
   ) {isSetParameters=true, module=SetVariables_#5}
   /SetVariables(Particles.Staff color:LinearColor=<rgba 0.2582 0.0437 0.6867 1>, Particles.staff size:Vector2f=<100 100>) {isSetParameters=true, module=SetVariables_#6}
  ]
  particleUpdateScript [
   &ParticleState() {module=ParticleState001}
   /SetVariables(Particles.Update Flare age:NiagaraFloat=&Divide_Float(A=$Particles.Age, B=$Particles.Initial.Flare lifetime)) {isSetParameters=true, module=SetVariables_#7}
   /SetVariables(
    Particles.Update Flare color:LinearColor=&MultiplyLinearColor(
     Linear Color=$Particles.Initial.Flare Color
     Scale Factor=&MakeVector4_FromLinearColor(LinearColor=&MakeLinearColorFromVectorAndFloat(
       "Vector (RGB)"=&Lerp_Vector(
        A=&VectorFromFloat(Value=&FloatFromCurve(
          FloatCurve=DataInterface() {propertyValues {Curve {keys [<at 0 1>,<at 0.1 1>,<at 0.2 2>,<at 0.3 1>,<at 0.6 1>,<at 0.7 2>,<at 0.8 1>,<at 1 1>]}}}
          CurveIndex=$Particles.Update Flare age
         ))
        B=&VectorFromFloat(Value=&RandomRangeFloat(Minimum=0.3, Maximum=3, Evaluation Type=::NewEnumerator1))
        Alpha=&FloatFromCurve(
         FloatCurve=DataInterface() {propertyValues {Curve {keys [
             {interpMode=RCIM_Cubic, leaveTangent=2.5294, tangentMode=RCTM_Break, time=0, value=0}
             {arriveTangent=-0.2166, interpMode=RCIM_Cubic, tangentMode=RCTM_Break, time=0.4, value=1}
            ]}}}
         CurveIndex=$Particles.Update Flare age
         Scale Curve=0.5
        )
       )
       "Float (Alpha)"=&FloatFromCurve(
        FloatCurve=DataInterface() {propertyValues {Curve {keys [
            {tangentMode=RCTM_User, time=0, value=1}
            {arriveTangent=-0.0405, interpMode=RCIM_Cubic, leaveTangent=-0.0405, tangentMode=RCTM_User, time=0.7009, value=1}
            {arriveTangent=-8.0089, interpMode=RCIM_Cubic, leaveTangent=-8.0089, tangentMode=RCTM_User, time=1, value=0}
           ]}}}
        CurveIndex=$Particles.Update Flare age
       )
      ))
    )
    Particles.Update Flare size:Vector2f=&Multiply_Vector2D_ByFloat(
     Float=&FloatFromCurve(
      FloatCurve=DataInterface() {propertyValues {Curve {keys [{tangentMode=RCTM_User, time=0, value=0}, {tangentMode=RCTM_User, time=0.3, value=5}]}}}
      CurveIndex=$Particles.Update Flare age
     )
     Vector2D=$Particles.Initial.Flare size
    )
   ) {isSetParameters=true, module=SetVariables_#8}
   &ScaleColor(
    Scale Alpha=&MaskFloat(
     Mask Source=0
     Mask Parameter=?
     Mask=0
     Masked Float=0
     Passthrough Float=&MaskFloat(
      Mask Source=0
      Mask Parameter=?
      Mask=1
      Masked Float=&FloatFromCurve(FloatCurve=DataInterface() {propertyValues {Curve {keys [
           {tangentMode=RCTM_User, time=0, value=1}
           {arriveTangent=-0.0405, interpMode=RCIM_Cubic, leaveTangent=-0.0405, tangentMode=RCTM_User, time=0.7009, value=1}
           {arriveTangent=-8.0089, interpMode=RCIM_Cubic, leaveTangent=-8.0089, tangentMode=RCTM_User, time=1, value=0}
          ]}}})
      Passthrough Float=&Multiply_Float(A=$Emitter.Flicker, B=10)
     )
    )
    Scale RGB=&MaskVector(
     Mask Source=0
     Mask Parameter=?
     Mask=0
     Masked Vector=<0 0 0>
     Passthrough Vector=&MaskVector(
      Mask Source=0
      Mask Parameter=?
      Mask=1
      Masked Vector=&Lerp_Vector(
       A=&VectorFromFloat(Value=&FloatFromCurve(
         FloatCurve=DataInterface() {propertyValues {Curve {
            defaultValue=1
            keys [
             {interpMode=RCIM_Cubic, tangentMode=RCTM_User, time=0, value=0}
             {arriveTangent=16.9912, interpMode=RCIM_Cubic, leaveTangent=16.9912, tangentMode=RCTM_User, time=15, value=100}
            ]
           }}}
         Scale Curve=&Lerp_Float(A=0.5, B=5, Alpha=$Emitter.LoopedAge)
        ))
       B=&VectorFromFloat(Value=&RandomRangeFloat(Minimum=0.3, Maximum=&RandomRangeFloat(Minimum=5, Maximum=20), Evaluation Type=::NewEnumerator1))
       Alpha=&FloatFromCurve(
        FloatCurve=DataInterface() {propertyValues {Curve {keys [
            {interpMode=RCIM_Cubic, leaveTangent=2.5294, tangentMode=RCTM_Break, time=0, value=0}
            {arriveTangent=-0.2166, interpMode=RCIM_Cubic, tangentMode=RCTM_Break, time=0.4, value=1}
           ]}}}
        Scale Curve=0.5
       )
      )
      Passthrough Vector=&Lerp_Vector(
       A=&VectorFromFloat(Value=&FloatFromCurve(
         FloatCurve=DataInterface() {propertyValues {Curve {
            defaultValue=1
            keys [
             {interpMode=RCIM_Cubic, tangentMode=RCTM_User, time=0, value=0.3}
             {arriveTangent=2.811, interpMode=RCIM_Cubic, leaveTangent=2.811, tangentMode=RCTM_User, time=1, value=1}
            ]
           }}}
         Scale Curve=&Lerp_Float(A=1, B=2, Alpha=$Emitter.Age)
        ))
       B=&VectorFromFloat(Value=&FloatFromCurve(
         FloatCurve=DataInterface() {propertyValues {Curve {
            defaultValue=1
            keys [
             {interpMode=RCIM_Cubic, tangentMode=RCTM_User, time=0, value=0.3}
             {arriveTangent=2.811, interpMode=RCIM_Cubic, leaveTangent=2.811, tangentMode=RCTM_User, time=1, value=1}
            ]
           }}}
         Scale Curve=&Lerp_Float(A=1, B=2, Alpha=$Emitter.Age)
        ))
       Alpha=&FloatFromCurve(
        FloatCurve=DataInterface() {propertyValues {Curve {keys [
            {interpMode=RCIM_Cubic, leaveTangent=2.5294, tangentMode=RCTM_Break, time=0, value=0}
            {arriveTangent=-0.2166, interpMode=RCIM_Cubic, tangentMode=RCTM_Break, time=0.4, value=1}
            <at 1.1 0 RCIM_Cubic>
           ]}}}
        Scale Curve=5
       )
      )
     )
    )
    ScaleA=true
    ScaleRGB=true
   ) {module=ScaleColor001}
   &SolveForcesAndVelocity(Clamp Velocity=false, Limit Acceleration=false) {enabled=false}
   &ScaleMeshSize(Scale Factor=&MaskVectorBySpawnGroup(
     Masked Vector=<1 1 1>
     Passthrough Vector=&MakeVector(
      X=1
      Y=2
      Z=&FloatFromCurve(FloatCurve=DataInterface() {propertyValues {Curve {keys [
           {interpMode=RCIM_Cubic, leaveTangent=0.0966, tangentMode=RCTM_Break, time=0, value=1}
           {interpMode=RCIM_Cubic, leaveTangent=-0.1745, time=0.7665, value=1.07}
          ]}}})
     )
    )) {module=ScaleMeshSize001}
   &ScaleMeshSize(Scale Factor=&MaskVectorBySpawnGroup(
     Masked Vector=&Multiply_VectorByFloat(Vector=<1 1 1>, Float=&RandomRangeFloat(Minimum=0.35, Maximum=0.5))
     Passthrough Vector=&VectorFromFloat(Value=&RandomRangeFloat(Minimum=0.1, Maximum=0.15))
    ))
   &UpdateMeshOrientation(
    Rotation Rate=&MaskFloatBySpawnGroup(Masked Float=&RandomRangeFloat(Minimum=-1, Maximum=1), Passthrough Float=0)
    Rotation Vector=<0 0 1>
   )
   &SpriteFacingAndAlignment(
    Alignment Coordinate Space=::Local
    Facing Coordinate Space=::Local
    Sprite Alignment=<0 0 -1>
    Sprite Facing=&RotateVectorByQuaternion(Vector To Rotate=<0 0 0>, Quaternion=$Particles.SkeletalMeshLocation.SampledRotation)
    WriteAlignment=true
    WriteFacing=true
   ) {enabled=false}
   &SpriteFacingAndAlignment(Sprite Alignment=<0 0 1>, Sprite Facing=<0 0 1>, WriteAlignment=true, WriteFacing=true) {module=SpriteFacingAndAlignment001}
   &ScaleColor(
    Scale Alpha=&FloatFromCurve(FloatCurve=DataInterface() {propertyValues {Curve {keys [
         {tangentMode=RCTM_User, time=0, value=1}
         {arriveTangent=-0.0405, interpMode=RCIM_Cubic, leaveTangent=-0.0405, tangentMode=RCTM_User, time=0.7009, value=1}
         {arriveTangent=-8.0089, interpMode=RCIM_Cubic, leaveTangent=-8.0089, tangentMode=RCTM_User, time=1, value=0}
        ]}}})
    Scale RGB=&Lerp_Vector(
     A=&VectorFromFloat(Value=&FloatFromCurve(
       FloatCurve=DataInterface() {propertyValues {Curve {
          defaultValue=1
          keys [
           {interpMode=RCIM_Cubic, tangentMode=RCTM_User, time=0, value=0}
           {arriveTangent=16.9912, interpMode=RCIM_Cubic, leaveTangent=16.9912, tangentMode=RCTM_User, time=15, value=100}
          ]
         }}}
       Scale Curve=&Lerp_Float(A=0.5, B=5, Alpha=$Emitter.LoopedAge)
      ))
     B=&VectorFromFloat(Value=&RandomRangeFloat(Minimum=0.3, Maximum=&RandomRangeFloat(Minimum=5, Maximum=20), Evaluation Type=::NewEnumerator1))
     Alpha=&FloatFromCurve(
      FloatCurve=DataInterface() {propertyValues {Curve {keys [
          {interpMode=RCIM_Cubic, leaveTangent=2.5294, tangentMode=RCTM_Break, time=0, value=0}
          {arriveTangent=-0.2166, interpMode=RCIM_Cubic, tangentMode=RCTM_Break, time=0.4, value=1}
         ]}}}
      Scale Curve=0.5
     )
    )
    ScaleA=true
    ScaleRGB=true
   ) {enabled=false}
   &DynamicMaterialParameters(
    Index 0 Param 1=&MakeFloatFromVector(Vector=$Particles.Scale, Channel=::NewEnumerator0)
    Index 0 Param 2=&RandomRangeFloat(Minimum=0, Maximum=10)
    Index 0 Param 3=&FloatFromCurve(
     FloatCurve=DataInterface() {propertyValues {Curve {keys [{interpMode=RCIM_Cubic, leaveTangent=8.1085, tangentMode=RCTM_Break, time=0.8568, value=1}, {tangentMode=RCTM_User, time=1, value=0}]}}}
     Scale Curve=0.07
    )
    Index 0 Param 4=0.8
    Index 1 Param 1=&FloatFromCurve(FloatCurve=DataInterface() {propertyValues {Curve {keys [
         {interpMode=RCIM_Cubic, leaveTangent=1.1034, tangentMode=RCTM_Break, time=0, value=0}
         {arriveTangent=2.2399, tangentMode=RCTM_User, time=0.5, value=0.5}
        ]}}})
    Index 1 Param 2=&MaskFloatBySpawnGroup(
     Masked Float=&FloatFromCurve(FloatCurve=DataInterface() {propertyValues {Curve {keys [
          {interpMode=RCIM_Cubic, tangentMode=RCTM_User, time=0, value=0}
          {arriveTangent=2.7735, interpMode=RCIM_Cubic, leaveTangent=2.7735, tangentMode=RCTM_User, time=1, value=1}
         ]}}})
     Passthrough Float=&FloatFromCurve(FloatCurve=DataInterface() {propertyValues {Curve {keys [<at 0.5 0>, {arriveTangent=3.2053, tangentMode=RCTM_User, time=0.9, value=1}]}}})
    )
    Index 1 Param 3=&FloatFromCurve(FloatCurve=DataInterface() {propertyValues {Curve {keys [<at 0 0 RCIM_Cubic>, {arriveTangent=3.4223, interpMode=RCIM_Cubic, leaveTangent=3.4223, tangentMode=RCTM_User, time=1, value=1}]}}})
    Index 1 Param 4=&FloatFromCurve(
     FloatCurve=DataInterface() {propertyValues {Curve {keys [
         {interpMode=RCIM_Cubic, leaveTangent=0.1169, tangentMode=RCTM_Break, time=0, value=0}
         {arriveTangent=2.5708, interpMode=RCIM_Cubic, leaveTangent=2.5708, tangentMode=RCTM_User, time=1, value=1}
        ]}}}
     Scale Curve=&RandomRangeFloat(Minimum=1, Maximum=40)
    )
    Index 2 Param 1=&FloatFromCurve(FloatCurve=DataInterface() {propertyValues {Curve {keys [
         {interpMode=RCIM_Cubic, leaveTangent=-0.0025, tangentMode=RCTM_Break, time=0, value=0.5}
         {arriveTangent=-0.0316, interpMode=RCIM_Cubic, tangentMode=RCTM_Break, time=1, value=-0.2}
        ]}}})
    Index 2 Param 2=&FloatFromCurve(
     FloatCurve=DataInterface() {propertyValues {Curve {keys [
         {interpMode=RCIM_Cubic, tangentMode=RCTM_User, time=0, value=0}
         {arriveTangent=2.0522, interpMode=RCIM_Cubic, leaveTangent=2.0522, tangentMode=RCTM_User, time=1, value=1}
        ]}}}
     Scale Curve=2
    )
    Module.Index0Param0Write=true
    Module.Index0Param1Write=true
    Module.Index0Param2Write=true
    Module.Index0Param3Write=true
    Module.Index1Param0Write=true
    Module.Index1Param1Write=true
    Module.Index1Param2Write=true
    Module.Index1Param3Write=true
    Module.Index2Param0Write=true
    Module.Index2Param1Write=true
    Module.Index2Param2Write=false
    Module.Index2Param3Write=false
    Module.Index3Param0Write=false
    Module.Index3Param1Write=false
    Module.Index3Param2Write=false
    Module.Index3Param3Write=true
    Write Parameter Index 1=true
    Write Parameter Index 2=true
    Write Parameter Index 3=true
   )
   &FN_Light_Attributes(
    Color=<rgba 0 0 0 1>
    Create Color Attribute=true
    Create Exponent Attribute=true
    Create Radius Attribute=true
    Exponent=5
    Radius=80
    Scale Color=&Multiply_Float(A=7, B=$Emitter.Flicker)
   ) {module=FN_Light_Attributes001}
  ]
  rendererClasses [
   &NiagaraSpriteRendererProperties
   &NiagaraMeshRendererProperties
   &NiagaraLightRendererProperties
  ]
  renderers [
   {rendererClass=&NiagaraSpriteRendererProperties}
   {rendererClass=&NiagaraMeshRendererProperties, rendererIndex=1}
   {rendererClass=&NiagaraMeshRendererProperties, rendererIndex=2}
   {rendererClass=&NiagaraSpriteRendererProperties, rendererIndex=3}
   {rendererClass=&NiagaraLightRendererProperties, rendererIndex=4}
  ]
 }
 Lightning002 {
  bEnabled=false
  emitterUpdateScript [
   &EmitterState(
    Inactive Response=::NewEnumerator1
    Life Cycle Mode=::NewEnumerator1
    Loop Behavior=::NewEnumerator1
    Loop Duration=6.4167
    UseLoopDelay=false
   )
   &SpawnRate(SpawnGroup=1, SpawnRate=&RandomRangeFloat(Minimum=35, Maximum=75), Use Spawn Probability=false) {module=SpawnRate001}
  ]
  particleSpawnScript [
   &InitializeParticle(
    Color=&Multiply_LinearColorByFloat(Linear Color=<rgba 0.2582 0.0437 0.6867 1>, Float=0.4)
    Lifetime Max=0.45
    Lifetime Min:NiagaraFloat=&MaskFloatBySpawnGroup(
     Masked Float=&RandomRangeFloat(Minimum=0.175, Maximum=0.4)
     Passthrough Float=&RandomRangeFloat(Minimum=0.25, Maximum=0.4)
    )
    Lifetime Mode=::NewEnumerator1
    Mesh Scale Max=&MaskVectorBySpawnGroup(Masked Vector=<1.3 1.3 2.5>, Passthrough Vector=<2 2 3>)
    Mesh Scale Min=&MaskVectorBySpawnGroup(Masked Vector=<1 1 2.5>, Passthrough Vector=<1 1 2>)
    Mesh Scale Mode=::NewEnumerator2
    UsePositionOffset=true
    Write Material Random=true
   )
   /SetVariables(Particles.VisibilityTag=&SpawnGroup()) {enabled=false, isSetParameters=true, module=SetVariables_#4}
   &ShapeLocation(
    Cone Angle=12
    Cone Inner Angle=25.7489
    Cone Length=-10
    MaskSpawnGroup=false
    Offset=<0 0 10>
    Offset Mode=::NewEnumerator0
    Override Seed=false
    Shape Primitive=::NewEnumerator5
    UseOwnerScale=true
   ) {module=ShapeLocation001}
   &InitialMeshOrientation(
    Mesh Axis To Orient=&MaskVectorBySpawnGroup(Masked Vector=<0 0 0>, Passthrough Vector=<0 0 0>)
    Mesh Orientation Mode=::NewEnumerator2
    Orientation Vector=&MaskVectorBySpawnGroup(Masked Vector=<0 0 0>, Passthrough Vector=<0 0 0>)
    Rotation=&MaskVectorBySpawnGroup(
     Masked Vector=&RandomRangeVector(Minimum=<0 0 0>, Maximum=<0 0 1>)
     Passthrough Vector=&RandomRangeVector(Minimum=<-0.05 -0.05 0.5>, Maximum=<0.05 0.05 1>)
    )
    Use Rotation Vector=true
   )
   &AddVelocity(
    Velocity=<0 0 1>
    Velocity Speed Scale=&MaskFloatBySpawnGroup(Masked Float=&RandomRangeFloat(Minimum=5, Maximum=5), Passthrough Float=1)
   )
   &CameraOffset(Camera Offset Amount=&MaskFloatBySpawnGroup(Masked Float=0, Passthrough Float=2))
  ]
  particleUpdateScript [
   &ParticleState() {module=ParticleState001}
   &SolveForcesAndVelocity(Clamp Velocity=false, Limit Acceleration=false)
   &ParticleState() {enabled=false}
   &ScaleMeshSize(Scale Factor=&MaskVectorBySpawnGroup(
     Masked Vector=<1 1 1>
     Passthrough Vector=&MakeVector(
      X=1
      Y=2
      Z=&FloatFromCurve(FloatCurve=DataInterface() {propertyValues {Curve {keys [
           {interpMode=RCIM_Cubic, leaveTangent=0.0966, tangentMode=RCTM_Break, time=0, value=1}
           {interpMode=RCIM_Cubic, leaveTangent=-0.1745, time=0.7665, value=1.07}
          ]}}})
     )
    )) {module=ScaleMeshSize001}
   &ScaleMeshSize(Scale Factor=&MaskVectorBySpawnGroup(
     Masked Vector=&Multiply_VectorByFloat(Vector=<1 1 1>, Float=&RandomRangeFloat(Minimum=0.35, Maximum=0.5))
     Passthrough Vector=&VectorFromFloat(Value=&RandomRangeFloat(Minimum=0.1, Maximum=0.15))
    ))
   &ScaleColor(
    Scale Alpha=&FloatFromCurve(FloatCurve=DataInterface() {propertyValues {Curve {keys [
         {tangentMode=RCTM_User, time=0, value=1}
         {arriveTangent=-0.0405, interpMode=RCIM_Cubic, leaveTangent=-0.0405, tangentMode=RCTM_User, time=0.7009, value=1}
         {arriveTangent=-8.0089, interpMode=RCIM_Cubic, leaveTangent=-8.0089, tangentMode=RCTM_User, time=1, value=0}
        ]}}})
    Scale RGB=&Lerp_Vector(
     A=&VectorFromFloat(Value=&FloatFromCurve(
       FloatCurve=DataInterface() {propertyValues {Curve {
          defaultValue=1
          keys [
           {interpMode=RCIM_Cubic, tangentMode=RCTM_User, time=0, value=0}
           {arriveTangent=16.9912, interpMode=RCIM_Cubic, leaveTangent=16.9912, tangentMode=RCTM_User, time=15, value=100}
          ]
         }}}
       Scale Curve=&Lerp_Float(A=0.5, B=5, Alpha=$Emitter.LoopedAge)
      ))
     B=&VectorFromFloat(Value=&RandomRangeFloat(Minimum=0.3, Maximum=&RandomRangeFloat(Minimum=5, Maximum=20), Evaluation Type=::NewEnumerator1))
     Alpha=&FloatFromCurve(
      FloatCurve=DataInterface() {propertyValues {Curve {keys [
          {interpMode=RCIM_Cubic, leaveTangent=2.5294, tangentMode=RCTM_Break, time=0, value=0}
          {arriveTangent=-0.2166, interpMode=RCIM_Cubic, tangentMode=RCTM_Break, time=0.4, value=1}
         ]}}}
      Scale Curve=0.5
     )
    )
    ScaleA=true
    ScaleRGB=true
   )
   &DynamicMaterialParameters(
    Index 0 Param 1=&MakeFloatFromVector(Vector=$Particles.Scale, Channel=::NewEnumerator0)
    Index 0 Param 2=&RandomRangeFloat(Minimum=0, Maximum=10)
    Index 0 Param 3=&FloatFromCurve(
     FloatCurve=DataInterface() {propertyValues {Curve {keys [{interpMode=RCIM_Cubic, leaveTangent=8.1085, tangentMode=RCTM_Break, time=0.8568, value=1}, {tangentMode=RCTM_User, time=1, value=0}]}}}
     Scale Curve=0.07
    )
    Index 0 Param 4=0.8
    Index 1 Param 1=&FloatFromCurve(FloatCurve=DataInterface() {propertyValues {Curve {keys [
         {interpMode=RCIM_Cubic, leaveTangent=1.1034, tangentMode=RCTM_Break, time=0, value=0}
         {arriveTangent=2.2399, tangentMode=RCTM_User, time=0.5, value=0.5}
        ]}}})
    Index 1 Param 2=&MaskFloatBySpawnGroup(
     Masked Float=&FloatFromCurve(FloatCurve=DataInterface() {propertyValues {Curve {keys [
          {interpMode=RCIM_Cubic, tangentMode=RCTM_User, time=0, value=0}
          {arriveTangent=2.7735, interpMode=RCIM_Cubic, leaveTangent=2.7735, tangentMode=RCTM_User, time=1, value=1}
         ]}}})
     Passthrough Float=&FloatFromCurve(FloatCurve=DataInterface() {propertyValues {Curve {keys [<at 0.5 0>, {arriveTangent=3.2053, tangentMode=RCTM_User, time=0.9, value=1}]}}})
    )
    Index 1 Param 3=&FloatFromCurve(FloatCurve=DataInterface() {propertyValues {Curve {keys [<at 0 0 RCIM_Cubic>, {arriveTangent=3.4223, interpMode=RCIM_Cubic, leaveTangent=3.4223, tangentMode=RCTM_User, time=1, value=1}]}}})
    Index 1 Param 4=&FloatFromCurve(
     FloatCurve=DataInterface() {propertyValues {Curve {keys [
         {interpMode=RCIM_Cubic, leaveTangent=0.1169, tangentMode=RCTM_Break, time=0, value=0}
         {arriveTangent=2.5708, interpMode=RCIM_Cubic, leaveTangent=2.5708, tangentMode=RCTM_User, time=1, value=1}
        ]}}}
     Scale Curve=&RandomRangeFloat(Minimum=1, Maximum=40)
    )
    Index 2 Param 1=&FloatFromCurve(FloatCurve=DataInterface() {propertyValues {Curve {keys [
         {interpMode=RCIM_Cubic, leaveTangent=-0.0025, tangentMode=RCTM_Break, time=0, value=0.5}
         {arriveTangent=-0.0316, interpMode=RCIM_Cubic, tangentMode=RCTM_Break, time=1, value=-0.2}
        ]}}})
    Index 2 Param 2=&FloatFromCurve(
     FloatCurve=DataInterface() {propertyValues {Curve {keys [
         {interpMode=RCIM_Cubic, tangentMode=RCTM_User, time=0, value=0}
         {arriveTangent=2.0522, interpMode=RCIM_Cubic, leaveTangent=2.0522, tangentMode=RCTM_User, time=1, value=1}
        ]}}}
     Scale Curve=2
    )
    Module.Index0Param0Write=true
    Module.Index0Param1Write=true
    Module.Index0Param2Write=true
    Module.Index0Param3Write=true
    Module.Index1Param0Write=true
    Module.Index1Param1Write=true
    Module.Index1Param2Write=true
    Module.Index1Param3Write=true
    Module.Index2Param0Write=true
    Module.Index2Param1Write=true
    Module.Index2Param2Write=false
    Module.Index2Param3Write=false
    Module.Index3Param0Write=false
    Module.Index3Param1Write=false
    Module.Index3Param2Write=false
    Module.Index3Param3Write=true
    Write Parameter Index 1=true
    Write Parameter Index 2=true
    Write Parameter Index 3=true
   )
   &UpdateMeshOrientation(
    Rotation Rate=&MaskFloatBySpawnGroup(Masked Float=&RandomRangeFloat(Minimum=-1, Maximum=1), Passthrough Float=0)
    Rotation Vector=<0 0 1>
   )
  ]
  rendererClasses [&NiagaraMeshRendererProperties]
  renderers [
   {rendererClass=&NiagaraMeshRendererProperties, rendererIndex=0}
   {rendererClass=&NiagaraMeshRendererProperties, rendererIndex=1}
  ]
 }
 Lightning_Up001 {
  bEnabled=false
  emitterUpdateScript [
   &EmitterState(
    Inactive Response=::NewEnumerator1
    Life Cycle Mode=::NewEnumerator1
    Loop Behavior=::NewEnumerator1
    Loop Duration Mode=::NewEnumerator1
    UseLoopDelay=false
   )
   &SpawnBurst_Instantaneous(Spawn Time=4.8, Use Spawn Probability=false, UseLoopCountLimit=false) {module=SpawnBurst_Instantaneous001}
   &SpawnBurst_Instantaneous(Spawn Time=8.0333, Use Spawn Probability=false, UseLoopCountLimit=false) {module=SpawnBurst_Instantaneous002}
   &SpawnRate(Spawn Probability=0.85, SpawnGroup=1, SpawnRate=15, Use Spawn Probability=true)
   &SpawnBurst_Instantaneous(Spawn Count=5, Spawn Group=1, Use Spawn Probability=false, UseLoopCountLimit=false)
   &SpawnBurst_Instantaneous(Spawn Group=2, Use Spawn Probability=false, UseLoopCountLimit=false) {module=SpawnBurst_Instantaneous003}
   &Waveform(
    EnableExponent1=false
    GlobalClampMax=false
    GlobalClampMin=false
    UseInput1=false
    UsePhase1=false
    "Waveform [1]":ENiagara_Waveforms=::NewEnumerator12
    "[1] Amplitude Min/Max":Vector2f=<0.75 1>
    "[1] Frequency"=7
   ) {module=Waveform001}
   /SetVariables(Emitter.Flicker:NiagaraFloat=&Add_Float(A=&Multiply_Float(A=$Emitter.Waveform.Waveform, B=0.5), B=0.5)) {isSetParameters=true, module=SetVariables_#9}
   /SetVariables(Emitter.LightColor:LinearColor=&MultiplyLinearColor(
     Linear Color=<rgba 1.142 0.1228 3 0.142>
     Scale Factor=&MakeVector4_FromFloat(Float=&Multiply_Float(
       A=$Emitter.Flicker
       B=&FN_FrontEndFloatModifier(
        Gameplay=&Multiply_Float(A=2, B=&FNTimeOfDayScalar(Night Scale=0.73, Morning Scale=0.85, Evening Scale=0.85))
        Front End=0.7
       )
      ))
    )) {isSetParameters=true, module=SetVariables_#10}
  ]
  particleSpawnScript [
   /SetVariables(Particles.VisibilityTag=&SpawnGroup()) {isSetParameters=true, module=SetVariables_#11}
   &InitializeParticle(
    Color=<rgba 0.4018 0.03 1 1>
    Lifetime=&MaskFloat(
     Mask Source=0
     Mask Parameter=?
     Mask=0
     Masked Float=1
     Passthrough Float=&MaskFloat(Mask Source=0, Mask Parameter=?, Mask=1, Masked Float=1, Passthrough Float=6)
    )
    Mesh Scale=<0.8 0.8 1.2>
    Mesh Scale Mode=::NewEnumerator4
    Position Offset=<0 0 30>
    UsePositionOffset=true
    Write Material Random=true
   )
   &InitialMeshOrientation(
    Mesh Orientation Mode=::NewEnumerator2
    Orientation Vector=&RandomRangeVector(Minimum=<-1 -1 0>, Maximum=<1 1 0>)
    Rotation=<0 0 1>
    Use Rotation Vector=true
   )
   /SetVariables(
    Particles.Electric arcs Lifetime:NiagaraFloat=&RandomRangeFloat(Minimum=0.1, Maximum=1.1)
    Particles.Electric arcs mesh index:NiagaraInt32=&MakeIntFromFloat(Float=&FloatFromCurve(
      FloatCurve=DataInterface() {propertyValues {Curve {keys [
          {arriveTangent=0.6236, interpMode=RCIM_Cubic, tangentMode=RCTM_Break, time=0.25, value=0}
          {arriveTangent=22.9944, interpMode=RCIM_Cubic, tangentMode=RCTM_Break, time=1, value=5}
         ]}}}
      CurveIndex=&RandomRangeFloat(Minimum=0, Maximum=1)
     ))
    Particles.electric arcs position:NiagaraPosition=&AddVectorToPosition(Position=<0 0 34.9679>, Vector=<0 0 0>)
   ) {isSetParameters=true, module=SetVariables_#12}
   /SetVariables(
    Particles.Flare CameraOffset=7
    Particles.Flare Color:LinearColor=<rgba 0.2582 0.0437 0.6867 1>
    Particles.Flare lifetime=0.5
    Particles.Flare size:Vector2f=<30 30>
    Particles.flare position:NiagaraPosition=&AddVectorToPosition(Position=<0 0 15.9742>, Vector=<0 0 0>)
   ) {isSetParameters=true, module=SetVariables_#13}
   /SetVariables(Particles.Staff glow Position:NiagaraPosition=&AddVectorToPosition(Position=<0 0 2.855>, Vector=<0 0 0>)) {isSetParameters=true, module=SetVariables_#14}
  ]
  particleUpdateScript [
   &ParticleState()
   /SetVariables(Particles.Update Flare age:NiagaraFloat=&Divide_Float(A=$Particles.Age, B=$Particles.Initial.Flare lifetime)) {isSetParameters=true, module=SetVariables_#15}
   /SetVariables(Particles.LightningFlicker:NiagaraFloat=/Niagara/DynamicInputs/Helpers/Waveform(
     Global Input Position=&RandomRangeFloat(Minimum=0, Maximum=1)
     Global Phase=$Particles.Age
     "Waveform [1]"=::NewEnumerator12
     "[1] Phase Scale"=66
     "[1] Frequency"=1
     "[1] Amplitude Min/Max"=<0.4 1>
    )) {isSetParameters=true, module=SetVariables_#16}
   /SetVariables(
    Particles.Update Flare color:LinearColor=&MultiplyLinearColor(
     Linear Color=$Particles.Initial.Flare Color
     Scale Factor=&MakeVector4_FromLinearColor(LinearColor=&MakeLinearColorFromVectorAndFloat(
       "Vector (RGB)"=&Lerp_Vector(
        A=&VectorFromFloat(Value=&FloatFromCurve(
          FloatCurve=DataInterface() {propertyValues {Curve {keys [<at 0 1>,<at 0.1 1>,<at 0.2 2>,<at 0.3 1>,<at 0.6 1>,<at 0.7 2>,<at 0.8 1>,<at 1 1>]}}}
          CurveIndex=$Particles.Update Flare age
         ))
        B=&VectorFromFloat(Value=&RandomRangeFloat(Minimum=0.3, Maximum=3, Evaluation Type=::NewEnumerator1))
        Alpha=&FloatFromCurve(
         FloatCurve=DataInterface() {propertyValues {Curve {keys [
             {interpMode=RCIM_Cubic, leaveTangent=2.5294, tangentMode=RCTM_Break, time=0, value=0}
             {arriveTangent=-0.2166, interpMode=RCIM_Cubic, tangentMode=RCTM_Break, time=0.4, value=1}
            ]}}}
         CurveIndex=$Particles.Update Flare age
        )
       )
       "Float (Alpha)"=&FloatFromCurve(
        FloatCurve=DataInterface() {propertyValues {Curve {keys [
            {tangentMode=RCTM_User, time=0, value=1}
            {arriveTangent=-0.0405, interpMode=RCIM_Cubic, leaveTangent=-0.0405, tangentMode=RCTM_User, time=0.7009, value=1}
            {arriveTangent=-8.0089, interpMode=RCIM_Cubic, leaveTangent=-8.0089, tangentMode=RCTM_User, time=1, value=0}
           ]}}}
        CurveIndex=$Particles.Update Flare age
       )
      ))
    )
    Particles.Update Flare size:Vector2f=&Multiply_Vector2D_ByFloat(
     Float=&FloatFromCurve(
      FloatCurve=DataInterface() {propertyValues {Curve {keys [{tangentMode=RCTM_User, time=0, value=0}, {tangentMode=RCTM_User, time=0.3, value=5}]}}}
      CurveIndex=$Particles.Update Flare age
     )
     Vector2D=$Particles.Initial.Flare size
    )
   ) {isSetParameters=true, module=SetVariables_#17}
   &ScaleColor(
    Scale Alpha=&MaskFloat(
     Mask Source=0
     Mask Parameter=?
     Mask=0
     Masked Float=0
     Passthrough Float=&MaskFloat(
      Mask Source=0
      Mask Parameter=?
      Mask=1
      Masked Float=&FloatFromCurve(
       FloatCurve=DataInterface() {propertyValues {Curve {keys [
           {arriveTangent=9.0725, interpMode=RCIM_Cubic, leaveTangent=9.0725, tangentMode=RCTM_User, time=0, value=0}
           {arriveTangent=-2.5249, interpMode=RCIM_Cubic, leaveTangent=-2.5249, tangentMode=RCTM_User, time=0.1408, value=0.9847}
           {arriveTangent=-0.0405, interpMode=RCIM_Cubic, leaveTangent=-0.0405, tangentMode=RCTM_User, time=1, value=0}
          ]}}}
       Scale Curve=$Particles.LightningFlicker
      )
      Passthrough Float=&FloatFromCurve(
       FloatCurve=DataInterface() {propertyValues {Curve {keys [<at 0 0.1>,<at 1 0>]}}}
       Scale Curve=&Multiply_Float(A=1, B=$Emitter.Flicker)
      )
     )
    )
    Scale RGB=&MaskVector(
     Mask Source=0
     Mask Parameter=?
     Mask=0
     Masked Vector=<0 0 0>
     Passthrough Vector=&MaskVector(
      Mask Source=0
      Mask Parameter=?
      Mask=1
      Masked Vector=&VectorFromFloat(Value=&FloatFromCurve(
        FloatCurve=DataInterface() {propertyValues {Curve {keys [
            {arriveTangent=9.0725, interpMode=RCIM_Cubic, leaveTangent=9.0725, tangentMode=RCTM_Break, time=0, value=0}
            {arriveTangent=-2.5249, interpMode=RCIM_Cubic, leaveTangent=-2.5249, tangentMode=RCTM_Break, time=0.1408, value=1}
            {arriveTangent=-0.0405, interpMode=RCIM_Cubic, leaveTangent=-0.0405, tangentMode=RCTM_User, time=1, value=0}
           ]}}}
        Scale Curve=$Particles.LightningFlicker
       ))
      Passthrough Vector=&Lerp_Vector(
       A=&VectorFromFloat(Value=&FloatFromCurve(
         FloatCurve=DataInterface() {propertyValues {Curve {
            defaultValue=1
            keys [
             {interpMode=RCIM_Cubic, tangentMode=RCTM_User, time=0, value=0}
             {arriveTangent=2.811, interpMode=RCIM_Cubic, leaveTangent=2.811, tangentMode=RCTM_User, time=1, value=1}
            ]
           }}}
         Scale Curve=&Lerp_Float(A=1, B=2, Alpha=$Emitter.Age)
        ))
       B=&VectorFromFloat(Value=&FloatFromCurve(
         FloatCurve=DataInterface() {propertyValues {Curve {
            defaultValue=1
            keys [
             {interpMode=RCIM_Cubic, tangentMode=RCTM_User, time=0, value=0.3}
             {arriveTangent=2.811, interpMode=RCIM_Cubic, leaveTangent=2.811, tangentMode=RCTM_User, time=1, value=1}
            ]
           }}}
         Scale Curve=&Lerp_Float(A=1, B=2, Alpha=$Emitter.Age)
        ))
       Alpha=&FloatFromCurve(
        FloatCurve=DataInterface() {propertyValues {Curve {keys [
            {interpMode=RCIM_Cubic, leaveTangent=2.5294, tangentMode=RCTM_Break, time=0, value=0}
            {arriveTangent=-0.2166, interpMode=RCIM_Cubic, tangentMode=RCTM_Break, time=0.4, value=1}
            <at 1.1 0 RCIM_Cubic>
           ]}}}
        Scale Curve=5
       )
      )
     )
    )
    ScaleA=true
    ScaleRGB=true
   ) {module=ScaleColor001}
   /SetVariables(
    Particles.ColorCore:LinearColor=&MultiplyLinearColor(
     Linear Color=$Particles.Color
     Scale Factor=&MakeFloat4FromVectorAndFloat("Vector (RGB)"=&VectorFromFloat(Value=33), "Float (Alpha)"=1)
    )
    Particles.ColorGlow:LinearColor=&MultiplyLinearColor(
     Linear Color=$Particles.Color
     Scale Factor=&MakeFloat4FromVectorAndFloat("Vector (RGB)"=&VectorFromFloat(Value=0.15), "Float (Alpha)"=1)
    )
    Particles.DynamicCore:Vector4f=<1 6 2 0>
    Particles.DynamicGlow:Vector4f=<5.0981 3 4 0>
    Particles.LightningScale:NiagaraFloat=&FloatFromCurve(FloatCurve=DataInterface() {propertyValues {Curve {keys [
         {arriveTangent=2.7028, interpMode=RCIM_Constant, time=0, value=0.75}
         {interpMode=RCIM_Cubic, tangentMode=RCTM_User, time=0.5, value=1}
         {arriveTangent=-2.6499, tangentMode=RCTM_User, time=1, value=0.5}
        ]}}})
   ) {isSetParameters=true, module=SetVariables_#18}
   /SetVariables(
    Particles.CoreGlowColor:LinearColor=&MultiplyLinearColor(
     Linear Color=<rgba 1.142 0.1228 3 0.25>
     Scale Factor=&MakeVector4_FromFloat(Float=&Multiply_Float(
       A=$Emitter.Flicker
       B=&FN_FrontEndFloatModifier(Gameplay=&FNTimeOfDayScalar(Night Scale=2, Morning Scale=1, Evening Scale=1), Front End=0.7)
      ))
    )
    Particles.CoreGlowSize:Vector2f=<12 40>
    Particles.DarkGlowSize:Vector2f=&Vector2DFromFloat(Value=75)
    Particles.SpriteAlignment=<0 0 1>
   ) {isSetParameters=true, module=SetVariables_#19}
   /SetVariables(
    Particles.CoreGlowColor:LinearColor=&MultiplyLinearColor(
     Linear Color=<rgba 1.142 0.1228 3 0.25>
     Scale Factor=&MakeVector4_FromFloat(Float=&Multiply_Float(
       A=$Emitter.Flicker
       B=&FN_FrontEndFloatModifier(Gameplay=&FNTimeOfDayScalar(Night Scale=1, Morning Scale=1, Evening Scale=1), Front End=0.7)
      ))
    )
    Particles.DarkGlowColor:LinearColor=&MultiplyLinearColor(
     Linear Color=<rgba 0.2854 0.0292 0.564 0.12>
     Scale Factor=&MakeVector4_FromFloat(Float=&Multiply_Float(
       A=$Emitter.Flicker
       B=&FN_FrontEndFloatModifier(Gameplay=&FNTimeOfDayScalar(Night Scale=1, Morning Scale=1, Evening Scale=1), Front End=0.7)
      ))
    )
   ) {isSetParameters=true, module=SetVariables_#20}
   &ScaleMeshSize(Scale Factor=&MakeVector(X=$Particles.LightningScale, Y=$Particles.LightningScale, Z=1))
   &FN_Light_Attributes(
    Color=<rgba 0.2582 0.0437 0.6867 1>
    Create Color Attribute=true
    Create Exponent Attribute=true
    Create Radius Attribute=true
    Exponent=5
    Radius=30
    Scale Color=&Multiply_Float(A=7, B=$Emitter.Flicker)
   ) {module=FN_Light_Attributes001}
   &UpdateMeshOrientation(
    Rotation Rate=&RandomRangeFloat(Minimum=-0.5, Maximum=0.5)
    Rotation Vector=&RotateVectorByQuaternion(Vector To Rotate=<0 0 1>, Quaternion=$StackContext.SkeletalMeshLocation.SampledRotation)
   )
  ]
  rendererClasses [
   &NiagaraSpriteRendererProperties
   &NiagaraMeshRendererProperties
   &NiagaraLightRendererProperties
  ]
  renderers [
   {rendererClass=&NiagaraSpriteRendererProperties}
   {rendererClass=&NiagaraMeshRendererProperties, rendererIndex=1}
   {rendererClass=&NiagaraMeshRendererProperties, rendererIndex=2}
   {rendererClass=&NiagaraLightRendererProperties, rendererIndex=3}
   {rendererClass=&NiagaraLightRendererProperties, rendererIndex=4}
   {rendererClass=&NiagaraSpriteRendererProperties, rendererIndex=5}
   {rendererClass=&NiagaraSpriteRendererProperties, rendererIndex=6}
  ]
 }
 Lightning_down {
  bEnabled=false
  emitterUpdateScript [
   &EmitterState(
    Inactive Response=::NewEnumerator1
    Life Cycle Mode=::NewEnumerator1
    Loop Behavior=::NewEnumerator1
    Loop Duration Mode=::NewEnumerator1
    UseLoopDelay=false
   )
   &SpawnBurst_Instantaneous(Spawn Time=4.8, Use Spawn Probability=false, UseLoopCountLimit=false) {module=SpawnBurst_Instantaneous001}
   &SpawnBurst_Instantaneous(Spawn Time=8.0333, Use Spawn Probability=false, UseLoopCountLimit=false) {module=SpawnBurst_Instantaneous002}
   &SpawnRate(Spawn Probability=0.85, SpawnGroup=1, SpawnRate=15, Use Spawn Probability=true)
   &SpawnBurst_Instantaneous(Spawn Count=5, Spawn Group=1, Use Spawn Probability=false, UseLoopCountLimit=false)
   &SpawnBurst_Instantaneous(Spawn Group=2, Use Spawn Probability=false, UseLoopCountLimit=false) {module=SpawnBurst_Instantaneous003}
   /Niagara/Noise/Waveform(
    EnableExponent1=false
    GlobalClampMax=false
    GlobalClampMin=false
    UseInput1=false
    UsePhase1=false
    "Waveform [1]":ENiagara_Waveforms=::NewEnumerator12
    "[1] Amplitude Min/Max":Vector2f=<0.75 1>
    "[1] Frequency"=7
   ) {module=Waveform001}
   /SetVariables(Emitter.Flicker:NiagaraFloat=&Add_Float(A=&Multiply_Float(A=$Emitter.Waveform.Waveform, B=0.5), B=0.5)) {isSetParameters=true, module=SetVariables_#9}
   /SetVariables(Emitter.LightColor:LinearColor=&MultiplyLinearColor(
     Linear Color=<rgba 1.142 0.1228 3 0.142>
     Scale Factor=&MakeVector4_FromFloat(Float=&Multiply_Float(
       A=$Emitter.Flicker
       B=&FN_FrontEndFloatModifier(
        Gameplay=&Multiply_Float(A=2, B=&FNTimeOfDayScalar(Night Scale=0.73, Morning Scale=0.85, Evening Scale=0.85))
        Front End=0.7
       )
      ))
    )) {isSetParameters=true, module=SetVariables_#10}
  ]
  particleSpawnScript [
   /SetVariables(Particles.VisibilityTag=&SpawnGroup()) {isSetParameters=true, module=SetVariables_#11}
   &InitializeParticle(
    Color=<rgba 0.4018 0.03 1 1>
    Lifetime=&MaskFloat(
     Mask Source=0
     Mask Parameter=?
     Mask=0
     Masked Float=1
     Passthrough Float=&MaskFloat(Mask Source=0, Mask Parameter=?, Mask=1, Masked Float=1, Passthrough Float=6)
    )
    Mesh Scale=<0.8 0.8 1.2>
    Mesh Scale Mode=::NewEnumerator4
    Position Offset=<0 0 30>
    UsePositionOffset=true
    Write Material Random=true
   )
   &InitialMeshOrientation(
    Mesh Orientation Mode=::NewEnumerator2
    Orientation Vector=&RandomRangeVector(Minimum=<-1 -1 0>, Maximum=<1 1 0>)
    Rotation=<0 0 1>
    Use Rotation Vector=true
   )
   /SetVariables(
    Particles.Electric arcs Lifetime:NiagaraFloat=&RandomRangeFloat(Minimum=0.1, Maximum=1.1)
    Particles.Electric arcs mesh index:NiagaraInt32=&MakeIntFromFloat(Float=&FloatFromCurve(
      FloatCurve=DataInterface() {propertyValues {Curve {keys [
          {arriveTangent=0.6236, interpMode=RCIM_Cubic, tangentMode=RCTM_Break, time=0.25, value=0}
          {arriveTangent=22.9944, interpMode=RCIM_Cubic, tangentMode=RCTM_Break, time=1, value=5}
         ]}}}
      CurveIndex=&RandomRangeFloat(Minimum=0, Maximum=1)
     ))
    Particles.electric arcs position:NiagaraPosition=&AddVectorToPosition(Position=<0 0 -90>, Vector=<0 0 0>)
   ) {isSetParameters=true, module=SetVariables_#12}
   /SetVariables(
    Particles.Flare CameraOffset=7
    Particles.Flare Color:LinearColor=<rgba 0.2582 0.0437 0.6867 1>
    Particles.Flare lifetime=0.5
    Particles.Flare size:Vector2f=<30 30>
    Particles.flare position:NiagaraPosition=&AddVectorToPosition(Position=<0 0 -104.4797>, Vector=<0 0 0>)
   ) {isSetParameters=true, module=SetVariables_#13}
   /SetVariables(Particles.Staff glow Position:NiagaraPosition=&AddVectorToPosition(Position=<0 0 -126.35>, Vector=<0 0 0>)) {isSetParameters=true, module=SetVariables_#14}
  ]
  particleUpdateScript [
   &ParticleState()
   /SetVariables(Particles.Update Flare age:NiagaraFloat=&Divide_Float(A=$Particles.Age, B=$Particles.Initial.Flare lifetime)) {isSetParameters=true, module=SetVariables_#15}
   /SetVariables(Particles.LightningFlicker:NiagaraFloat=/Niagara/DynamicInputs/Helpers/Waveform(
     Global Input Position=&RandomRangeFloat(Minimum=0, Maximum=1)
     Global Phase=$Particles.Age
     "Waveform [1]"=::NewEnumerator12
     "[1] Phase Scale"=66
     "[1] Frequency"=1
     "[1] Amplitude Min/Max"=<0.4 1>
    )) {isSetParameters=true, module=SetVariables_#16}
   /SetVariables(
    Particles.Update Flare color:LinearColor=&MultiplyLinearColor(
     Linear Color=$Particles.Initial.Flare Color
     Scale Factor=&MakeVector4_FromLinearColor(LinearColor=&MakeLinearColorFromVectorAndFloat(
       "Vector (RGB)"=&Lerp_Vector(
        A=&VectorFromFloat(Value=&FloatFromCurve(
          FloatCurve=DataInterface() {propertyValues {Curve {keys [<at 0 1>,<at 0.1 1>,<at 0.2 2>,<at 0.3 1>,<at 0.6 1>,<at 0.7 2>,<at 0.8 1>,<at 1 1>]}}}
          CurveIndex=$Particles.Update Flare age
         ))
        B=&VectorFromFloat(Value=&RandomRangeFloat(Minimum=0.3, Maximum=3, Evaluation Type=::NewEnumerator1))
        Alpha=&FloatFromCurve(
         FloatCurve=DataInterface() {propertyValues {Curve {keys [
             {interpMode=RCIM_Cubic, leaveTangent=2.5294, tangentMode=RCTM_Break, time=0, value=0}
             {arriveTangent=-0.2166, interpMode=RCIM_Cubic, tangentMode=RCTM_Break, time=0.4, value=1}
            ]}}}
         CurveIndex=$Particles.Update Flare age
        )
       )
       "Float (Alpha)"=&FloatFromCurve(
        FloatCurve=DataInterface() {propertyValues {Curve {keys [
            {tangentMode=RCTM_User, time=0, value=1}
            {arriveTangent=-0.0405, interpMode=RCIM_Cubic, leaveTangent=-0.0405, tangentMode=RCTM_User, time=0.7009, value=1}
            {arriveTangent=-8.0089, interpMode=RCIM_Cubic, leaveTangent=-8.0089, tangentMode=RCTM_User, time=1, value=0}
           ]}}}
        CurveIndex=$Particles.Update Flare age
       )
      ))
    )
    Particles.Update Flare size:Vector2f=&Multiply_Vector2D_ByFloat(
     Float=&FloatFromCurve(
      FloatCurve=DataInterface() {propertyValues {Curve {keys [{tangentMode=RCTM_User, time=0, value=0}, {tangentMode=RCTM_User, time=0.3, value=5}]}}}
      CurveIndex=$Particles.Update Flare age
     )
     Vector2D=$Particles.Initial.Flare size
    )
   ) {isSetParameters=true, module=SetVariables_#17}
   &ScaleColor(
    Scale Alpha=&MaskFloat(
     Mask Source=0
     Mask Parameter=?
     Mask=0
     Masked Float=0
     Passthrough Float=&MaskFloat(
      Mask Source=0
      Mask Parameter=?
      Mask=1
      Masked Float=&FloatFromCurve(
       FloatCurve=DataInterface() {propertyValues {Curve {keys [
           {arriveTangent=9.0725, interpMode=RCIM_Cubic, leaveTangent=9.0725, tangentMode=RCTM_User, time=0, value=0}
           {arriveTangent=-2.5249, interpMode=RCIM_Cubic, leaveTangent=-2.5249, tangentMode=RCTM_User, time=0.1408, value=0.9847}
           {arriveTangent=-0.0405, interpMode=RCIM_Cubic, leaveTangent=-0.0405, tangentMode=RCTM_User, time=1, value=0}
          ]}}}
       Scale Curve=$Particles.LightningFlicker
      )
      Passthrough Float=&FloatFromCurve(
       FloatCurve=DataInterface() {propertyValues {Curve {keys [<at 0 0.1>,<at 1 0>]}}}
       Scale Curve=&Multiply_Float(A=1, B=$Emitter.Flicker)
      )
     )
    )
    Scale RGB=&MaskVector(
     Mask Source=0
     Mask Parameter=?
     Mask=0
     Masked Vector=<0 0 0>
     Passthrough Vector=&MaskVector(
      Mask Source=0
      Mask Parameter=?
      Mask=1
      Masked Vector=&VectorFromFloat(Value=&FloatFromCurve(
        FloatCurve=DataInterface() {propertyValues {Curve {keys [
            {arriveTangent=9.0725, interpMode=RCIM_Cubic, leaveTangent=9.0725, tangentMode=RCTM_Break, time=0, value=0}
            {arriveTangent=-2.5249, interpMode=RCIM_Cubic, leaveTangent=-2.5249, tangentMode=RCTM_Break, time=0.1408, value=1}
            {arriveTangent=-0.0405, interpMode=RCIM_Cubic, leaveTangent=-0.0405, tangentMode=RCTM_User, time=1, value=0}
           ]}}}
        Scale Curve=$Particles.LightningFlicker
       ))
      Passthrough Vector=&Lerp_Vector(
       A=&VectorFromFloat(Value=&FloatFromCurve(
         FloatCurve=DataInterface() {propertyValues {Curve {
            defaultValue=1
            keys [
             {interpMode=RCIM_Cubic, tangentMode=RCTM_User, time=0, value=0}
             {arriveTangent=2.811, interpMode=RCIM_Cubic, leaveTangent=2.811, tangentMode=RCTM_User, time=1, value=1}
            ]
           }}}
         Scale Curve=&Lerp_Float(A=1, B=2, Alpha=$Emitter.Age)
        ))
       B=&VectorFromFloat(Value=&FloatFromCurve(
         FloatCurve=DataInterface() {propertyValues {Curve {
            defaultValue=1
            keys [
             {interpMode=RCIM_Cubic, tangentMode=RCTM_User, time=0, value=0.3}
             {arriveTangent=2.811, interpMode=RCIM_Cubic, leaveTangent=2.811, tangentMode=RCTM_User, time=1, value=1}
            ]
           }}}
         Scale Curve=&Lerp_Float(A=1, B=2, Alpha=$Emitter.Age)
        ))
       Alpha=&FloatFromCurve(
        FloatCurve=DataInterface() {propertyValues {Curve {keys [
            {interpMode=RCIM_Cubic, leaveTangent=2.5294, tangentMode=RCTM_Break, time=0, value=0}
            {arriveTangent=-0.2166, interpMode=RCIM_Cubic, tangentMode=RCTM_Break, time=0.4, value=1}
            <at 1.1 0 RCIM_Cubic>
           ]}}}
        Scale Curve=5
       )
      )
     )
    )
    ScaleA=true
    ScaleRGB=true
   ) {module=ScaleColor001}
   /SetVariables(
    Particles.ColorCore:LinearColor=&MultiplyLinearColor(
     Linear Color=$Particles.Color
     Scale Factor=&MakeFloat4FromVectorAndFloat("Vector (RGB)"=&VectorFromFloat(Value=33), "Float (Alpha)"=1)
    )
    Particles.ColorGlow:LinearColor=&MultiplyLinearColor(
     Linear Color=$Particles.Color
     Scale Factor=&MakeFloat4FromVectorAndFloat("Vector (RGB)"=&VectorFromFloat(Value=1), "Float (Alpha)"=1)
    )
    Particles.DynamicCore:Vector4f=<1 6 2 0>
    Particles.DynamicGlow:Vector4f=<5.0981 3 4 0>
    Particles.LightningScale:NiagaraFloat=&FloatFromCurve(FloatCurve=DataInterface() {propertyValues {Curve {keys [
         {arriveTangent=2.7028, interpMode=RCIM_Constant, time=0, value=0.75}
         {interpMode=RCIM_Cubic, tangentMode=RCTM_User, time=0.5, value=1}
         {arriveTangent=-2.6499, tangentMode=RCTM_User, time=1, value=0.5}
        ]}}})
   ) {isSetParameters=true, module=SetVariables_#18}
   /SetVariables(
    Particles.CoreGlowColor:LinearColor=&MultiplyLinearColor(
     Linear Color=<rgba 1.142 0.1228 3 0.25>
     Scale Factor=&MakeVector4_FromFloat(Float=&Multiply_Float(
       A=$Emitter.Flicker
       B=&FN_FrontEndFloatModifier(Gameplay=&FNTimeOfDayScalar(Night Scale=2, Morning Scale=1, Evening Scale=1), Front End=0.7)
      ))
    )
    Particles.CoreGlowSize:Vector2f=<12 40>
    Particles.DarkGlowSize:Vector2f=&Vector2DFromFloat(Value=75)
    Particles.SpriteAlignment=<0 0 1>
   ) {isSetParameters=true, module=SetVariables_#19}
   /SetVariables(
    Particles.CoreGlowColor:LinearColor=&MultiplyLinearColor(
     Linear Color=<rgba 1.142 0.1228 3 0.25>
     Scale Factor=&MakeVector4_FromFloat(Float=&Multiply_Float(
       A=$Emitter.Flicker
       B=&FN_FrontEndFloatModifier(Gameplay=&FNTimeOfDayScalar(Night Scale=1, Morning Scale=1, Evening Scale=1), Front End=0.7)
      ))
    )
    Particles.DarkGlowColor:LinearColor=&MultiplyLinearColor(
     Linear Color=<rgba 0.2854 0.0292 0.564 0.12>
     Scale Factor=&MakeVector4_FromFloat(Float=&Multiply_Float(
       A=$Emitter.Flicker
       B=&FN_FrontEndFloatModifier(Gameplay=&FNTimeOfDayScalar(Night Scale=1, Morning Scale=1, Evening Scale=1), Front End=0.7)
      ))
    )
   ) {isSetParameters=true}
   &ScaleMeshSize(Scale Factor=&MakeVector(X=$Particles.LightningScale, Y=$Particles.LightningScale, Z=1))
   &FN_Light_Attributes(
    Color=<rgba 0.2582 0.0437 0.6867 1>
    Create Color Attribute=true
    Create Exponent Attribute=true
    Create Radius Attribute=true
    Exponent=5
    Radius=30
    Scale Color=&Multiply_Float(A=7, B=$Emitter.Flicker)
   ) {module=FN_Light_Attributes001}
   &UpdateMeshOrientation(
    Rotation Rate=&RandomRangeFloat(Minimum=-0.5, Maximum=0.5)
    Rotation Vector=&RotateVectorByQuaternion(Vector To Rotate=<0 0 1>, Quaternion=$StackContext.SkeletalMeshLocation.SampledRotation)
   )
  ]
  rendererClasses [
   &NiagaraSpriteRendererProperties
   &NiagaraMeshRendererProperties
   &NiagaraLightRendererProperties
  ]
  renderers [
   {rendererClass=&NiagaraSpriteRendererProperties}
   {rendererClass=&NiagaraMeshRendererProperties, rendererIndex=1}
   {rendererClass=&NiagaraMeshRendererProperties, rendererIndex=2}
   {rendererClass=&NiagaraLightRendererProperties, rendererIndex=3}
   {rendererClass=&NiagaraLightRendererProperties, rendererIndex=4}
   {rendererClass=&NiagaraSpriteRendererProperties, rendererIndex=5}
   {rendererClass=&NiagaraSpriteRendererProperties, rendererIndex=6}
  ]
 }
 Lightning_up {
  bEnabled=false
  emitterUpdateScript [
   &EmitterState(
    Inactive Response=::NewEnumerator1
    Life Cycle Mode=::NewEnumerator1
    Loop Behavior=::NewEnumerator1
    Loop Duration=6.4167
    UseLoopDelay=false
   )
   &SpawnBurst_Instantaneous(Spawn Time=8.0333, Use Spawn Probability=false, UseLoopCountLimit=false) {module=SpawnBurst_Instantaneous002}
   &SpawnBurst_Instantaneous(Spawn Time=4.8, Use Spawn Probability=false, UseLoopCountLimit=false)
   &SpawnRate(SpawnGroup=1, SpawnRate=&RandomRangeFloat(Minimum=35, Maximum=60), Use Spawn Probability=false) {module=SpawnRate001}
   &SpawnBurst_Instantaneous(Spawn Group=2, Use Spawn Probability=false, UseLoopCountLimit=false) {module=SpawnBurst_Instantaneous001}
   /Niagara/Noise/Waveform(
    EnableExponent1=false
    GlobalClampMax=false
    GlobalClampMin=false
    UseInput1=false
    UsePhase1=false
    "Waveform [1]":ENiagara_Waveforms=::NewEnumerator12
    "[1] Frequency"=7
   )
   /SetVariables(Emitter.Flicker:NiagaraFloat=&Add_Float(A=&Multiply_Float(A=$Emitter.Waveform.Waveform, B=0.5), B=0.5)) {isSetParameters=true, module=SetVariables_#3}
  ]
  particleSpawnScript [
   /SetVariables(Particles.VisibilityTag=&SpawnGroup()) {isSetParameters=true, module=SetVariables_#4}
   &InitializeParticle(
    Color=&Multiply_LinearColorByFloat(Linear Color=<rgba 0.2582 0.0437 0.6867 1>, Float=0.4)
    Lifetime=&MaskFloat(
     Mask Source=0
     Mask Parameter=?
     Mask=0
     Masked Float=1
     Passthrough Float=&MaskFloat(
      Mask Source=0
      Mask Parameter=?
      Mask=1
      Masked Float=&RandomRangeFloat(Minimum=0.24, Maximum=0.4)
      Passthrough Float=6
     )
    )
    Mesh Scale Max=&MaskVectorBySpawnGroup(Masked Vector=<1.3 1.3 2.5>, Passthrough Vector=<2 2 3>)
    Mesh Scale Min=&MaskVectorBySpawnGroup(Masked Vector=<1 1 2.5>, Passthrough Vector=<1 1 2>)
    Mesh Scale Mode=::NewEnumerator2
    Position Offset=&MaskPosition(
     Mask Source:NiagaraInt32=0
     Mask Parameter:NiagaraInt32=?
     Mask:NiagaraInt32=0
     Masked Position:NiagaraPosition=<0 0 0>
     Passthrough Position:NiagaraPosition=&MaskPosition(
      Mask Source:NiagaraInt32=0
      Mask Parameter:NiagaraInt32=?
      Mask:NiagaraInt32=1
      Masked Position:NiagaraPosition=<0 0 0>
      Passthrough Position:NiagaraPosition=<0 0 15>
     )
    )
    Sprite Size Mode=::NewEnumerator3
    Uniform Sprite Size=&MaskFloat(
     Mask Source=0
     Mask Parameter=?
     Mask=0
     Masked Float=0
     Passthrough Float=&MaskFloat(Mask Source=0, Mask Parameter=?, Mask=1, Masked Float=0, Passthrough Float=37)
    )
    UsePositionOffset=true
    Write Material Random=true
   )
   &InitialMeshOrientation(
    Mesh Axis To Orient=&MaskVectorBySpawnGroup(Masked Vector=<0 0 0>, Passthrough Vector=<0 0 0>)
    Mesh Orientation Mode=::NewEnumerator2
    Orientation Vector=&MaskVectorBySpawnGroup(Masked Vector=<0 0 0>, Passthrough Vector=<0 0 0>)
    Rotation=&MaskVectorBySpawnGroup(
     Masked Vector=&RandomRangeVector(Minimum=<0 0 0>, Maximum=<0 0 1>)
     Passthrough Vector=&RandomRangeVector(Minimum=<-0.05 -0.05 0.5>, Maximum=<0.05 0.05 1>)
    )
    Use Rotation Vector=true
   )
   &CameraOffset(Camera Offset Amount=&MaskFloatBySpawnGroup(Masked Float=0, Passthrough Float=2))
   /SetVariables(
    Particles.Flare CameraOffset=7
    Particles.Flare Color:LinearColor=<rgba 0.2582 0.0437 0.6867 1>
    Particles.Flare lifetime=0.5
    Particles.Flare size:Vector2f=<30 30>
    Particles.flare position:NiagaraPosition=&AddVectorToPosition(Position=<0 0 15>, Vector=<0 0 0>)
   ) {isSetParameters=true, module=SetVariables_#5}
   /SetVariables(Particles.Staff color:LinearColor=<rgba 0.2582 0.0437 0.6867 1>, Particles.staff size:Vector2f=<100 100>) {isSetParameters=true, module=SetVariables_#6}
  ]
  particleUpdateScript [
   &ParticleState() {module=ParticleState001}
   /SetVariables(Particles.Update Flare age:NiagaraFloat=&Divide_Float(A=$Particles.Age, B=$Particles.Initial.Flare lifetime)) {isSetParameters=true, module=SetVariables_#7}
   /SetVariables(
    Particles.Update Flare color:LinearColor=&MultiplyLinearColor(
     Linear Color=$Particles.Initial.Flare Color
     Scale Factor=&MakeVector4_FromLinearColor(LinearColor=&MakeLinearColorFromVectorAndFloat(
       "Vector (RGB)"=&Lerp_Vector(
        A=&VectorFromFloat(Value=&FloatFromCurve(
          FloatCurve=DataInterface() {propertyValues {Curve {keys [<at 0 1>,<at 0.1 1>,<at 0.2 2>,<at 0.3 1>,<at 0.6 1>,<at 0.7 2>,<at 0.8 1>,<at 1 1>]}}}
          CurveIndex=$Particles.Update Flare age
         ))
        B=&VectorFromFloat(Value=&RandomRangeFloat(Minimum=0.3, Maximum=3, Evaluation Type=::NewEnumerator1))
        Alpha=&FloatFromCurve(
         FloatCurve=DataInterface() {propertyValues {Curve {keys [
             {interpMode=RCIM_Cubic, leaveTangent=2.5294, tangentMode=RCTM_Break, time=0, value=0}
             {arriveTangent=-0.2166, interpMode=RCIM_Cubic, tangentMode=RCTM_Break, time=0.4, value=1}
            ]}}}
         CurveIndex=$Particles.Update Flare age
         Scale Curve=0.5
        )
       )
       "Float (Alpha)"=&FloatFromCurve(
        FloatCurve=DataInterface() {propertyValues {Curve {keys [
            {tangentMode=RCTM_User, time=0, value=1}
            {arriveTangent=-0.0405, interpMode=RCIM_Cubic, leaveTangent=-0.0405, tangentMode=RCTM_User, time=0.7009, value=1}
            {arriveTangent=-8.0089, interpMode=RCIM_Cubic, leaveTangent=-8.0089, tangentMode=RCTM_User, time=1, value=0}
           ]}}}
        CurveIndex=$Particles.Update Flare age
       )
      ))
    )
    Particles.Update Flare size:Vector2f=&Multiply_Vector2D_ByFloat(
     Float=&FloatFromCurve(
      FloatCurve=DataInterface() {propertyValues {Curve {keys [{tangentMode=RCTM_User, time=0, value=0}, {tangentMode=RCTM_User, time=0.3, value=5}]}}}
      CurveIndex=$Particles.Update Flare age
     )
     Vector2D=$Particles.Initial.Flare size
    )
   ) {isSetParameters=true, module=SetVariables_#8}
   &ScaleColor(
    Scale Alpha=&MaskFloat(
     Mask Source=0
     Mask Parameter=?
     Mask=0
     Masked Float=0
     Passthrough Float=&MaskFloat(
      Mask Source=0
      Mask Parameter=?
      Mask=1
      Masked Float=&FloatFromCurve(FloatCurve=DataInterface() {propertyValues {Curve {keys [
           {tangentMode=RCTM_User, time=0, value=1}
           {arriveTangent=-0.0405, interpMode=RCIM_Cubic, leaveTangent=-0.0405, tangentMode=RCTM_User, time=0.7009, value=1}
           {arriveTangent=-8.0089, interpMode=RCIM_Cubic, leaveTangent=-8.0089, tangentMode=RCTM_User, time=1, value=0}
          ]}}})
      Passthrough Float=&Multiply_Float(A=$Emitter.Flicker, B=10)
     )
    )
    Scale RGB=&MaskVector(
     Mask Source=0
     Mask Parameter=?
     Mask=0
     Masked Vector=<0 0 0>
     Passthrough Vector=&MaskVector(
      Mask Source=0
      Mask Parameter=?
      Mask=1
      Masked Vector=&Lerp_Vector(
       A=&VectorFromFloat(Value=&FloatFromCurve(
         FloatCurve=DataInterface() {propertyValues {Curve {
            defaultValue=1
            keys [
             {interpMode=RCIM_Cubic, tangentMode=RCTM_User, time=0, value=0}
             {arriveTangent=16.9912, interpMode=RCIM_Cubic, leaveTangent=16.9912, tangentMode=RCTM_User, time=15, value=100}
            ]
           }}}
         Scale Curve=&Lerp_Float(A=0.5, B=5, Alpha=$Emitter.LoopedAge)
        ))
       B=&VectorFromFloat(Value=&RandomRangeFloat(Minimum=0.3, Maximum=&RandomRangeFloat(Minimum=5, Maximum=20), Evaluation Type=::NewEnumerator1))
       Alpha=&FloatFromCurve(
        FloatCurve=DataInterface() {propertyValues {Curve {keys [
            {interpMode=RCIM_Cubic, leaveTangent=2.5294, tangentMode=RCTM_Break, time=0, value=0}
            {arriveTangent=-0.2166, interpMode=RCIM_Cubic, tangentMode=RCTM_Break, time=0.4, value=1}
           ]}}}
        Scale Curve=0.5
       )
      )
      Passthrough Vector=&Lerp_Vector(
       A=&VectorFromFloat(Value=&FloatFromCurve(
         FloatCurve=DataInterface() {propertyValues {Curve {
            defaultValue=1
            keys [
             {interpMode=RCIM_Cubic, tangentMode=RCTM_User, time=0, value=0.3}
             {arriveTangent=2.811, interpMode=RCIM_Cubic, leaveTangent=2.811, tangentMode=RCTM_User, time=1, value=1}
            ]
           }}}
         Scale Curve=&Lerp_Float(A=1, B=2, Alpha=$Emitter.Age)
        ))
       B=&VectorFromFloat(Value=&FloatFromCurve(
         FloatCurve=DataInterface() {propertyValues {Curve {
            defaultValue=1
            keys [
             {interpMode=RCIM_Cubic, tangentMode=RCTM_User, time=0, value=0.3}
             {arriveTangent=2.811, interpMode=RCIM_Cubic, leaveTangent=2.811, tangentMode=RCTM_User, time=1, value=1}
            ]
           }}}
         Scale Curve=&Lerp_Float(A=1, B=2, Alpha=$Emitter.Age)
        ))
       Alpha=&FloatFromCurve(
        FloatCurve=DataInterface() {propertyValues {Curve {keys [
            {interpMode=RCIM_Cubic, leaveTangent=2.5294, tangentMode=RCTM_Break, time=0, value=0}
            {arriveTangent=-0.2166, interpMode=RCIM_Cubic, tangentMode=RCTM_Break, time=0.4, value=1}
            <at 1.1 0 RCIM_Cubic>
           ]}}}
        Scale Curve=5
       )
      )
     )
    )
    ScaleA=true
    ScaleRGB=true
   ) {module=ScaleColor001}
   &SolveForcesAndVelocity(Clamp Velocity=false, Limit Acceleration=false) {enabled=false}
   &ScaleMeshSize(Scale Factor=&MaskVectorBySpawnGroup(
     Masked Vector=<1 1 1>
     Passthrough Vector=&MakeVector(
      X=1
      Y=2
      Z=&FloatFromCurve(FloatCurve=DataInterface() {propertyValues {Curve {keys [
           {interpMode=RCIM_Cubic, leaveTangent=0.0966, tangentMode=RCTM_Break, time=0, value=1}
           {interpMode=RCIM_Cubic, leaveTangent=-0.1745, time=0.7665, value=1.07}
          ]}}})
     )
    )) {module=ScaleMeshSize001}
   &ScaleMeshSize(Scale Factor=&MaskVectorBySpawnGroup(
     Masked Vector=&Multiply_VectorByFloat(Vector=<1 1 1>, Float=&RandomRangeFloat(Minimum=0.35, Maximum=0.5))
     Passthrough Vector=&VectorFromFloat(Value=&RandomRangeFloat(Minimum=0.1, Maximum=0.15))
    ))
   &UpdateMeshOrientation(
    Rotation Rate=&MaskFloatBySpawnGroup(Masked Float=&RandomRangeFloat(Minimum=-1, Maximum=1), Passthrough Float=0)
    Rotation Vector=<0 0 1>
   )
   &SpriteFacingAndAlignment(
    Alignment Coordinate Space=::Local
    Facing Coordinate Space=::Local
    Sprite Alignment=<0 0 -1>
    Sprite Facing=&RotateVectorByQuaternion(Vector To Rotate=<0 0 0>, Quaternion=$Particles.SkeletalMeshLocation.SampledRotation)
    WriteAlignment=true
    WriteFacing=true
   ) {enabled=false}
   &SpriteFacingAndAlignment(Sprite Alignment=<0 0 1>, Sprite Facing=<0 0 1>, WriteAlignment=true, WriteFacing=true) {module=SpriteFacingAndAlignment001}
   &ScaleColor(
    Scale Alpha=&FloatFromCurve(FloatCurve=DataInterface() {propertyValues {Curve {keys [
         {tangentMode=RCTM_User, time=0, value=1}
         {arriveTangent=-0.0405, interpMode=RCIM_Cubic, leaveTangent=-0.0405, tangentMode=RCTM_User, time=0.7009, value=1}
         {arriveTangent=-8.0089, interpMode=RCIM_Cubic, leaveTangent=-8.0089, tangentMode=RCTM_User, time=1, value=0}
        ]}}})
    Scale RGB=&Lerp_Vector(
     A=&VectorFromFloat(Value=&FloatFromCurve(
       FloatCurve=DataInterface() {propertyValues {Curve {
          defaultValue=1
          keys [
           {interpMode=RCIM_Cubic, tangentMode=RCTM_User, time=0, value=0}
           {arriveTangent=16.9912, interpMode=RCIM_Cubic, leaveTangent=16.9912, tangentMode=RCTM_User, time=15, value=100}
          ]
         }}}
       Scale Curve=&Lerp_Float(A=0.5, B=5, Alpha=$Emitter.LoopedAge)
      ))
     B=&VectorFromFloat(Value=&RandomRangeFloat(Minimum=0.3, Maximum=&RandomRangeFloat(Minimum=5, Maximum=20), Evaluation Type=::NewEnumerator1))
     Alpha=&FloatFromCurve(
      FloatCurve=DataInterface() {propertyValues {Curve {keys [
          {interpMode=RCIM_Cubic, leaveTangent=2.5294, tangentMode=RCTM_Break, time=0, value=0}
          {arriveTangent=-0.2166, interpMode=RCIM_Cubic, tangentMode=RCTM_Break, time=0.4, value=1}
         ]}}}
      Scale Curve=0.5
     )
    )
    ScaleA=true
    ScaleRGB=true
   ) {enabled=false}
   &DynamicMaterialParameters(
    Index 0 Param 1=&MakeFloatFromVector(Vector=$Particles.Scale, Channel=::NewEnumerator0)
    Index 0 Param 2=&RandomRangeFloat(Minimum=0, Maximum=10)
    Index 0 Param 3=&FloatFromCurve(
     FloatCurve=DataInterface() {propertyValues {Curve {keys [{interpMode=RCIM_Cubic, leaveTangent=8.1085, tangentMode=RCTM_Break, time=0.8568, value=1}, {tangentMode=RCTM_User, time=1, value=0}]}}}
     Scale Curve=0.07
    )
    Index 0 Param 4=0.8
    Index 1 Param 1=&FloatFromCurve(FloatCurve=DataInterface() {propertyValues {Curve {keys [
         {interpMode=RCIM_Cubic, leaveTangent=1.1034, tangentMode=RCTM_Break, time=0, value=0}
         {arriveTangent=2.2399, tangentMode=RCTM_User, time=0.5, value=0.5}
        ]}}})
    Index 1 Param 2=&MaskFloatBySpawnGroup(
     Masked Float=&FloatFromCurve(FloatCurve=DataInterface() {propertyValues {Curve {keys [
          {interpMode=RCIM_Cubic, tangentMode=RCTM_User, time=0, value=0}
          {arriveTangent=2.7735, interpMode=RCIM_Cubic, leaveTangent=2.7735, tangentMode=RCTM_User, time=1, value=1}
         ]}}})
     Passthrough Float=&FloatFromCurve(FloatCurve=DataInterface() {propertyValues {Curve {keys [<at 0.5 0>, {arriveTangent=3.2053, tangentMode=RCTM_User, time=0.9, value=1}]}}})
    )
    Index 1 Param 3=&FloatFromCurve(FloatCurve=DataInterface() {propertyValues {Curve {keys [<at 0 0 RCIM_Cubic>, {arriveTangent=3.4223, interpMode=RCIM_Cubic, leaveTangent=3.4223, tangentMode=RCTM_User, time=1, value=1}]}}})
    Index 1 Param 4=&FloatFromCurve(
     FloatCurve=DataInterface() {propertyValues {Curve {keys [
         {interpMode=RCIM_Cubic, leaveTangent=0.1169, tangentMode=RCTM_Break, time=0, value=0}
         {arriveTangent=2.5708, interpMode=RCIM_Cubic, leaveTangent=2.5708, tangentMode=RCTM_User, time=1, value=1}
        ]}}}
     Scale Curve=&RandomRangeFloat(Minimum=1, Maximum=40)
    )
    Index 2 Param 1=&FloatFromCurve(FloatCurve=DataInterface() {propertyValues {Curve {keys [
         {interpMode=RCIM_Cubic, leaveTangent=-0.0025, tangentMode=RCTM_Break, time=0, value=0.5}
         {arriveTangent=-0.0316, interpMode=RCIM_Cubic, tangentMode=RCTM_Break, time=1, value=-0.2}
        ]}}})
    Index 2 Param 2=&FloatFromCurve(
     FloatCurve=DataInterface() {propertyValues {Curve {keys [
         {interpMode=RCIM_Cubic, tangentMode=RCTM_User, time=0, value=0}
         {arriveTangent=2.0522, interpMode=RCIM_Cubic, leaveTangent=2.0522, tangentMode=RCTM_User, time=1, value=1}
        ]}}}
     Scale Curve=2
    )
    Module.Index0Param0Write=true
    Module.Index0Param1Write=true
    Module.Index0Param2Write=true
    Module.Index0Param3Write=true
    Module.Index1Param0Write=true
    Module.Index1Param1Write=true
    Module.Index1Param2Write=true
    Module.Index1Param3Write=true
    Module.Index2Param0Write=true
    Module.Index2Param1Write=true
    Module.Index2Param2Write=false
    Module.Index2Param3Write=false
    Module.Index3Param0Write=false
    Module.Index3Param1Write=false
    Module.Index3Param2Write=false
    Module.Index3Param3Write=true
    Write Parameter Index 1=true
    Write Parameter Index 2=true
    Write Parameter Index 3=true
   )
   &FN_Light_Attributes(
    Color=<rgba 0.2582 0.0437 0.6867 1>
    Create Color Attribute=true
    Create Exponent Attribute=true
    Create Radius Attribute=true
    Exponent=5
    Radius=80
    Scale Color=&Multiply_Float(A=7, B=$Emitter.Flicker)
   ) {module=FN_Light_Attributes001}
  ]
  rendererClasses [
   &NiagaraSpriteRendererProperties
   &NiagaraMeshRendererProperties
   &NiagaraLightRendererProperties
  ]
  renderers [
   {rendererClass=&NiagaraSpriteRendererProperties}
   {rendererClass=&NiagaraMeshRendererProperties, rendererIndex=1}
   {rendererClass=&NiagaraMeshRendererProperties, rendererIndex=2}
   {rendererClass=&NiagaraSpriteRendererProperties, rendererIndex=3}
   {rendererClass=&NiagaraLightRendererProperties, rendererIndex=4}
  ]
 }
 Pickaxe_version {
  emitterUpdateScript [
   &EmitterState(
    Inactive Response=::NewEnumerator1
    Life Cycle Mode=::NewEnumerator1
    Loop Behavior=::NewEnumerator1
    Loop Duration Mode=::NewEnumerator1
    UseLoopDelay=false
   )
   &SpawnBurst_Instantaneous(Spawn Count=2, Spawn Time=4.8, Use Spawn Probability=false, UseLoopCountLimit=false) {module=SpawnBurst_Instantaneous001}
   &SpawnBurst_Instantaneous(Spawn Count=2, Spawn Time=8.0333, Use Spawn Probability=false, UseLoopCountLimit=false) {module=SpawnBurst_Instantaneous002}
   &SpawnRate(Spawn Probability=0.85, SpawnGroup=1, SpawnRate=15, Use Spawn Probability=true)
   &SpawnBurst_Instantaneous(Spawn Count=5, Spawn Group=1, Use Spawn Probability=false, UseLoopCountLimit=false)
   &SpawnBurst_Instantaneous(Spawn Count=2, Spawn Group=2, Use Spawn Probability=false, UseLoopCountLimit=false) {module=SpawnBurst_Instantaneous003}
   &Waveform(
    EnableExponent1=false
    GlobalClampMax=false
    GlobalClampMin=false
    UseInput1=false
    UsePhase1=false
    "Waveform [1]":ENiagara_Waveforms=::NewEnumerator12
    "[1] Amplitude Min/Max":Vector2f=<0.75 1>
    "[1] Frequency"=7
   ) {module=Waveform001}
   /SetVariables(Emitter.LightPosition2:NiagaraPosition=<0 0 -55>) {isSetParameters=true}
   /SetVariables(Emitter.LightPosition:NiagaraPosition=<0 0 55>) {isSetParameters=true}
   /SetVariables(Emitter.Flicker 2:NiagaraFloat=&Add_Float(A=&Multiply_Float(A=$Emitter.Waveform.Waveform, B=0.5), B=0.5)) {isSetParameters=true, module=SetVariables_#9}
   /SetVariables(
    Emitter.Flicker:NiagaraFloat=&Add_Float(
     A=&Multiply_Float(A=$Emitter.Waveform001.Waveform, B=&Lerp_Float(A=1, B=1, Alpha=$System.Timeline.Time))
     B=0
    )
    Emitter.FlickerHeavy:NiagaraFloat=&Add_Float(A=&Multiply_Float(A=$Emitter.Waveform001.Waveform, B=2), B=0.8)
   ) {isSetParameters=true}
   /SetVariables(Emitter.LightColor:LinearColor=&MultiplyLinearColor(
     Linear Color=<rgba 1.142 0.1228 3 0.142>
     Scale Factor=&MakeVector4_FromFloat(Float=&Multiply_Float(
       A=$Emitter.Flicker 2
       B=&FN_FrontEndFloatModifier(
        Gameplay=&Multiply_Float(A=2, B=&FNTimeOfDayScalar(Night Scale=0.73, Morning Scale=0.85, Evening Scale=0.85))
        Front End=0.7
       )
      ))
    )) {isSetParameters=true, module=SetVariables_#10}
  ]
  particleSpawnScript [
   &FN_SetSpawnGroup(Number of Spawn Groups=2, Set Ribbon ID and Ribbon Link Order=true) {enabled=false}
   &PartitionParticles(Write Partition to Ribbon ID=::NewEnumerator0)
   /SetVariables(Particles.VisibilityTag=&SpawnGroup()) {isSetParameters=true, module=SetVariables_#11}
   &DynamicMaterialParameters(
    Index 0 Param 1=0.2
    Index 0 Param 2=1
    Module.Index0Param0Write=true
    Module.Index0Param1Write=true
    Module.Index0Param2Write=true
    Module.Index0Param3Write=true
   )
   &InitializeParticle(
    Color=<rgba 0.4018 0.03 1 1>
    Lifetime=&MaskFloat(
     Mask Source=0
     Mask Parameter=?
     Mask=0
     Masked Float=1
     Passthrough Float=&MaskFloat(Mask Source=0, Mask Parameter=?, Mask=1, Masked Float=1, Passthrough Float=6)
    )
    Mesh Scale=<0.8 0.8 1.2>
    Mesh Scale Mode=::NewEnumerator4
    Position Offset=<0 0 30>
    UsePositionOffset=false
    Write Material Random=true
   )
   &InitialMeshOrientation(
    Mesh Orientation Mode=::NewEnumerator2
    Orientation Vector=&RandomRangeVector(Minimum=<1 -1 0>, Maximum=<-1 1 0>)
    Rotation=<0 0 1>
    Use Rotation Vector=true
   )
   /SetVariables(
    Particles.Electric arcs Lifetime:NiagaraFloat=&RandomRangeFloat(Minimum=0.1, Maximum=1.1)
    Particles.Electric arcs mesh index:NiagaraInt32=&MakeIntFromFloat(Float=&FloatFromCurve(
      FloatCurve=DataInterface() {propertyValues {Curve {keys [
          {arriveTangent=0.6236, interpMode=RCIM_Cubic, tangentMode=RCTM_Break, time=0.25, value=0}
          {arriveTangent=22.9944, interpMode=RCIM_Cubic, tangentMode=RCTM_Break, time=1, value=5}
         ]}}}
      CurveIndex=&RandomRangeFloat(Minimum=0, Maximum=1)
     ))
    Particles.electric arcs position:NiagaraPosition=&AddVectorToPosition(Position=$Particles.Position, Vector=<0 0 0>)
   ) {isSetParameters=true, module=SetVariables_#12}
   /SetVariables(
    Particles.Flare CameraOffset=7
    Particles.Flare Color:LinearColor=<rgba 0.2582 0.0437 0.6867 1>
    Particles.Flare lifetime=0.5
    Particles.Flare size:Vector2f=<30 30>
   ) {isSetParameters=true, module=SetVariables_#13}
   /SetVariables(Particles.Staff glow Position:NiagaraPosition=&AddVectorToPosition(Position=$Particles.Position, Vector=<0 0 0>)) {isSetParameters=true, module=SetVariables_#14}
   &ApplyInitialForces()
  ]
  particleUpdateScript [
   &SolveForcesAndVelocity(Clamp Velocity=false, Limit Acceleration=false)
   &ParticleState()
   &SkeletalMeshLocation(
    Bone / Socket Index=&SelectIntFromArray(Direct Array Index=$Particles.PartitionParticles.Partition, Int Selection Array=DataInterface() {propertyValues {IntData [0,1]}})
    MaskSpawnGroup=false
    Mesh Sampling Type=::NewEnumerator5
    Sampled Position Offset=&SelectVectorFromArray(Direct Array Index=0, Vector Selection Array=DataInterface() {propertyValues {FloatData [<0 0 32>, <0 0 0>]}})
    Skeletal Mesh=$System.Skeletal Mesh
    Socket Sampling Mode=::NewEnumerator5
   )
   &SkeletalMeshLocation(
    Bone / Socket Index=&SelectIntFromArray(Direct Array Index=$Particles.PartitionParticles.Partition, Int Selection Array=DataInterface() {propertyValues {IntData [0,1]}})
    MaskSpawnGroup=false
    Mesh Sampling Type=::NewEnumerator5
    Position Sampling=::NewEnumerator2
    Skeletal Mesh=$System.Skeletal Mesh
    Socket Sampling Mode=::NewEnumerator5
   ) {module=SkeletalMeshLocation001}
   &SkeletalMeshLocation(
    Bone / Socket Index=&SelectIntFromArray(Direct Array Index=$Particles.PartitionParticles.Partition, Int Selection Array=DataInterface() {propertyValues {IntData [0,1]}})
    MaskSpawnGroup=false
    Mesh Sampling Type=::NewEnumerator5
    Position Sampling=::NewEnumerator2
    Sampled Position Offset=<0 0 16.3437>
    Skeletal Mesh=$System.Skeletal Mesh
    Socket Sampling Mode=::NewEnumerator5
   ) {module=SkeletalMeshLocation002}
   &SkeletalMeshLocation(
    Bone / Socket Index=&SelectIntFromArray(Direct Array Index=$Particles.PartitionParticles.Partition, Int Selection Array=DataInterface() {propertyValues {IntData [0,1]}})
    MaskSpawnGroup=false
    Mesh Sampling Type=::NewEnumerator5
    Sampled Position Offset=<0 0 32>
    Skeletal Mesh=$System.Skeletal Mesh
    Socket Sampling Mode=::NewEnumerator5
   ) {module=SkeletalMeshLocation003}
   /SetVariables(Particles.Update Flare age:NiagaraFloat=&Divide_Float(A=$Particles.Age, B=$Particles.Initial.Flare lifetime)) {isSetParameters=true, module=SetVariables_#15}
   /SetVariables(Particles.LightningFlicker:NiagaraFloat=/Niagara/DynamicInputs/Helpers/Waveform(
     Global Input Position=&RandomRangeFloat(Minimum=0, Maximum=1)
     Global Phase=$Particles.Age
     "Waveform [1]"=::NewEnumerator12
     "[1] Phase Scale"=66
     "[1] Frequency"=1
     "[1] Amplitude Min/Max"=<0.4 1>
    )) {isSetParameters=true, module=SetVariables_#16}
   /SetVariables(
    Particles.Update Flare color:LinearColor=&MultiplyLinearColor(
     Linear Color=$Particles.Initial.Flare Color
     Scale Factor=&MakeVector4_FromLinearColor(LinearColor=&MakeLinearColorFromVectorAndFloat(
       "Vector (RGB)"=&Lerp_Vector(
        A=&VectorFromFloat(Value=&FloatFromCurve(
          FloatCurve=DataInterface() {propertyValues {Curve {keys [<at 0 1>,<at 0.1 1>,<at 0.2 2>,<at 0.3 1>,<at 0.6 1>,<at 0.7 2>,<at 0.8 1>,<at 1 1>]}}}
          CurveIndex=$Particles.Update Flare age
         ))
        B=&VectorFromFloat(Value=&RandomRangeFloat(Minimum=0.3, Maximum=3, Evaluation Type=::NewEnumerator1))
        Alpha=&FloatFromCurve(
         FloatCurve=DataInterface() {propertyValues {Curve {keys [
             {interpMode=RCIM_Cubic, leaveTangent=2.5294, tangentMode=RCTM_Break, time=0, value=0}
             {arriveTangent=-0.2166, interpMode=RCIM_Cubic, tangentMode=RCTM_Break, time=0.4, value=1}
            ]}}}
         CurveIndex=$Particles.Update Flare age
        )
       )
       "Float (Alpha)"=&FloatFromCurve(
        FloatCurve=DataInterface() {propertyValues {Curve {keys [
            {tangentMode=RCTM_User, time=0, value=1}
            {arriveTangent=-0.0405, interpMode=RCIM_Cubic, leaveTangent=-0.0405, tangentMode=RCTM_User, time=0.7009, value=1}
            {arriveTangent=-8.0089, interpMode=RCIM_Cubic, leaveTangent=-8.0089, tangentMode=RCTM_User, time=1, value=0}
           ]}}}
        CurveIndex=$Particles.Update Flare age
       )
      ))
    )
    Particles.Update Flare size:Vector2f=&Multiply_Vector2D_ByFloat(
     Float=&FloatFromCurve(
      FloatCurve=DataInterface() {propertyValues {Curve {keys [{tangentMode=RCTM_User, time=0, value=0}, {tangentMode=RCTM_User, time=0.3, value=5}]}}}
      CurveIndex=$Particles.Update Flare age
     )
     Vector2D=$Particles.Initial.Flare size
    )
   ) {isSetParameters=true, module=SetVariables_#17}
   &ScaleColor(
    Scale Alpha=&MaskFloat(
     Mask Source=0
     Mask Parameter=?
     Mask=0
     Masked Float=0
     Passthrough Float=&MaskFloat(
      Mask Source=0
      Mask Parameter=?
      Mask=1
      Masked Float=&FloatFromCurve(
       FloatCurve=DataInterface() {propertyValues {Curve {keys [
           {arriveTangent=9.0725, interpMode=RCIM_Cubic, leaveTangent=9.0725, tangentMode=RCTM_User, time=0, value=0}
           {arriveTangent=-2.5249, interpMode=RCIM_Cubic, leaveTangent=-2.5249, tangentMode=RCTM_User, time=0.1408, value=0.9847}
           {arriveTangent=-0.0405, interpMode=RCIM_Cubic, leaveTangent=-0.0405, tangentMode=RCTM_User, time=1, value=0}
          ]}}}
       Scale Curve=$Particles.LightningFlicker
      )
      Passthrough Float=&FloatFromCurve(
       FloatCurve=DataInterface() {propertyValues {Curve {keys [<at 0 0.1>,<at 1 0>]}}}
       Scale Curve=&Multiply_Float(A=1, B=$Emitter.Flicker)
      )
     )
    )
    Scale RGB=&MaskVector(
     Mask Source=0
     Mask Parameter=?
     Mask=0
     Masked Vector=<0 0 0>
     Passthrough Vector=&MaskVector(
      Mask Source=0
      Mask Parameter=?
      Mask=1
      Masked Vector=&VectorFromFloat(Value=&FloatFromCurve(
        FloatCurve=DataInterface() {propertyValues {Curve {keys [
            {arriveTangent=9.0725, interpMode=RCIM_Cubic, leaveTangent=9.0725, tangentMode=RCTM_Break, time=0, value=0}
            {arriveTangent=-2.5249, interpMode=RCIM_Cubic, leaveTangent=-2.5249, tangentMode=RCTM_Break, time=0.1408, value=1}
            {arriveTangent=-0.0405, interpMode=RCIM_Cubic, leaveTangent=-0.0405, tangentMode=RCTM_User, time=1, value=0}
           ]}}}
        Scale Curve=$Particles.LightningFlicker
       ))
      Passthrough Vector=&Lerp_Vector(
       A=&VectorFromFloat(Value=&FloatFromCurve(
         FloatCurve=DataInterface() {propertyValues {Curve {
            defaultValue=1
            keys [
             {interpMode=RCIM_Cubic, tangentMode=RCTM_User, time=0, value=0}
             {arriveTangent=2.811, interpMode=RCIM_Cubic, leaveTangent=2.811, tangentMode=RCTM_User, time=1, value=1}
            ]
           }}}
         Scale Curve=&Lerp_Float(A=1, B=2, Alpha=$Emitter.Age)
        ))
       B=&VectorFromFloat(Value=&FloatFromCurve(
         FloatCurve=DataInterface() {propertyValues {Curve {
            defaultValue=1
            keys [
             {interpMode=RCIM_Cubic, tangentMode=RCTM_User, time=0, value=0.3}
             {arriveTangent=2.811, interpMode=RCIM_Cubic, leaveTangent=2.811, tangentMode=RCTM_User, time=1, value=1}
            ]
           }}}
         Scale Curve=&Lerp_Float(A=1, B=2, Alpha=$Emitter.Age)
        ))
       Alpha=&FloatFromCurve(
        FloatCurve=DataInterface() {propertyValues {Curve {keys [
            {interpMode=RCIM_Cubic, leaveTangent=2.5294, tangentMode=RCTM_Break, time=0, value=0}
            {arriveTangent=-0.2166, interpMode=RCIM_Cubic, tangentMode=RCTM_Break, time=0.4, value=1}
            <at 1.1 0 RCIM_Cubic>
           ]}}}
        Scale Curve=5
       )
      )
     )
    )
    ScaleA=true
    ScaleRGB=true
   ) {module=ScaleColor001}
   /SetVariables(
    Particles.ColorCore:LinearColor=&MultiplyLinearColor(
     Linear Color=$Particles.Color
     Scale Factor=&MakeFloat4FromVectorAndFloat("Vector (RGB)"=&VectorFromFloat(Value=33), "Float (Alpha)"=1)
    )
    Particles.ColorGlow:LinearColor=&MultiplyLinearColor(
     Linear Color=$Particles.Color
     Scale Factor=&MakeFloat4FromVectorAndFloat("Vector (RGB)"=&VectorFromFloat(Value=0.15), "Float (Alpha)"=1)
    )
    Particles.DynamicCore:Vector4f=<1 6 2 0>
    Particles.DynamicGlow:Vector4f=<5.0981 3 4 0>
    Particles.LightningScale:NiagaraFloat=&FloatFromCurve(FloatCurve=DataInterface() {propertyValues {Curve {keys [
         {arriveTangent=2.7028, interpMode=RCIM_Constant, time=0, value=0.75}
         {interpMode=RCIM_Cubic, tangentMode=RCTM_User, time=0.5, value=1}
         {arriveTangent=-2.6499, tangentMode=RCTM_User, time=1, value=0.5}
        ]}}})
   ) {isSetParameters=true, module=SetVariables_#18}
   /SetVariables(
    Particles.Core GlowPosition:NiagaraPosition=$Particles.SkeletalMeshLocation001.SampledPosition
    Particles.CoreGlowSize:Vector2f=<12 37>
    Particles.DarkGlowPosition:NiagaraPosition=$Particles.SkeletalMeshLocation002.SampledPosition
    Particles.DarkGlowSize:Vector2f=&Vector2DFromFloat(Value=75)
    Particles.Glow Alignment:Vector3f=&ReturnMeshOrientationAxis(Mesh Orientation Axis=::ZAxis, Mesh Orientation Quat=$Particles.SkeletalMeshLocation.SampledRotation)
    Particles.SpriteAlignment=<0 0 1>
   ) {isSetParameters=true, module=SetVariables_#19}
   /SetVariables(
    Particles.CoreGlowColor:LinearColor=&MultiplyLinearColor(
     Linear Color=<rgba 1.142 0.1228 3 0.25>
     Scale Factor=&MakeVector4_FromFloat(Float=&Multiply_Float(
       A=$Emitter.Flicker 2
       B=&FN_FrontEndFloatModifier(Gameplay=&FNTimeOfDayScalar(Night Scale=1, Morning Scale=1, Evening Scale=1), Front End=0.7)
      ))
    )
    Particles.DarkGlowColor:LinearColor=&MultiplyLinearColor(
     Linear Color=<rgba 0.2854 0.0292 0.564 0.195>
     Scale Factor=&MakeVector4_FromFloat(Float=&Multiply_Float(
       A=$Emitter.Flicker 2
       B=&FN_FrontEndFloatModifier(Gameplay=&FNTimeOfDayScalar(Night Scale=1, Morning Scale=1, Evening Scale=1), Front End=0.7)
      ))
    )
   ) {isSetParameters=true, module=SetVariables_#20}
   &ScaleMeshSize(Scale Factor=&MakeVector(X=$Particles.LightningScale, Y=$Particles.LightningScale, Z=1))
   &FN_Light_Attributes(
    Color=<rgba 0.2582 0.0437 0.6867 1>
    Create Color Attribute=true
    Create Exponent Attribute=true
    Create Radius Attribute=true
    Exponent=5
    Radius=30
    Scale Color=&Multiply_Float(A=7, B=$Emitter.Flicker)
   ) {module=FN_Light_Attributes001}
   &UpdateMeshOrientation(
    Rotation Rate=&RandomRangeFloat(Minimum=-0.5, Maximum=0.5)
    Rotation Vector=&RotateVectorByQuaternion(Vector To Rotate=<0 0 1>, Quaternion=$StackContext.SkeletalMeshLocation.SampledRotation)
   )
   &AlignSpriteToMeshOrientation(
    Mesh Orientation Relative Sprite Alignment Vector=<0 0 1>
    Mesh Orientation Relative Sprite Facing Vector:Vector3f=<0 1 0>
    Orientation Quaternion:Quat4f=$Particles.SkeletalMeshLocation.SampledRotation
    SetAlignment=true
    SetFacing=true
   )
  ]
  rendererClasses [
   &NiagaraSpriteRendererProperties
   &NiagaraMeshRendererProperties
   &NiagaraLightRendererProperties
  ]
  renderers [
   {rendererClass=&NiagaraSpriteRendererProperties}
   {rendererClass=&NiagaraMeshRendererProperties, rendererIndex=1}
   {rendererClass=&NiagaraMeshRendererProperties, rendererIndex=2}
   {rendererClass=&NiagaraLightRendererProperties, rendererIndex=3}
   {rendererClass=&NiagaraLightRendererProperties, rendererIndex=4}
   {rendererClass=&NiagaraSpriteRendererProperties, rendererIndex=5}
   {rendererClass=&NiagaraSpriteRendererProperties, rendererIndex=6}
  ]
 }
 StaffGlow005 {
  bEnabled=false
  emitterUpdateScript [
   &EmitterState(Life Cycle Mode=::NewEnumerator1, Loop Behavior=::NewEnumerator1, Loop Duration=6.4167, UseLoopDelay=false)
   &SpawnBurst_Instantaneous(Use Spawn Probability=false, UseLoopCountLimit=false)
   /Niagara/Noise/Waveform(
    EnableExponent1=false
    GlobalClampMax=false
    GlobalClampMin=false
    UseInput1=false
    UsePhase1=false
    "Waveform [1]":ENiagara_Waveforms=::NewEnumerator12
    "[1] Frequency"=7
   )
   /SetVariables(Emitter.Flicker:NiagaraFloat=&Add_Float(A=&Multiply_Float(A=$Emitter.Waveform.Waveform, B=0.5), B=0.5)) {isSetParameters=true}
  ]
  particleSpawnScript [
   &InitializeParticle(
    Color=<rgba 0.2582 0.0437 0.6867 1>
    Position Offset=<0 0 20>
    Sprite Size Mode=::NewEnumerator3
    Uniform Sprite Size=37
    UsePositionOffset=true
    Write Material Random=true
   )
   &DynamicMaterialParameters(
    Index 0 Param 1=0.2
    Index 0 Param 2=1
    Module.Index0Float0Write=true
    Module.Index0Float1Write=true
    Module.Index0Float2Write=true
    Module.Index0Float3Write=true
   )
  ]
  particleUpdateScript [
   &ParticleState(Kill Particles When Lifetime Has Elapsed=false, Let Infinitely Lived Particles Die When Emitter Deactivates=true)
   &ScaleSpriteSize(Uniform Curve Sprite Scale=DataInterface() {propertyValues {Curve {
       defaultValue=0
       keys [{interpMode=RCIM_Cubic, tangentMode=RCTM_User, time=0, value=0.7}, {interpMode=RCIM_Cubic, tangentMode=RCTM_User, time=1, value=1}]
      }}})
   &ScaleColor(
    Scale Alpha=&Multiply_Float(A=$Emitter.Flicker, B=1)
    Scale RGB=&Lerp_Vector(
     A=&VectorFromFloat(Value=&FloatFromCurve(
       FloatCurve=DataInterface() {propertyValues {Curve {
          defaultValue=1
          keys [
           {interpMode=RCIM_Cubic, tangentMode=RCTM_User, time=0, value=0.3}
           {arriveTangent=2.811, interpMode=RCIM_Cubic, leaveTangent=2.811, tangentMode=RCTM_User, time=1, value=1}
          ]
         }}}
       Scale Curve=&Lerp_Float(A=1, B=2, Alpha=$Emitter.Age)
      ))
     B=&VectorFromFloat(Value=&FloatFromCurve(
       FloatCurve=DataInterface() {propertyValues {Curve {
          defaultValue=1
          keys [
           {interpMode=RCIM_Cubic, tangentMode=RCTM_User, time=0, value=0.3}
           {arriveTangent=2.811, interpMode=RCIM_Cubic, leaveTangent=2.811, tangentMode=RCTM_User, time=1, value=1}
          ]
         }}}
       Scale Curve=&Lerp_Float(A=1, B=2, Alpha=$Emitter.Age)
      ))
     Alpha=&FloatFromCurve(
      FloatCurve=DataInterface() {propertyValues {Curve {keys [
          {interpMode=RCIM_Cubic, leaveTangent=2.5294, tangentMode=RCTM_Break, time=0, value=0}
          {arriveTangent=-0.2166, interpMode=RCIM_Cubic, tangentMode=RCTM_Break, time=0.4, value=1}
          <at 1.1 0 RCIM_Cubic>
         ]}}}
      Scale Curve=5
     )
    )
    ScaleA=true
    ScaleRGB=true
   ) {module=ScaleColor001}
   &ScaleColor(
    Scale Alpha=&Multiply_Float(A=$Emitter.Flicker, B=1)
    Scale RGB=&Lerp_Vector(
     A=&VectorFromFloat(Value=&FloatFromCurve(
       FloatCurve=DataInterface() {propertyValues {Curve {
          defaultValue=1
          keys [
           {interpMode=RCIM_Cubic, tangentMode=RCTM_User, time=0, value=0}
           {arriveTangent=2.811, interpMode=RCIM_Cubic, leaveTangent=2.811, tangentMode=RCTM_User, time=1, value=1}
          ]
         }}}
       Scale Curve=&Lerp_Float(A=0.5, B=1, Alpha=$Emitter.Age)
      ))
     B=&VectorFromFloat(Value=&RandomRangeFloat(Minimum=0.5, Maximum=&RandomRangeFloat(Minimum=5, Maximum=10), Evaluation Type=::NewEnumerator1))
     Alpha=&FloatFromCurve(
      FloatCurve=DataInterface() {propertyValues {Curve {keys [
          {interpMode=RCIM_Cubic, leaveTangent=2.5294, tangentMode=RCTM_Break, time=0, value=0}
          {arriveTangent=-0.2166, interpMode=RCIM_Cubic, tangentMode=RCTM_Break, time=0.4, value=1}
          <at 1.1 0 RCIM_Cubic>
         ]}}}
      Scale Curve=5
     )
    )
    ScaleA=true
    ScaleRGB=true
   ) {enabled=false}
   &SpriteFacingAndAlignment(
    Alignment Coordinate Space=::Local
    Sprite Alignment=<0 0 -1>
    Sprite Facing=&RotateVectorByQuaternion(Vector To Rotate=<0 0 0>, Quaternion=$Particles.SkeletalMeshLocation.SampledRotation)
    WriteAlignment=true
    WriteFacing=false
   )
   &FN_Light_Attributes(
    Color=$Particles.Color
    Create Color Attribute=true
    Create Exponent Attribute=true
    Create Position Attribute=true
    Create Radius Attribute=true
    Enable Particle Light Enable=true
    Exponent=5
    Light Enabled=true
    Radius=85
    Scale Color=&FloatFromCurve(
     FloatCurve=DataInterface() {propertyValues {Curve {keys [
         {arriveTangent=9.0725, interpMode=RCIM_Cubic, leaveTangent=9.0725, tangentMode=RCTM_User, time=0, value=0}
         {arriveTangent=-0.5493, interpMode=RCIM_Cubic, leaveTangent=-0.5493, tangentMode=RCTM_User, time=0.3, value=0.9847}
         {arriveTangent=-0.0405, interpMode=RCIM_Cubic, leaveTangent=-0.0405, tangentMode=RCTM_User, time=1, value=0}
        ]}}}
     Scale Curve=15
    )
   ) {enabled=false}
   &FN_Light_Attributes(
    Color=<rgba 0.2582 0.0437 0.6867 1>
    Create Color Attribute=true
    Create Exponent Attribute=true
    Create Radius Attribute=true
    Exponent=5
    Radius=80
    Scale Color=&Multiply_Float(A=7, B=$Emitter.Flicker)
   ) {module=FN_Light_Attributes001}
  ]
  rendererClasses [&NiagaraSpriteRendererProperties, &NiagaraLightRendererProperties]
  renderers [
   {rendererClass=&NiagaraSpriteRendererProperties}
   {rendererClass=&NiagaraLightRendererProperties, rendererIndex=1}
  ]
 }
}
summary {
 systemName=NS_Emote_MuffinLadle_Gas_PropElectricity
 emitters [
  {bEnabled=false, emitterName=BaseFlash002, rendererClasses [&NiagaraSpriteRendererProperties]}
  {bEnabled=false, emitterName=Lightning002, rendererClasses [&NiagaraMeshRendererProperties]}
  {
   bEnabled=false
   emitterName=StaffGlow005
   rendererClasses [&NiagaraSpriteRendererProperties, &NiagaraLightRendererProperties]
  }
  {
   bEnabled=false
   emitterName=Lightning_up
   rendererClasses [
    &NiagaraSpriteRendererProperties
    &NiagaraMeshRendererProperties
    &NiagaraLightRendererProperties
   ]
  }
  {
   bEnabled=false
   emitterName=Lightning001_Down
   rendererClasses [
    &NiagaraSpriteRendererProperties
    &NiagaraMeshRendererProperties
    &NiagaraLightRendererProperties
   ]
  }
  {
   bEnabled=false
   emitterName=Lightning_down
   rendererClasses [
    &NiagaraSpriteRendererProperties
    &NiagaraMeshRendererProperties
    &NiagaraLightRendererProperties
   ]
  }
  {
   bEnabled=false
   emitterName=Lightning_Up001
   rendererClasses [
    &NiagaraSpriteRendererProperties
    &NiagaraMeshRendererProperties
    &NiagaraLightRendererProperties
   ]
  }
  {
   emitterName=Pickaxe_version
   rendererClasses [
    &NiagaraSpriteRendererProperties
    &NiagaraMeshRendererProperties
    &NiagaraLightRendererProperties
   ]
  }
 ]
 userVariables [
  User.bIsFrontEnd {type=NiagaraBool}
  User.bIsFrontEndPreview {type=NiagaraBool}
  User.LightOpacity {type=NiagaraFloat, defaultValue=1, description=""}
  User.NormalizedNotifyProgress {type=NiagaraFloat, defaultValue=0, description=""}
  User.SkeletalMesh {
   type=NiagaraDataInterfaceSkeletalMesh
   defaultValue {
    struct=NiagaraExt_VariableValue_DataInterface
    value {dataInterface=&self:NiagaraDataInterfaceSkeletalMesh_0}
   }
  }
 ]
}
