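# Niagara system for the prop-spawn effect of the MuffinLadle_Gas player emote.
# It contains smoke, airglow and sparkles emitters, all disabled and listing no renderers, with SystemState as the only listed module.
# It exposes two float user parameters, Dogscale and NormalizedNotifyProgress, though nothing shown consumes them.
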NiagaraSystem /BRCosmetics/Animation/Game/MainPlayer/Emotes/MuffinLadle_Gas/FX/NS_Emote_MuffinLadle_PropSpawn
dependencies {usedDataInterfaces [], usedDynamicInputs [], usedModules [&SystemState]}
emitters {
 AirglowLWE001 {bEnabled=false, rendererClasses [], renderers []}
 SmokeLWE002 {bEnabled=false, rendererClasses [], renderers []}
 SparklesLWE002 {bEnabled=false, rendererClasses [], renderers []}
}
summary {
 systemName=NS_Emote_MuffinLadle_PropSpawn
 emitters [
  {bEnabled=false, emitterName=SmokeLWE002, rendererClasses []}
  {bEnabled=false, emitterName=AirglowLWE001, rendererClasses []}
  {bEnabled=false, emitterName=SparklesLWE002, rendererClasses []}
 ]
 userVariables [
  User.Dogscale {type=NiagaraFloat, defaultValue=1, description=""}
  User.NormalizedNotifyProgress {type=NiagaraFloat, defaultValue=0, description=""}
 ]
}
