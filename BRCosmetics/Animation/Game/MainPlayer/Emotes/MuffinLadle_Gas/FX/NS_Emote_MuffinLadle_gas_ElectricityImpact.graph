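# Niagara system for the MuffinLadle_Gas emote's purple electricity impact: flashes, sparks, electric arcs, shockwave, spikes, staff pulse and debris.
# Its "Compressed" emitter packs these layers into one, using burst spawn groups copied to VisibilityTag and per-group masks for motion.
# Each layer has its own lifetime, normalized age and curve-driven color, size and material parameters.
# Time-of-day scalars adjust some opacities; smoke logic remains, though its apparent burst is disabled.
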NiagaraSystem /BRCosmetics/Animation/Game/MainPlayer/Emotes/MuffinLadle_Gas/FX/NS_Emote_MuffinLadle_gas_ElectricityImpact
dependencies {
 usedDataInterfaces [
  &NiagaraDataInterfaceCurve
  &NiagaraDataInterfaceArrayColor
  &NiagaraDataInterfaceSpriteRendererInfo
  &NiagaraDataInterfaceColorCurve
  &NiagaraDataInterfaceVector2DCurve
  &NiagaraDataInterfaceVectorCurve
 ]
 usedDynamicInputs []
 usedModules [
  &self:SystemSpawnScript.SystemScriptSource.SystemScriptGraph.NiagaraNodeAssignment_0.SetVariables_#1
  &SystemState
  &self:SystemSpawnScript.SystemScriptSource.SystemScriptGraph.NiagaraNodeAssignment_1.SetVariables_#2
 ]
}
emitters {
 Compressed {
  emitterSpawnScript [/SetVariables(
    Emitter.Shockwave DarkColor:LinearColor=<rgba 0.2582 0.0437 0.6867 1>
    "Emitter.Shockwave MainColor ":LinearColor=&SelectLinearColorFromArray(
     Array Sampling Mode=::NewEnumerator0
     Direct Array Index=&RandomRangeInt(Minimum=0, Maximum=1)
     Color Selection Array=DataInterface() {propertyValues {ColorData [<rgba 0.4553 0.3412 0.6867 1>, <rgba 0.3115 0.1237 0.6867 1>]}}
     Random Seed=0
    )
   ) {isSetParameters=true}]
  emitterUpdateScript [
   &EmitterState(Life Cycle Mode=::NewEnumerator1, Loop Behavior=::NewEnumerator1, Loop Duration=1, UseLoopDelay=true)
   &SpawnBurst_Instantaneous(Use Spawn Probability=false, UseLoopCountLimit=false)
   &SpawnBurst_Instantaneous(Spawn Group=1, Use Spawn Probability=false, UseLoopCountLimit=false) {module=SpawnBurst_Instantaneous001}
   &SpawnBurst_Instantaneous(
    Spawn Count=&RandomRangeInt(Minimum=15, Maximum=25)
    Spawn Group=2
    Use Spawn Probability=false
    UseLoopCountLimit=false
   ) {module=SpawnBurst_Instantaneous002}
   &SpawnBurst_Instantaneous(
    Spawn Count=&RandomRangeInt(Minimum=25, Maximum=30)
    Spawn Group=2
    Spawn Time=0.1
    Use Spawn Probability=false
    UseLoopCountLimit=false
   ) {module=SpawnBurst_Instantaneous003}
   &SpawnBurst_Instantaneous(
    Spawn Count=&RandomRangeInt(Minimum=5, Maximum=15)
    Spawn Group=3
    Spawn Time=0.05
    Use Spawn Probability=false
    UseLoopCountLimit=true
   ) {module=SpawnBurst_Instantaneous004}
   &SpawnBurst_Instantaneous(Spawn Group=4, Spawn Time=0.05, Use Spawn Probability=false, UseLoopCountLimit=false) {module=SpawnBurst_Instantaneous005}
   &SpawnBurst_Instantaneous(
    Spawn Group=5
    Spawn Probability=&RandomRangeFloat(Minimum=0, Maximum=1)
    Use Spawn Probability=true
    UseLoopCountLimit=false
   ) {enabled=false, module=SpawnBurst_Instantaneous007}
   &SpawnBurst_Instantaneous(
    Spawn Count=3
    Spawn Group=5
    Spawn Probability=&MakeCustomFloatFromBool(Bool=$System.SpawnMud, True Float=0, False Float=1)
    Use Spawn Probability=false
    UseLoopCountLimit=false
   ) {module=SpawnBurst_Instantaneous006}
   &SpawnBurst_Instantaneous(Spawn Group=6, Use Spawn Probability=false, UseLoopCountLimit=false) {module=SpawnBurst_Instantaneous008}
   &SpawnBurst_Instantaneous(
    Spawn Count=&RandomRangeInt(Minimum=15, Maximum=5)
    Spawn Group=7
    Spawn Time=0.05
    Use Spawn Probability=false
    UseLoopCountLimit=false
   ) {module=SpawnBurst_Instantaneous009}
   &SpawnBurst_Instantaneous(
    Spawn Count=&RandomRangeInt(Minimum=25, Maximum=15)
    Spawn Group=8
    Spawn Time=0.03
    Use Spawn Probability=false
   ) {enabled=false, module=SpawnBurst_Instantaneous010}
  ]
  particleSpawnScript [
   /SetVariables(Particles.VisibilityTag=&SpawnGroup()) {isSetParameters=true}
   &InitializeParticle(Color=<rgba 1 1 1 0>, UsePositionOffset=false, Write Material Random=true)
   &InitialMeshOrientation(
    Mesh Axis To Orient=/Niagara/DynamicInputs/Helpers/MaskPosition(
     Mask Source:NiagaraInt32=0
     Mask Parameter:NiagaraInt32=?
     Mask:NiagaraInt32=3
     Masked Position:NiagaraPosition=<1 0 0>
     Passthrough Position:NiagaraPosition=&MaskVector(Mask Source=0, Mask Parameter=?, Mask=4, Masked Vector=<1 0 0>, Passthrough Vector=<0.5 0 0>)
    )
    Mesh Orientation Mode=::NewEnumerator2
    Orientation Coordinate Space=::Local
    Orientation Vector=&MaskPosition(
     Mask Source:NiagaraInt32=0
     Mask Parameter:NiagaraInt32=?
     Mask:NiagaraInt32=3
     Masked Position:NiagaraPosition=<1 0 0>
     Passthrough Position:NiagaraPosition=&MaskVector(Mask Source=0, Mask Parameter=?, Mask=4, Masked Vector=<0 0 1>, Passthrough Vector=<0.5 0 0>)
    )
    Rotation=&MaskVector(
     Mask Source=0
     Mask Parameter=?
     Mask=3
     Masked Vector=&RandomRangeVector(Minimum=<-1 -1 -0.5>, Maximum=<1 1 0.5>)
     Passthrough Vector=&MaskVector(
      Mask Source=0
      Mask Parameter=?
      Mask=4
      Masked Vector=&RandomRangeVector(Minimum=<-5 0 0>, Maximum=<5 0 0>)
      Passthrough Vector=<0 0 0>
     )
    )
    Use Rotation Vector=true
   )
   &SubUVAnimation(
    End Frame Range Override=3
    Output Alternate Renderer Binding=true
    "Sprite Renderer "=DataInterface() {propertyValues {SpriteRenderer=&self:Compressed_0.NiagaraSpriteRendererProperties_4}}
    SubUV Animation Mode:ENiagara_SubUVLookupModeV2=::NewEnumerator2
    UseEndFrame=true
    UseStartFrame=true
   ) {enabled=false}
   &ShapeLocation(MaskSpawnGroup=true, Override Seed=false, Spawn Group Mask=7, Sphere Radius=10, UseOwnerScale=true)
   &AddVelocity(
    Cone Angle=95
    Cone Axis=<0 0 1>
    Distribution Along Cone Axis=0.35
    Module.UseConeFalloff=false
    Module.UseVelDistribution=true
    Override Seed=false
    Velocity Mode=::NewEnumerator2
    Velocity Speed=&MaskFloat(
     Mask Source=0
     Mask Parameter=?
     Mask=2
     Masked Float=&RandomRangeFloat(Minimum=250, Maximum=450)
     Passthrough Float=0
    )
   )
   &SubUVAnimation(End Frame=3) {module=SubUVAnimation001}
   &CameraOffset(Camera Offset Amount=7) {enabled=false}
   /SetVariables(
    Particles.Ground Flash Position:NiagaraPosition=&AddVectorToPosition(Position=<0 0 7.5>, Vector=<0 0 0>)
    Particles.Ground Flash lifetime=1
    Particles.Ground Flash size:Vector2f=<77.122 41.5447>
    Particles.Ground Initial Color:LinearColor=<rgba 0.2582 0.0437 0.6867 1>
    Particles.Ground flash CameraOffset=7
   ) {isSetParameters=true}
   /SetVariables(
    Particles.Top Flare lifetime=1
    Particles.Top Flare position:NiagaraPosition=&AddVectorToPosition(Position=<0 0 20>, Vector=<0 0 0>)
    Particles.Top flare size:Vector2f=<80 80>
   ) {isSetParameters=true}
   /SetVariables(
    Particles.Spark size:Vector2f=&RandomRangeVector2D(Minimum=<3.5 3.5>, Maximum=<6.5 6.5>)
    Particles.Sparks Lifetime:NiagaraFloat=&RandomRangeFloat(Minimum=0.1, Maximum=0.7)
    Particles.Sparks Position:NiagaraPosition=&AddVectorToPosition(Position=<0 0 0.2>, Vector=<0 0 0>)
    Particles.Sparks color:LinearColor=&RandomRangeLinearColor(Minimum=<rgba 0.2582 0.0437 0.6867 1>, Maximum=<rgba 0.376 0.0637 1 1>)
   ) {isSetParameters=true}
   /SetVariables(
    Particles.Electricity Color:LinearColor=&Multiply_LinearColorByFloat(Linear Color=<rgba 0.2582 0.0437 0.6867 1>, Float=0.4)
    Particles.Electricity Lifetime:NiagaraFloat=&RandomRangeFloat(Minimum=0.25, Maximum=0.4)
    Particles.Electricity Position:NiagaraPosition=&AddVectorToPosition(Position=<77.8524 0 10>, Vector=<0 0 0>)
    Particles.electricity scale:Vector3f=&RandomRangeVector(Minimum=<1 1 1>, Maximum=<1 1 1>)
   ) {isSetParameters=true}
   /SetVariables(
    Particles.Ground Shockwave lifetime=0.55
    Particles.Ground shockwave CameraOffset=5
    Particles.Ground shockwave mesh scale:Vector3f=<0.6 0.6 0.6>
    Particles.Ground shockwave position:NiagaraPosition=&AddVectorToPosition(Position=<0 0 10>, Vector=<0 0 0>)
    Particles.Ground shockwave rotation=90
   ) {isSetParameters=true}
   /SetVariables(
    Particles.Impact Spikes Lifetime:NiagaraFloat=&RandomRangeFloat(Minimum=0.15, Maximum=0.25)
    Particles.Impact Spikes mesh scale:Vector3f=&RandomRangeVector(Minimum=<1.5 1 1>, Maximum=<1.85 1 1>)
    Particles.Impact spikes CameraOffset=-30
    Particles.Impact spikes color:LinearColor=&RandomRangeLinearColor(
     Minimum=&ColorFromCurve(ColorCurve=DataInterface() {propertyValues {
        AlphaCurve {keys [<at 0 1 RCIM_Cubic>,<at 1 0 RCIM_Cubic>]}
        BlueCurve {keys [<at 0 0.9909 RCIM_Cubic>,<at 1 0.5615 RCIM_Cubic>]}
        GreenCurve {keys [<at 0 0.9094 RCIM_Cubic>,<at 1 0.5615 RCIM_Cubic>]}
        RedCurve {keys [<at 0 0.9595 RCIM_Cubic>,<at 1 0.5615 RCIM_Cubic>]}
       }})
     Maximum=<rgba 0.6698 0.661 0.6558 0.25>
    )
   ) {isSetParameters=true}
   /SetVariables(
    Particles.Staff Pulse Color:LinearColor=<rgba 1.7071 0.6284 5 1>
    Particles.Staff Pulse Position:NiagaraPosition=&AddVectorToPosition(Position=<0 0 15>, Vector=<0 0 0>)
    Particles.Staff Pulse Size:Vector2f=<75 75>
    Particles.Staff Pulse lifetime=0.6
    Particles.Staff Pulse rotation=95
   ) {isSetParameters=true}
   /SetVariables(
    Particles.Debris InitSpriteRotate:NiagaraFloat=&UniformRangedFloat(Minimum=-3, Maximum=3)
    Particles.Debris Lifetime:NiagaraFloat=&RandomRangeFloat(Minimum=0.5, Maximum=1)
    "Particles.Debris RandomInt ":NiagaraInt32=&RandomRangeInt(Minimum=0, Maximum=1)
    Particles.Debris color:LinearColor=&RandomRangeLinearColor(Minimum=<rgba 0.2279 0.2292 0.2078 5>, Maximum=<rgba 0.526 0.5217 0.4789 5>)
    Particles.Debris size:Vector2f=&Vector2DFromFloat(Value=&RandomRangeFloat(Minimum=5, Maximum=5))
   ) {isSetParameters=true}
   /SetVariables(
    Particles.Smoke Sprite Size:Vector2f=&Vector2DFromFloat(Value=&RandomRangeFloat(Minimum=27, Maximum=42))
    Particles.Smoke Sprite rotate:NiagaraFloat=&RandomRangeFloat(Minimum=0, Maximum=360)
    Particles.Smoke VelCurveRand:NiagaraFloat=&FloatFromCurve(
     FloatCurve=DataInterface() {propertyValues {Curve {keys [
         {interpMode=RCIM_Cubic, leaveTangent=0.1132, tangentMode=RCTM_Break, time=0, value=0}
         {arriveTangent=1.9858, interpMode=RCIM_Cubic, leaveTangent=1.7203, tangentMode=RCTM_Break, time=0.3, value=0.4}
         {interpMode=RCIM_Cubic, tangentMode=RCTM_Break, time=1, value=1}
        ]}}}
     CurveIndex=&RandomRangeFloat(Minimum=0, Maximum=1)
    )
    Particles.Smoke color:LinearColor=&RandomRangeLinearColor(Minimum=<rgba 0.6073 0.6073 0.6073 1>, Maximum=<rgba 0.6128 0.5758 0.6867 1>)
    Particles.Smoke lifetime:NiagaraFloat=&RandomRangeFloat(Minimum=0.75, Maximum=1.55)
    Particles.smoke InitSpriteRotate:NiagaraFloat=&RandomRangeFloat(Minimum=-1, Maximum=1)
   ) {isSetParameters=true}
   &ShapeLocation(
    Handle Radius=0.1
    Large Radius=5
    MaskSpawnGroup=true
    Offset=<0 0 20>
    Offset Coordinate Space=::World
    Offset Mode=::NewEnumerator0
    Override Seed=false
    Rotation Coordinate Space=::World
    Shape Primitive=::NewEnumerator3
    Spawn Group Mask=8
    UseOwnerScale=true
   ) {module=ShapeLocation001}
   &AddVelocity(
    Module.Use Falloff Distance=false
    Origin Offset=&MaskVector(Mask Source=0, Mask Parameter=?, Mask=7, Masked Vector=<0 0 -15>, Passthrough Vector=<0 0 20>)
    Override Seed=false
    Velocity Mode=::NewEnumerator1
    Velocity Speed=&MaskFloat(
     Mask Source=0
     Mask Parameter=?
     Mask=0
     Masked Float=0
     Passthrough Float=&MaskFloat(
      Mask Source=0
      Mask Parameter=?
      Mask=1
      Masked Float=0
      Passthrough Float=&MaskFloat(
       Mask Source=0
       Mask Parameter=?
       Mask=2
       Masked Float=0
       Passthrough Float=&MaskFloat(
        Mask Source=0
        Mask Parameter=?
        Mask=3
        Masked Float=0
        Passthrough Float=&MaskFloat(
         Mask Source=0
         Mask Parameter=?
         Mask=4
         Masked Float=0
         Passthrough Float=&MaskFloat(
          Mask Source=0
          Mask Parameter=?
          Mask=5
          Masked Float=0
          Passthrough Float=&MaskFloat(
           Mask Source=0
           Mask Parameter=?
           Mask=6
           Masked Float=0
           Passthrough Float=&MaskFloat(
            Mask Source=0
            Mask Parameter=?
            Mask=7
            Masked Float=&RandomRangeFloat(Minimum=500, Maximum=1100)
            Passthrough Float=&Lerp_Float(A=350, B=490, Alpha=$Particles.Smoke VelCurveRand)
           )
          )
         )
        )
       )
      )
     )
    )
   ) {module=AddVelocity002}
  ]
  particleUpdateScript [
   &ParticleState()
   &ScaleMeshSize(Scale Factor=&MaskVector(
     Mask Source=0
     Mask Parameter=?
     Mask=3
     Masked Vector=&VectorFromFloat(Value=&RandomRangeFloat(Minimum=0.1, Maximum=0.15))
     Passthrough Vector=<0 0 0>
    ))
   /SetVariables(
    Particles.Update Electricity age:NiagaraFloat=&Divide_Float(A=$Particles.Age, B=$Particles.Initial.Electricity Lifetime)
    Particles.Update Ground Flash Age:NiagaraFloat=&Divide_Float(A=$Particles.Age, B=$Particles.Initial.Ground Flash lifetime)
    Particles.Update Ground shockwave Age:NiagaraFloat=&Divide_Float(A=$Particles.Age, B=$Particles.Initial.Ground Shockwave lifetime)
    Particles.Update Sparks Age:NiagaraFloat=&Divide_Float(A=$Particles.Age, B=$Particles.Initial.Sparks Lifetime)
    Particles.Update Spikes Impact Age:NiagaraFloat=&Divide_Float(A=$Particles.Age, B=$Particles.Initial.Impact Spikes Lifetime)
    Particles.Update Staff Pulse Age:NiagaraFloat=&Divide_Float(A=$Particles.Age, B=$Particles.Initial.Staff Pulse lifetime)
    Particles.Update Top Flash Age:NiagaraFloat=&Divide_Float(A=$Particles.Age, B=$Particles.Initial.Top Flare lifetime)
    Particles.Update debris Age:NiagaraFloat=&Divide_Float(A=$Particles.Age, B=$Particles.Initial.Debris Lifetime)
    Particles.Update smoke Impact Age:NiagaraFloat=&Divide_Float(A=$Particles.Age, B=$Particles.Initial.Smoke lifetime)
   ) {isSetParameters=true}
   /SetVariables(
    Particles.Update Ground Flash color:LinearColor=&MultiplyLinearColor(
     Linear Color=$Particles.Ground Initial Color
     Scale Factor=&MakeVector4_FromLinearColor(LinearColor=&MakeLinearColorFromVectorAndFloat(
       "Vector (RGB)"=&VectorFromFloat(Value=&FloatFromCurve(
         FloatCurve=DataInterface() {propertyValues {Curve {keys [<at 0 1>,<at 0.1 15>,<at 0.2 1>,<at 0.4 13.5>,<at 0.6 0>,<at 1 1>]}}}
         CurveIndex=$Particles.Update Ground Flash Age
        ))
       "Float (Alpha)"=&FloatFromCurve(
        FloatCurve=DataInterface() {propertyValues {Curve {keys [
            {arriveTangent=9.0725, interpMode=RCIM_Cubic, leaveTangent=9.0725, tangentMode=RCTM_User, time=0, value=0}
            {arriveTangent=-2.5249, interpMode=RCIM_Cubic, leaveTangent=-2.5249, tangentMode=RCTM_User, time=0.1408, value=0.9847}
            {arriveTangent=-0.0405, interpMode=RCIM_Cubic, leaveTangent=-0.0405, tangentMode=RCTM_User, time=1, value=0}
           ]}}}
        CurveIndex=$Particles.Update Ground Flash Age
       )
      ))
    )
    Particles.Update Ground Flash size:Vector2f=&Vector2DFromFloat(Value=&FloatFromCurve(
      FloatCurve=DataInterface() {propertyValues {Curve {keys [
          {arriveTangent=9.0725, interpMode=RCIM_Cubic, leaveTangent=9.0725, tangentMode=RCTM_User, time=0, value=0}
          {arriveTangent=-0.0081, interpMode=RCIM_Cubic, leaveTangent=-0.0081, tangentMode=RCTM_User, time=0.1408, value=0.9847}
          {arriveTangent=-3.0837, interpMode=RCIM_Cubic, leaveTangent=-3.0837, tangentMode=RCTM_User, time=1, value=0}
         ]}}}
      CurveIndex=$Particles.Ground Flash lifetime
      Scale Curve=0.75
     ))
   ) {isSetParameters=true}
   /SetVariables(Particles.Update Top flare color:LinearColor=&MultiplyLinearColor(
     Linear Color=$Particles.Ground Initial Color
     Scale Factor=&MakeVector4_FromLinearColor(LinearColor=&MakeLinearColorFromVectorAndFloat(
       "Vector (RGB)"=&VectorFromFloat(Value=&FloatFromCurve(
         FloatCurve=DataInterface() {propertyValues {Curve {keys [<at 0 1>,<at 0.1 15>,<at 0.2 1>,<at 0.4 13.5>,<at 0.6 0>,<at 1 1>]}}}
         CurveIndex=$Particles.Update Top Flash Age
        ))
       "Float (Alpha)"=&FloatFromCurve(
        FloatCurve=DataInterface() {propertyValues {Curve {keys [
            {arriveTangent=9.0725, interpMode=RCIM_Cubic, leaveTangent=9.0725, tangentMode=RCTM_User, time=0, value=0}
            {arriveTangent=-2.5249, interpMode=RCIM_Cubic, leaveTangent=-2.5249, tangentMode=RCTM_User, time=0.1408, value=0.9847}
            {arriveTangent=-0.0405, interpMode=RCIM_Cubic, leaveTangent=-0.0405, tangentMode=RCTM_User, time=1, value=0}
           ]}}}
        CurveIndex=$Particles.Update Top Flash Age
        Scale Curve=0.5
       )
      ))
    )) {isSetParameters=true}
   /SetVariables(
    Particles.Update sparks color:LinearColor=&MultiplyLinearColor(
     Linear Color=$Particles.Initial.Sparks color
     Scale Factor=&MakeVector4_FromLinearColor(LinearColor=&MakeLinearColorFromVectorAndFloat(
       "Vector (RGB)"=&VectorFromFloat(Value=&FloatFromCurve(
         FloatCurve=DataInterface() {propertyValues {Curve {keys [<at 0 1>,<at 0.2 1>,<at 0.3 5>,<at 0.4 1>,<at 0.7 1>,<at 0.8 5>,<at 0.9 1>,<at 1 1>]}}}
         CurveIndex=$Particles.Update Sparks Age
         Scale Curve=5
        ))
       "Float (Alpha)"=&FloatFromCurve(
        FloatCurve=DataInterface() {propertyValues {Curve {keys [
            {arriveTangent=9.0725, interpMode=RCIM_Cubic, leaveTangent=9.0725, tangentMode=RCTM_User, time=0, value=0}
            {arriveTangent=-2.5249, interpMode=RCIM_Cubic, leaveTangent=-2.5249, tangentMode=RCTM_User, time=0.1408, value=0.9847}
            {arriveTangent=-0.5654, interpMode=RCIM_Cubic, leaveTangent=-0.5654, time=0.5, value=0.35}
            {arriveTangent=-1.25, interpMode=RCIM_Cubic, leaveTangent=-1.25, time=0.6, value=0.725}
            {arriveTangent=-3.375, interpMode=RCIM_Cubic, leaveTangent=-0.1412, tangentMode=RCTM_Break, time=0.7, value=0.1}
            {arriveTangent=-0.7119, interpMode=RCIM_Cubic, leaveTangent=1.625, tangentMode=RCTM_Break, time=0.8, value=0.05}
            {arriveTangent=-0.25, interpMode=RCIM_Cubic, leaveTangent=-0.25, time=0.9, value=0.425}
            {arriveTangent=-0.0405, interpMode=RCIM_Cubic, leaveTangent=-0.0405, tangentMode=RCTM_User, time=1, value=0}
           ]}}}
        CurveIndex=$Particles.Update Sparks Age
       )
      ))
    )
    Particles.Update sparks size:Vector2f=&Vector2DFromCurve(
     Vector2Curve=DataInterface() {propertyValues {XCurve {keys [<at 0 0.7>,<at 0.5 1.05>,<at 1 0.85>]}, YCurve {keys [<at 0 2>,<at 1 0.05>]}}}
     CurveIndex=$Particles.Update Sparks Age
     ScaleCurve=&Multiply_Vector2D(
      A=$Particles.Initial.Spark size
      B=&Vector2DFromCurve(
       Vector2Curve=DataInterface() {propertyValues {XCurve {keys [<at 0 1>,<at 0.4 1.125>,<at 1 1>]}, YCurve {keys [<at 0 3>,<at 0.4 3>,<at 1 3>]}}}
       CurveIndex=$Particles.Update Sparks Age
       ScaleCurve=<0.5 0.5>
      )
     )
    )
   ) {isSetParameters=true}
   /SetVariables(
    Particles.Update electricity Color:LinearColor=&MultiplyLinearColor(
     Linear Color=$Particles.Initial.Electricity Color
     Scale Factor=&MakeVector4_FromLinearColor(LinearColor=&MakeLinearColorFromVectorAndFloat(
       "Vector (RGB)"=&Lerp_Vector(
        A=&VectorFromFloat(Value=&FloatFromCurve(
          FloatCurve=DataInterface() {propertyValues {Curve {keys [<at 0 1>,<at 0.1 1>,<at 0.2 2>,<at 0.3 1>,<at 0.6 1>,<at 0.7 2>,<at 0.8 1>,<at 1 1>]}}}
          CurveIndex=$Particles.Update Electricity age
          Scale Curve=10
         ))
        B=&VectorFromFloat(Value=&RandomRangeFloat(Minimum=0.3, Maximum=8, Evaluation Type=::NewEnumerator1))
        Alpha=&FloatFromCurve(
         FloatCurve=DataInterface() {propertyValues {Curve {keys [
             {interpMode=RCIM_Cubic, leaveTangent=2.5294, tangentMode=RCTM_Break, time=0, value=0}
             {arriveTangent=-0.2166, interpMode=RCIM_Cubic, tangentMode=RCTM_Break, time=0.4, value=1}
            ]}}}
         CurveIndex=$Particles.Update Electricity age
        )
       )
       "Float (Alpha)"=&FloatFromCurve(
        FloatCurve=DataInterface() {propertyValues {Curve {keys [
            {tangentMode=RCTM_User, time=0, value=1}
            {arriveTangent=-0.0405, interpMode=RCIM_Cubic, leaveTangent=-0.0405, tangentMode=RCTM_User, time=0.7009, value=1}
            {arriveTangent=-8.0089, interpMode=RCIM_Cubic, leaveTangent=-8.0089, tangentMode=RCTM_User, time=1, value=0}
           ]}}}
        CurveIndex=$Particles.Update Electricity age
       )
      ))
    )
    Particles.Update electricity dynamic parameters 0:Vector4f=&MakeVector4(
     X=&MakeFloatFromVector(Vector=$Particles.electricity scale, Channel=::NewEnumerator0)
     Y=&RandomRangeFloat(Minimum=0, Maximum=10)
     Z=&FloatFromCurve(
      FloatCurve=DataInterface() {propertyValues {Curve {keys [{interpMode=RCIM_Cubic, leaveTangent=8.1085, tangentMode=RCTM_Break, time=0.8568, value=1}, {tangentMode=RCTM_User, time=1, value=0}]}}}
      CurveIndex=$Particles.Update Electricity age
      Scale Curve=0.0263
     )
     W=0.8
    )
    Particles.Update electricity dynamic parameters 1:Vector4f=&MakeVector4(
     X=&FloatFromCurve(
      FloatCurve=DataInterface() {propertyValues {Curve {keys [
          {interpMode=RCIM_Cubic, leaveTangent=1.1034, tangentMode=RCTM_Break, time=0, value=0}
          {arriveTangent=2.2399, tangentMode=RCTM_User, time=0.5, value=1}
         ]}}}
      CurveIndex=$Particles.Update Electricity age
      Scale Curve=&RandomRangeFloat(Minimum=0.3, Maximum=1)
     )
     Y=&MaskFloatBySpawnGroup(
      Masked Float=&FloatFromCurve(
       FloatCurve=DataInterface() {propertyValues {Curve {keys [
           {interpMode=RCIM_Cubic, tangentMode=RCTM_User, time=0, value=0}
           {arriveTangent=2.7735, interpMode=RCIM_Cubic, leaveTangent=2.7735, tangentMode=RCTM_User, time=1, value=1}
          ]}}}
       CurveIndex=$Particles.Update Electricity age
      )
      Passthrough Float=&FloatFromCurve(FloatCurve=DataInterface() {propertyValues {Curve {keys [<at 0.5 0>, {arriveTangent=3.2053, tangentMode=RCTM_User, time=0.9, value=1}]}}})
     )
     Z=&FloatFromCurve(
      FloatCurve=DataInterface() {propertyValues {Curve {keys [<at 0 0 RCIM_Cubic>, {arriveTangent=3.4223, interpMode=RCIM_Cubic, leaveTangent=3.4223, tangentMode=RCTM_User, time=1, value=1}]}}}
      CurveIndex=$Particles.Update Electricity age
     )
     W=&FloatFromCurve(
      FloatCurve=DataInterface() {propertyValues {Curve {keys [
          {interpMode=RCIM_Cubic, leaveTangent=0.1169, tangentMode=RCTM_Break, time=0, value=0}
          {arriveTangent=2.5708, interpMode=RCIM_Cubic, leaveTangent=2.5708, tangentMode=RCTM_User, time=1, value=1}
         ]}}}
      CurveIndex=$Particles.Update Electricity age
      Scale Curve=&RandomRangeFloat(Minimum=1, Maximum=40)
     )
    )
    Particles.Update electricity dynamic parameters 2:Vector4f=&MakeVector4(
     X=&FloatFromCurve(
      FloatCurve=DataInterface() {propertyValues {Curve {keys [
          {interpMode=RCIM_Cubic, leaveTangent=-0.0025, tangentMode=RCTM_Break, time=0, value=0.5}
          {arriveTangent=-0.0316, interpMode=RCIM_Cubic, tangentMode=RCTM_Break, time=1, value=-0.2}
         ]}}}
      CurveIndex=$Particles.Update Electricity age
     )
     Y=&FloatFromCurve(
      FloatCurve=DataInterface() {propertyValues {Curve {keys [
          {interpMode=RCIM_Cubic, tangentMode=RCTM_User, time=0, value=0}
          {arriveTangent=2.0522, interpMode=RCIM_Cubic, leaveTangent=2.0522, tangentMode=RCTM_User, time=1, value=1}
         ]}}}
      Scale Curve=2
     )
     Z=0
     W=0.8
    )
    Particles.Update electricity dynamic parameters 3:Vector4f=&MakeVector4(X=0, Y=0, Z=0, W=0)
    Particles.Update electricity scale:Vector3f=&Multiply_Vector(
     A=$Particles.Initial.electricity scale
     B=&VectorFromCurve(
      VectorCurve=DataInterface() {propertyValues {
        XCurve {keys [
          {arriveTangent=1.8776, interpMode=RCIM_Cubic, leaveTangent=1.8776, tangentMode=RCTM_User, time=0, value=0}
          {arriveTangent=0.7, interpMode=RCIM_Cubic, leaveTangent=0.7, time=0.2485, value=0.4984}
          <at 1 0.7 RCIM_Cubic>
         ]}
        YCurve {keys [
          {arriveTangent=1.8776, interpMode=RCIM_Cubic, leaveTangent=1.8776, tangentMode=RCTM_User, time=0, value=0}
          {arriveTangent=0.7, interpMode=RCIM_Cubic, leaveTangent=0.7, time=0.2497, value=0.5076}
          <at 1 0.7 RCIM_Cubic>
         ]}
        ZCurve {keys [
          {arriveTangent=1.8776, interpMode=RCIM_Cubic, leaveTangent=1.8776, tangentMode=RCTM_User, time=0, value=0}
          {arriveTangent=0.7, interpMode=RCIM_Cubic, leaveTangent=0.7, time=0.2485, value=0.506}
          <at 1 0.7 RCIM_Cubic>
         ]}
       }}
      CurveIndex=$Particles.Update Electricity age
      Scale Curve=<0.5 0.5 0.5>
     )
    )
   ) {isSetParameters=true}
   /SetVariables(
    Particles.Update Ground Shockwave Color:LinearColor=&MakeLinearColorFromVectorAndFloat(
     "Vector (RGB)"=&Multiply_VectorByFloat(Vector=<7 7 7>, Float=2)
     "Float (Alpha)"=&FloatFromCurve(
      FloatCurve=DataInterface() {propertyValues {Curve {keys [{interpMode=RCIM_Cubic, tangentMode=RCTM_User, time=0.5, value=1}, {arriveTangent=-0.0564, tangentMode=RCTM_User, time=0.8955, value=0}]}}}
      CurveIndex=$Particles.Update Ground shockwave Age
      Scale Curve=&Multiply_Float(A=2.5, B=&FNTimeOfDayScalar(Night Scale=1, Morning Scale=0.8, Day Scale=1.7, Evening Scale=0.8))
     )
    )
    Particles.Update Ground Shockwave Scale:Vector3f=&VectorFromFloat(Value=&FloatFromCurve(
      FloatCurve=DataInterface() {propertyValues {Curve {keys [
          {arriveTangent=0.9512, interpMode=RCIM_Cubic, leaveTangent=0.9512, tangentMode=RCTM_User, time=0, value=0.2}
          {arriveTangent=-0.0278, interpMode=RCIM_Cubic, leaveTangent=-0.0278, tangentMode=RCTM_User, time=1, value=0.75}
         ]}}}
      CurveIndex=$Particles.Update Ground shockwave Age
     ))
   ) {isSetParameters=true}
   /SetVariables(
    Particles.Update Impact Spikes MeshScale:Vector3f=&Multiply_Vector(
     A=$Particles.Initial.Impact Spikes mesh scale
     B=&VectorFromFloat(Value=&FloatFromCurve(
       FloatCurve=DataInterface() {propertyValues {Curve {keys [
           {arriveTangent=4.1785, interpMode=RCIM_Cubic, leaveTangent=4.1785, tangentMode=RCTM_User, time=0, value=0}
           {arriveTangent=1.1179, interpMode=RCIM_Cubic, leaveTangent=1.1179, tangentMode=RCTM_User, time=0.2042, value=0.6228}
           {arriveTangent=0.0224, interpMode=RCIM_Cubic, leaveTangent=0.0224, tangentMode=RCTM_User, time=1, value=1}
          ]}}}
       CurveIndex=$Particles.Update Spikes Impact Age
       Scale Curve=0.6
      ))
    )
    Particles.Update Impact Spikes Velocity:Vector3f=&RandomRangeVector(Minimum=<-50 -50 50>, Maximum=<50 50 50>)
    Particles.Update Impact Spikes color:LinearColor=&MakeLinearColorFromVectorAndFloat(
     "Vector (RGB)"=&Multiply_VectorByFloat(Vector=<1 1 1>, Float=1)
     "Float (Alpha)"=&FloatFromCurve(
      FloatCurve=DataInterface() {propertyValues {Curve {keys [<at 0 0 RCIM_Cubic>,<at 0.2 0.975 RCIM_Cubic>,<at 1 0 RCIM_Cubic>]}}}
      CurveIndex=$Particles.Update Spikes Impact Age
      Scale Curve=&Multiply_Float(
       A=&RandomRangeFloat(Minimum=0.3, Maximum=0.75)
       B=&FNTimeOfDayScalar(Night Scale=1, Morning Scale=0.8, Day Scale=1.7, Evening Scale=0.8)
      )
     )
    )
    Particles.Update Impact Spikes dynamic parameters 0:Vector4f=&MakeVector4(
     X=&FloatFromCurve(
      FloatCurve=DataInterface() {propertyValues {Curve {keys [
          {interpMode=RCIM_Cubic, tangentMode=RCTM_User, time=0, value=0}
          {arriveTangent=3.8507, interpMode=RCIM_Cubic, leaveTangent=3.8507, tangentMode=RCTM_User, time=0.2634, value=0.3109}
          {interpMode=RCIM_Cubic, tangentMode=RCTM_User, time=1, value=1}
         ]}}}
      CurveIndex=$Particles.Update Spikes Impact Age
     )
     Y=2
     Z=&RandomRangeFloat(Minimum=0, Maximum=1)
     W=1
    )
   ) {isSetParameters=true}
   /SetVariables(
    Particles.Update Staff Pulse Color:LinearColor=&MakeLinearColorFromVectorAndFloat(
     "Vector (RGB)"=&MakeVectorFromLinearColorRGB(Color=&MultiplyLinearColor(Linear Color=<rgba 1.7071 0.6284 5 1>, Scale Factor=<5 5 5 1>))
     "Float (Alpha)"=&FloatFromCurve(
      FloatCurve=DataInterface() {propertyValues {Curve {keys [
          {arriveTangent=-1.6399, interpMode=RCIM_Cubic, leaveTangent=-1.6399, tangentMode=RCTM_User, time=0, value=1}
          {interpMode=RCIM_Cubic, tangentMode=RCTM_User, time=1, value=0}
         ]}}}
      CurveIndex=$Particles.Update Staff Pulse Age
      Scale Curve=2
     )
    )
    Particles.Update Staff Pulse dynamic parameters 0:Vector4f=&MakeVector4(
     X=1
     Y=&FloatFromCurve(
      FloatCurve=DataInterface() {propertyValues {Curve {keys [
          {interpMode=RCIM_Cubic, tangentMode=RCTM_User, time=0, value=1}
          {arriveTangent=-1, interpMode=RCIM_Cubic, leaveTangent=-1, time=0.5, value=0.5}
          {interpMode=RCIM_Cubic, leaveTangent=-3.3753, time=1, value=0}
         ]}}}
      CurveIndex=$Particles.Update Staff Pulse Age
     )
     Z=1
     W=1
    )
    Particles.Update Staff Pulse size:Vector2f=&Multiply_Vector2D_ByFloat(
     Float=&FloatFromCurve(
      FloatCurve=DataInterface() {propertyValues {Curve {
         defaultValue=0
         keys [
          {interpMode=RCIM_Cubic, tangentMode=RCTM_User, time=0, value=0}
          {arriveTangent=0.4799, interpMode=RCIM_Cubic, leaveTangent=0.4799, tangentMode=RCTM_User, time=0.3, value=0.7}
          {interpMode=RCIM_Cubic, tangentMode=RCTM_User, time=1, value=1}
         ]
        }}}
      CurveIndex=$Particles.Update Staff Pulse Age
     )
     Vector2D=$Particles.Initial.Staff Pulse Size
    )
   ) {isSetParameters=true}
   /SetVariables(
    Particles.Update Debris Rotation:NiagaraFloat=&Multiply_Float(
     A=$Particles.Initial.Debris InitSpriteRotate
     B=&FloatFromCurve(
      FloatCurve=DataInterface() {propertyValues {Curve {keys [
          {arriveTangent=-0.2085, interpMode=RCIM_Cubic, leaveTangent=-0.2085, tangentMode=RCTM_User, time=-0.0046, value=1.0113}
          {arriveTangent=-0.2496, interpMode=RCIM_Cubic, leaveTangent=-0.2496, tangentMode=RCTM_User, time=0.5, value=0.05}
         ]}}}
      CurveIndex=$Emitter.Age
      Scale Curve=950
     )
    )
    Particles.Update Debris Size:Vector2f=&Multiply_Vector2D_ByFloat(
     Float=&FloatFromCurve(
      FloatCurve=DataInterface() {propertyValues {Curve {keys [
          {arriveTangent=40, interpMode=RCIM_Cubic, time=0, value=0}
          {arriveTangent=-0.2323, interpMode=RCIM_Cubic, leaveTangent=-0.2323, tangentMode=RCTM_User, time=0.025, value=1}
          <at 0.5 0.875>
          <at 0.9 0.3>
          {arriveTangent=-0.0285, interpMode=RCIM_Cubic, leaveTangent=-0.0285, tangentMode=RCTM_User, time=1, value=0}
         ]}}}
      CurveIndex=$Particles.Update debris Age
     )
     Vector2D=$Particles.Initial.Debris size
    )
    Particles.Update Debris color:LinearColor=&MakeLinearColorFromVectorAndFloat("Vector (RGB)"=&MakeVectorFromLinearColorRGB(Color=$Particles.Initial.Debris color), "Float (Alpha)"=1)
   ) {isSetParameters=true}
   /SetVariables(
    Particles.Update Smoke CameraOffset=10
    Particles.Update Smoke color:LinearColor=&MakeLinearColorFromVectorAndFloat(
     "Vector (RGB)"=&VectorFromFloat(Value=&Multiply_Float(
       A=&MakeFloatFromLinearColor(
        LinearColor=&RandomRangeLinearColor(Minimum=<rgba 0.6073 0.6073 0.6073 1>, Maximum=<rgba 0.6128 0.5758 0.6867 1>)
        Channel:ENiagaraFloatFromLinearColorOptions=::NewEnumerator0
       )
       B=&FloatFromCurve(
        FloatCurve=DataInterface() {propertyValues {Curve {keys [{arriveTangent=1.8145, interpMode=RCIM_Cubic, leaveTangent=0.4403, tangentMode=RCTM_Break, time=0, value=0.4}, <at 1 1>]}}}
        CurveIndex=$Particles.Update smoke Impact Age
        Scale Curve=0.85
       )
      ))
     "Float (Alpha)"=&Multiply_Float(
      A=&FloatFromCurve(
       FloatCurve=DataInterface() {propertyValues {Curve {keys [
           {interpMode=RCIM_Cubic, tangentMode=RCTM_Break, time=0, value=0}
           {interpMode=RCIM_Cubic, tangentMode=RCTM_Break, time=0.2, value=2.25}
           {interpMode=RCIM_Cubic, tangentMode=RCTM_Break, time=0.7612, value=1.5}
           {interpMode=RCIM_Cubic, tangentMode=RCTM_Break, time=1.5, value=0}
          ]}}}
       CurveIndex=$Particles.Update smoke Impact Age
       Scale Curve=&RandomRangeFloat(Minimum=0.5, Maximum=1.5)
      )
      B=&Multiply_Float(A=&FNTimeOfDayScalar(Night Scale=1, Morning Scale=0.7, Evening Scale=0.7), B=1.5)
     )
    )
    Particles.Update Smoke size:Vector2f=&Multiply_Vector2D_ByFloat(
     Float=&FloatFromCurve(
      FloatCurve=DataInterface() {propertyValues {Curve {keys [
          {arriveTangent=0.0418, interpMode=RCIM_Cubic, leaveTangent=0.0418, tangentMode=RCTM_User, time=0, value=0.5}
          {arriveTangent=0.8771, interpMode=RCIM_Cubic, leaveTangent=0.8771, tangentMode=RCTM_User, time=0.7, value=0.775}
          {arriveTangent=0.7996, arriveTangentWeight=0.3375, interpMode=RCIM_Cubic, leaveTangent=0.7996, tangentMode=RCTM_User, time=1, value=1}
         ]}}}
      CurveIndex=$Particles.Update smoke Impact Age
      Scale Curve=1.8
     )
     Vector2D=$Particles.Initial.Smoke Sprite Size
    )
    Particles.Update smoke Dynamic Parameters 0:Vector4f=&MakeVector4(
     X=&FloatFromCurve(
      FloatCurve=DataInterface() {propertyValues {Curve {keys [
          {arriveTangent=2.7533, interpMode=RCIM_Cubic, leaveTangent=1.8958, tangentMode=RCTM_User, time=0, value=1}
          {arriveTangent=0.8292, interpMode=RCIM_Cubic, leaveTangent=0.8292, tangentMode=RCTM_User, time=0.321, value=1.3453}
          {arriveTangent=0.0529, interpMode=RCIM_Cubic, leaveTangent=0.2021, tangentMode=RCTM_User, time=1, value=1.5}
         ]}}}
      CurveIndex=&Add_Float(
       A=&Multiply_Float(A=&RandomRangeFloat(Minimum=-1, Maximum=1), B=0.5)
       B=$Particles.Update smoke Impact Age
      )
     )
     Y=&FloatFromCurve(
      FloatCurve=DataInterface() {propertyValues {Curve {keys [
          {arriveTangent=0.0453, interpMode=RCIM_Cubic, tangentMode=RCTM_Break, time=0, value=-1}
          {arriveTangent=0.0604, interpMode=RCIM_Cubic, leaveTangent=0.0604, tangentMode=RCTM_User, time=1, value=-1}
         ]}}}
      CurveIndex=$Particles.Update smoke Impact Age
     )
     Z=&FloatFromCurve(
      FloatCurve=DataInterface() {propertyValues {Curve {keys [
          {arriveTangent=-7.4008, interpMode=RCIM_Cubic, leaveTangent=-7.4008, tangentMode=RCTM_User, time=0, value=0.7}
          {arriveTangent=-0.3891, interpMode=RCIM_Cubic, leaveTangent=0.2723, tangentMode=RCTM_Break, time=0.3, value=0}
         ]}}}
      CurveIndex=$Particles.Update smoke Impact Age
     )
     W=&FloatFromCurve(FloatCurve=DataInterface() {propertyValues {Curve {keys [<at 0.5 1>,<at 1 0>]}}}, CurveIndex=$Particles.Update smoke Impact Age)
    )
    Particles.Update smoke Dynamic Parameters 1:Vector4f=&MakeVector4(
     X=&FloatFromCurve(
      FloatCurve=DataInterface() {propertyValues {Curve {keys [
          {interpMode=RCIM_Cubic, leaveTangent=4.7088, tangentMode=RCTM_Break, time=0, value=0}
          {arriveTangent=4.2257, interpMode=RCIM_Cubic, leaveTangent=4.2257, tangentMode=RCTM_User, time=0.25, value=1}
         ]}}}
      CurveIndex=$Particles.Update smoke Impact Age
      Scale Curve=&RandomRangeFloat(Minimum=1.5, Maximum=1.5)
     )
     Y=&FloatFromCurve(
      FloatCurve=DataInterface() {propertyValues {Curve {keys [<at 0 0.2>, {leaveTangent=0.9066, time=1, value=2}]}}}
      CurveIndex=$Particles.Update smoke Impact Age
      Scale Curve=&RandomRangeFloat(Minimum=0.5, Maximum=0.6)
     )
     Z=&FloatFromCurve(
      FloatCurve=DataInterface() {propertyValues {Curve {keys [
          {interpMode=RCIM_Cubic, leaveTangent=6.3917, tangentMode=RCTM_Break, time=0, value=0.2}
          {arriveTangent=1.7549, interpMode=RCIM_Cubic, leaveTangent=10.6767, tangentMode=RCTM_User, time=0.25, value=1}
         ]}}}
      CurveIndex=$Particles.Update smoke Impact Age
     )
     W=&FloatFromCurve(
      FloatCurve=DataInterface() {propertyValues {Curve {keys [{arriveTangent=-0.136, interpMode=RCIM_Cubic, leaveTangent=-0.136, tangentMode=RCTM_User, time=0, value=1}, <at 1 0.95>]}}}
      CurveIndex=$Particles.Update smoke Impact Age
      Scale Curve=&RandomRangeFloat(Minimum=1.1, Maximum=1.4)
     )
    )
    Particles.Update smoke Dynamic Parameters 2:Vector4f=&MakeVector4(
     X=&FloatFromCurve(
      FloatCurve=DataInterface() {propertyValues {Curve {keys [
          {interpMode=RCIM_Cubic, leaveTangent=4.2485, tangentMode=RCTM_Break, time=0, value=0}
          {arriveTangent=1.0548, interpMode=RCIM_Cubic, leaveTangent=1.0548, tangentMode=RCTM_User, time=0.17, value=0.5715}
          <at 1 1>
         ]}}}
      CurveIndex=$Particles.Update smoke Impact Age
     )
     Y=&FloatFromCurve(
      FloatCurve=DataInterface() {propertyValues {Curve {keys [
          {arriveTangent=1.9056, interpMode=RCIM_Cubic, leaveTangent=1.9056, tangentMode=RCTM_User, time=0, value=0.2}
          {arriveTangent=0.0259, interpMode=RCIM_Cubic, leaveTangent=0.0259, tangentMode=RCTM_User, time=1, value=1.4}
         ]}}}
      CurveIndex=$Particles.Update smoke Impact Age
      Scale Curve=&RandomRangeFloat(Minimum=0.2, Maximum=0.7)
     )
     Z=&FloatFromCurve(
      FloatCurve=DataInterface() {propertyValues {Curve {keys [
          {interpMode=RCIM_Cubic, leaveTangent=2.3064, tangentMode=RCTM_Break, time=0, value=1}
          {interpMode=RCIM_Cubic, leaveTangent=11.1233, tangentMode=RCTM_Break, time=1, value=2}
         ]}}}
      CurveIndex=$Particles.Update smoke Impact Age
     )
     W=&FloatFromCurve(
      FloatCurve=DataInterface() {propertyValues {Curve {keys [
          {arriveTangent=-0.6867, interpMode=RCIM_Cubic, leaveTangent=-0.6867, tangentMode=RCTM_User, time=0, value=1}
          {interpMode=RCIM_Cubic, leaveTangent=-1.2956, tangentMode=RCTM_Break, time=1, value=0.75}
         ]}}}
      CurveIndex=$Particles.Update smoke Impact Age
      Scale Curve=1.6
     )
    )
    Particles.Update smoke SpriteRotation:NiagaraFloat=&Multiply_Float(
     A=&RandomRangeFloat(Minimum=-1, Maximum=1)
     B=&FloatFromCurve(
      FloatCurve=DataInterface() {propertyValues {Curve {keys [
          {arriveTangent=-4.0107, interpMode=RCIM_Cubic, leaveTangent=-4.0107, tangentMode=RCTM_User, time=-0.0046, value=1.0113}
          {arriveTangent=-0.2496, interpMode=RCIM_Cubic, leaveTangent=-0.2496, tangentMode=RCTM_User, time=0.5, value=0.16}
         ]}}}
      CurveIndex=$Emitter.Age
      Scale Curve=150
     )
    )
   ) {isSetParameters=true}
   &Drag(
    Drag=&MaskFloat(
     Mask Source=0
     Mask Parameter=?
     Mask=2
     Masked Float=&RandomRangeFloat(Minimum=0.7, Maximum=1.1)
     Passthrough Float=&MaskFloat(
      Mask Source=0
      Mask Parameter=?
      Mask=3
      Masked Float=0
      Passthrough Float=&MaskFloat(
       Mask Source=0
       Mask Parameter=?
       Mask=4
       Masked Float=0
       Passthrough Float=&MaskFloat(
        Mask Source=0
        Mask Parameter=?
        Mask=5
        Masked Float=3
        Passthrough Float=&MaskFloat(
         Mask Source=0
         Mask Parameter=?
         Mask=6
         Masked Float=0
         Passthrough Float=&MaskFloat(
          Mask Source=0
          Mask Parameter=?
          Mask=7
          Masked Float=&FloatFromCurve(
           FloatCurve=DataInterface() {propertyValues {Curve {keys [{arriveTangent=-1, interpMode=RCIM_Cubic, leaveTangent=-3.4078, tangentMode=RCTM_Break, time=0, value=1}, <at 0.3 0.1>, <at 1 0>]}}}
           CurveIndex=$Particles.Update debris Age
           Scale Curve=7
          )
          Passthrough Float=&FloatFromCurve(
           FloatCurve=DataInterface() {propertyValues {Curve {
              defaultValue=1
              keys [{interpMode=RCIM_Cubic, tangentMode=RCTM_User, time=0, value=1}, {interpMode=RCIM_Cubic, tangentMode=RCTM_User, time=1, value=0}]
             }}}
           CurveIndex=$Particles.Update smoke Impact Age
           Scale Curve=10
          )
         )
        )
       )
      )
     )
    )
    Use Linear Drag=true
    Use Rotational Drag=false
   )
   &GravityForce(Gravity=&MaskVector(
     Mask Source=0
     Mask Parameter=?
     Mask=2
     Masked Vector=<0 0 -800>
     Passthrough Vector=&MaskVector(
      Mask Source=0
      Mask Parameter=?
      Mask=3
      Masked Vector=<0 0 0>
      Passthrough Vector=&MaskVector(
       Mask Source=0
       Mask Parameter=?
       Mask=4
       Masked Vector=<0 0 0>
       Passthrough Vector=&MaskVector(
        Mask Source=0
        Mask Parameter=?
        Mask=5
        Masked Vector=<0 0 0>
        Passthrough Vector=&MaskPosition(
         Mask Source:NiagaraInt32=0
         Mask Parameter:NiagaraInt32=?
         Mask:NiagaraInt32=6
         Masked Position:NiagaraPosition=<0 0 0>
         Passthrough Position:NiagaraPosition=&MaskVector(
          Mask Source=0
          Mask Parameter=?
          Mask=7
          Masked Vector=&RandomRangeVector(Minimum=<0 0 -980>, Maximum=<1 1 -1500>)
          Passthrough Vector=<0 0 0>
         )
        )
       )
      )
     )
    ))
   &FN_Light_Attributes(
    Color=<rgba 0.2582 0.0437 0.6867 1>
    Create Color Attribute=true
    Create Exponent Attribute=true
    Create Position Attribute=true
    Create Radius Attribute=true
    Enable Particle Light Enable=true
    Exponent=1
    Light Enabled=true
    Radius=80
    Scale Color=&FloatFromCurve(
     FloatCurve=DataInterface() {propertyValues {Curve {keys [
         {arriveTangent=9.0725, interpMode=RCIM_Cubic, leaveTangent=9.0725, tangentMode=RCTM_User, time=0, value=0}
         {arriveTangent=-2.5249, interpMode=RCIM_Cubic, leaveTangent=-2.5249, tangentMode=RCTM_User, time=0.1408, value=0.9847}
         {arriveTangent=-0.0405, interpMode=RCIM_Cubic, leaveTangent=-0.0405, tangentMode=RCTM_User, time=1, value=0}
        ]}}}
     Scale Curve=50
    )
   ) {enabled=false}
   &SolveForcesAndVelocity(Clamp Velocity=false, Limit Acceleration=false)
  ]
  rendererClasses [
   &NiagaraSpriteRendererProperties
   &NiagaraLightRendererProperties
   &NiagaraMeshRendererProperties
  ]
  renderers [
   {rendererClass=&NiagaraSpriteRendererProperties}
   {rendererClass=&NiagaraLightRendererProperties, rendererIndex=1}
   {rendererClass=&NiagaraSpriteRendererProperties, rendererIndex=2}
   {rendererClass=&NiagaraSpriteRendererProperties, rendererIndex=3}
   {rendererClass=&NiagaraMeshRendererProperties, rendererIndex=4}
   {rendererClass=&NiagaraMeshRendererProperties, rendererIndex=5}
   {rendererClass=&NiagaraMeshRendererProperties, rendererIndex=6}
   {rendererClass=&NiagaraMeshRendererProperties, rendererIndex=7}
   {rendererClass=&NiagaraSpriteRendererProperties, rendererIndex=8}
   {rendererClass=&NiagaraSpriteRendererProperties, rendererIndex=9}
   {rendererClass=&NiagaraSpriteRendererProperties, rendererIndex=10}
   {rendererClass=&NiagaraSpriteRendererProperties, rendererIndex=11}
  ]
 }
 High-epic_dust {
  simTarget=GPUComputeSim
  emitterUpdateScript [
   &EmitterState(Life Cycle Mode=::NewEnumerator1, Loop Behavior=::NewEnumerator1, Loop Delay=0.03, Loop Duration=1, UseLoopDelay=true)
   &SpawnBurst_Instantaneous(
    Spawn Count=&RandomRangeInt(Minimum=25, Maximum=28)
    Spawn Group=3
    Spawn Time=0.05
    Use Spawn Probability=false
   ) {module=SpawnBurst_Instantaneous003}
  ]
  particleSpawnScript [
   &InitializeParticle(
    Color Maximum=<rgba 0.6128 0.5758 0.6867 1>
    Color Minimum=<rgba 0.6073 0.6073 0.6073 1>
    Color Mode=::NewEnumerator2
    Lifetime Max=1.55
    Lifetime Min=0.75
    Lifetime Mode=::NewEnumerator1
    Position Mode=::NewEnumerator1
    Position Offset=<0 0 -0.1>
    Sprite Rotation Angle=&RandomRangeFloat(Minimum=0, Maximum=360)
    Sprite Rotation Mode=::NewEnumerator0
    Sprite Size Mode=::NewEnumerator3
    Uniform Sprite Size=&Multiply_Float(A=&RandomRangeFloat(Minimum=27, Maximum=42), B=1.5)
    UsePositionOffset=true
    Write Material Random=true
   ) {module=InitializeParticle001}
   /SetVariables(Particles.VisibilityTag=&SpawnGroup()) {isSetParameters=true}
   &SubUVAnimation(End Frame=3, SubUV Animation Mode:ENiagara_SubUVLookupMode=::NewEnumerator2)
   /SetVariables(Particles.VelCurveRand:NiagaraFloat=&FloatFromCurve(
     FloatCurve=DataInterface() {propertyValues {Curve {keys [
         {interpMode=RCIM_Cubic, leaveTangent=0.1132, tangentMode=RCTM_Break, time=0, value=0}
         {arriveTangent=1.9858, interpMode=RCIM_Cubic, leaveTangent=1.7203, tangentMode=RCTM_Break, time=0.3, value=0.4}
         {interpMode=RCIM_Cubic, tangentMode=RCTM_Break, time=1, value=1}
        ]}}}
     CurveIndex=&RandomRangeFloat(Minimum=0, Maximum=1)
    )) {isSetParameters=true}
   &ShapeLocation(
    Handle Radius=0.1
    Large Radius=5
    MaskSpawnGroup=false
    Offset=<0 0 20>
    Offset Coordinate Space=::World
    Offset Mode=::NewEnumerator0
    Override Seed=false
    Rotation Coordinate Space=::World
    Shape Primitive=::NewEnumerator3
    UseOwnerScale=true
   )
   &AddVelocityFromPoint(
    Module.Use Falloff Distance=false
    Origin Offset=<0 0 20>
    Origin Offset Coordinate Space=::Simulation
    Velocity Strength=&Multiply_Float(A=&Lerp_Float(A=350, B=490, Alpha=$Particles.VelCurveRand), B=1)
   )
  ]
  particleUpdateScript [
   &ParticleState()
   &ScaleSpriteSize(Scale Factor=&Vector2DFromFloat(Value=&FloatFromCurve(FloatCurve=DataInterface() {propertyValues {Curve {keys [
          {arriveTangent=0.0418, interpMode=RCIM_Cubic, leaveTangent=0.0418, tangentMode=RCTM_User, time=0, value=0.5}
          {arriveTangent=0.8771, interpMode=RCIM_Cubic, leaveTangent=0.8771, tangentMode=RCTM_User, time=0.7, value=0.775}
          {arriveTangent=0.7996, arriveTangentWeight=0.3375, interpMode=RCIM_Cubic, leaveTangent=0.7996, tangentMode=RCTM_User, time=1, value=1}
         ]}}}))) {module=ScaleSpriteSize001}
   &Drag(
    Drag=&FloatFromCurve(
     FloatCurve=DataInterface() {propertyValues {Curve {
        defaultValue=1
        keys [{interpMode=RCIM_Cubic, tangentMode=RCTM_User, time=0, value=1}, {interpMode=RCIM_Cubic, tangentMode=RCTM_User, time=1, value=0}]
       }}}
     Scale Curve=11
    )
    Use Linear Drag=true
    Use Rotational Drag=false
   ) {module=Drag001}
   &SpriteRotationRate(Rotation Rate=&Multiply_Float(
     A=&RandomRangeFloat(Minimum=-1, Maximum=1)
     B=&FloatFromCurve(
      FloatCurve=DataInterface() {propertyValues {Curve {keys [
          {arriveTangent=-4.0107, interpMode=RCIM_Cubic, leaveTangent=-4.0107, tangentMode=RCTM_User, time=-0.0046, value=1.0113}
          {arriveTangent=-0.2496, interpMode=RCIM_Cubic, leaveTangent=-0.2496, tangentMode=RCTM_User, time=0.5, value=0.16}
         ]}}}
      CurveIndex=$Emitter.Age
      Scale Curve=150
     )
    ))
   &SolveForcesAndVelocity(Clamp Velocity=false, Limit Acceleration=false)
   &DynamicMaterialParameters(
    Index 0 Param 1=&FloatFromCurve(
     FloatCurve=DataInterface() {propertyValues {Curve {keys [
         {arriveTangent=2.7533, interpMode=RCIM_Cubic, leaveTangent=1.8958, tangentMode=RCTM_User, time=0, value=1}
         {arriveTangent=0.8292, interpMode=RCIM_Cubic, leaveTangent=0.8292, tangentMode=RCTM_User, time=0.321, value=1.3453}
         {arriveTangent=0.0529, interpMode=RCIM_Cubic, leaveTangent=0.2021, tangentMode=RCTM_User, time=1, value=1.5}
        ]}}}
     CurveIndex=&Add_Float(A=&RandomRangeFloat(Minimum=-1, Maximum=1), B=$Particles.NormalizedAge)
    )
    Index 0 Param 2=&FloatFromCurve(FloatCurve=DataInterface() {propertyValues {Curve {keys [
         {arriveTangent=0.0453, interpMode=RCIM_Cubic, tangentMode=RCTM_Break, time=0, value=-1}
         {arriveTangent=0.0604, interpMode=RCIM_Cubic, leaveTangent=0.0604, tangentMode=RCTM_User, time=1, value=-1}
        ]}}})
    Index 0 Param 3=&FloatFromCurve(FloatCurve=DataInterface() {propertyValues {Curve {keys [
         {arriveTangent=-7.4008, interpMode=RCIM_Cubic, leaveTangent=-7.4008, tangentMode=RCTM_User, time=0, value=0.7}
         {arriveTangent=-0.3891, interpMode=RCIM_Cubic, leaveTangent=0.2723, tangentMode=RCTM_Break, time=0.3, value=0}
        ]}}})
    Index 0 Param 4=&FloatFromCurve(FloatCurve=DataInterface() {propertyValues {Curve {keys [<at 0.5 1>,<at 1 0>]}}})
    Index 1 Param 1=&FloatFromCurve(
     FloatCurve=DataInterface() {propertyValues {Curve {keys [
         {interpMode=RCIM_Cubic, leaveTangent=4.7088, tangentMode=RCTM_Break, time=0, value=0}
         {arriveTangent=4.2257, interpMode=RCIM_Cubic, leaveTangent=4.2257, tangentMode=RCTM_User, time=0.25, value=1}
        ]}}}
     Scale Curve=&RandomRangeFloat(Minimum=1.5, Maximum=1.5)
    )
    Index 1 Param 2=&FloatFromCurve(
     FloatCurve=DataInterface() {propertyValues {Curve {keys [<at 0 0.2>, {leaveTangent=0.9066, time=1, value=2}]}}}
     Scale Curve=&RandomRangeFloat(Minimum=0.5, Maximum=0.6)
    )
    Index 1 Param 3=&FloatFromCurve(FloatCurve=DataInterface() {propertyValues {Curve {keys [
         {interpMode=RCIM_Cubic, leaveTangent=6.3917, tangentMode=RCTM_Break, time=0, value=0.2}
         {arriveTangent=1.7549, interpMode=RCIM_Cubic, leaveTangent=10.6767, tangentMode=RCTM_User, time=0.25, value=1}
        ]}}})
    Index 1 Param 4=&FloatFromCurve(
     FloatCurve=DataInterface() {propertyValues {Curve {keys [{arriveTangent=-0.136, interpMode=RCIM_Cubic, leaveTangent=-0.136, tangentMode=RCTM_User, time=0, value=1}, <at 1 0.95>]}}}
     Scale Curve=&RandomRangeFloat(Minimum=1.1, Maximum=1.4)
    )
    Index 2 Param 1=&FloatFromCurve(FloatCurve=DataInterface() {propertyValues {Curve {keys [
         {interpMode=RCIM_Cubic, leaveTangent=4.2485, tangentMode=RCTM_Break, time=0, value=0}
         {arriveTangent=1.0548, interpMode=RCIM_Cubic, leaveTangent=1.0548, tangentMode=RCTM_User, time=0.17, value=0.5715}
         <at 1 1>
        ]}}})
    Index 2 Param 2=&FloatFromCurve(
     FloatCurve=DataInterface() {propertyValues {Curve {keys [
         {arriveTangent=1.9056, interpMode=RCIM_Cubic, leaveTangent=1.9056, tangentMode=RCTM_User, time=0, value=0.2}
         {arriveTangent=0.0259, interpMode=RCIM_Cubic, leaveTangent=0.0259, tangentMode=RCTM_User, time=1, value=1.4}
        ]}}}
     Scale Curve=&RandomRangeFloat(Minimum=0.2, Maximum=0.7)
    )
    Index 2 Param 3=&FloatFromCurve(FloatCurve=DataInterface() {propertyValues {Curve {keys [
         {interpMode=RCIM_Cubic, leaveTangent=2.3064, tangentMode=RCTM_Break, time=0, value=1}
         {interpMode=RCIM_Cubic, leaveTangent=11.1233, tangentMode=RCTM_Break, time=1, value=2}
        ]}}})
    Index 2 Param 4=&FloatFromCurve(
     FloatCurve=DataInterface() {propertyValues {Curve {keys [
         {arriveTangent=-0.6867, interpMode=RCIM_Cubic, leaveTangent=-0.6867, tangentMode=RCTM_User, time=0, value=1}
         {interpMode=RCIM_Cubic, leaveTangent=-1.2956, tangentMode=RCTM_Break, time=1, value=0.75}
        ]}}}
     Scale Curve=1.6
    )
    Module.Param0WriteEnabled=true
    Module.Param0WriteEnabled1=true
    Module.Param0WriteEnabled2=true
    Module.Param1WriteEnabled=true
    Module.Param1WriteEnabled1=true
    Module.Param1WriteEnabled2=true
    Module.Param2WriteEnabled=true
    Module.Param2WriteEnabled1=true
    Module.Param2WriteEnabled2=true
    Module.Param3WriteEnabled=true
    Module.Param3WriteEnabled1=true
    Module.Param3WriteEnabled2=true
    Write Parameter Index 1=true
    Write Parameter Index 2=true
   ) {module=DynamicMaterialParameters001}
   &ScaleColor(
    Scale Alpha=&Multiply_Float(
     A=&FloatFromCurve(
      FloatCurve=DataInterface() {propertyValues {Curve {keys [
          <at 0 0>
          <at 0.1 0.75>
          {interpMode=RCIM_Cubic, tangentMode=RCTM_User, time=0.2, value=2.25}
          {interpMode=RCIM_Cubic, tangentMode=RCTM_User, time=0.7612, value=1.5}
          <at 1 0>
         ]}}}
      Scale Curve=&RandomRangeFloat(Minimum=1, Maximum=1.5)
     )
     B=&Multiply_Float(A=&FNTimeOfDayScalar(Night Scale=1, Morning Scale=0.7, Evening Scale=0.7), B=1.5)
    )
    Scale RGB=&VectorFromFloat(Value=&FloatFromCurve(FloatCurve=DataInterface() {propertyValues {Curve {keys [
          {arriveTangent=1.8145, interpMode=RCIM_Cubic, leaveTangent=0.4403, tangentMode=RCTM_Break, time=0, value=0.4}
          {arriveTangent=0.2639, tangentMode=RCTM_User, time=0.4, value=0.625}
          <at 1 1>
         ]}}}))
    ScaleA=true
    ScaleRGB=true
   )
   &CameraOffset(Camera Offset Amount=&MaskFloatBySpawnGroup(
     Masked Float=0
     Passthrough Float=&MaskFloatBySpawnGroup(
      Masked Float=0
      Passthrough Float=&MaskFloatBySpawnGroup(Masked Float=0, Passthrough Float=10, Spawn Group Mask=2)
      Spawn Group Mask=1
     )
    ))
  ]
  rendererClasses [&NiagaraSpriteRendererProperties]
  renderers [
   {rendererClass=&NiagaraSpriteRendererProperties}
   {rendererClass=&NiagaraSpriteRendererProperties, rendererIndex=1}
  ]
 }
}
summary {
 systemName=NS_Emote_MuffinLadle_gas_ElectricityImpact
 emitters [
  {emitterName=High-epic_dust, simTarget=GPUComputeSim, rendererClasses [&NiagaraSpriteRendererProperties]}
  {
   emitterName=Compressed
   rendererClasses [
    &NiagaraSpriteRendererProperties
    &NiagaraLightRendererProperties
    &NiagaraMeshRendererProperties
   ]
  }
 ]
 userVariables [
  User.ArcRevealTimeInSeconds {type=NiagaraFloat, defaultValue=0.28, description=""}
  User.bIsFrontEnd {type=NiagaraBool}
  User.bIsFrontEndPreview {type=NiagaraBool}
  User.ColorVariant {type=NiagaraInt32, defaultValue=1}
  User.EndingPoint {type=NiagaraPosition, description="", defaultValue {}}
  User.Global Scale {type=NiagaraFloat, defaultValue=1, description=""}
  User.Magnitude {type=NiagaraFloat, defaultValue=0, description=""}
  User.Scale {type=NiagaraFloat, defaultValue=1, description=""}
  User.StartingPoint {type=NiagaraPosition, description="", defaultValue {}}
 ]
}
