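# Niagara system for the MuffinLadle_Gas emote that draws short-lived purple anim trails between two bones or sockets of a skeletal mesh.
# Two near-identical emitters, each fed by its own skeletal mesh user parameter, spawn and update the trail through FN AnimTrail modules.
# Curves over particle and emitter age fade alpha and drive glow, swoosh and material parameters, with main color brightness scaled by time of day.
# Each emitter renders through four ribbon renderers.
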NiagaraSystem /BRCosmetics/Animation/Game/MainPlayer/Emotes/MuffinLadle_Gas/FX/NS_Emote_MuffinLadle_gas_Trail2
dependencies {
 usedDataInterfaces [&NiagaraDataInterfaceCurve, &NiagaraDataInterfaceColorCurve]
 usedDynamicInputs []
 usedModules [&SystemState]
}
emitters {
 Trail001 {
  emitterUpdateScript [
   &EmitterState()
   &FN_SpawnAnimTrail()
  ]
  particleSpawnScript [
   &FN_InitializeAnimTrailBoneIndecies(First Bone or Socket Index=1, Second Bone or Socket Index=0)
   &FN_InitializeAnimTrail(Color=<rgba 6.3113 0.9102 16.9043 1>, Lifetime=0.1)
   &AddVelocity(Velocity=$User.NewVector3f, Velocity Speed Scale=307.5105) {enabled=false}
  ]
  particleUpdateScript [
   &FN_UpdateAnimTrail(SkeletalMesh=$User.SkeletalMesh) {module=FN_UpdateAnimTrail001}
   /SetVariables(
    Particles.AlphaByEmitterAge:NiagaraFloat=&FloatFromCurve(
     FloatCurve=DataInterface() {propertyValues {Curve {
        defaultValue=1
        keys [{interpMode=RCIM_Cubic, tangentMode=RCTM_User, time=0, value=1}, {interpMode=RCIM_Cubic, tangentMode=RCTM_User, time=11, value=0}]
       }}}
     CurveIndex=$Emitter.Age
    )
    Particles.AlphaByNormalizedAge:NiagaraFloat=&FloatFromCurve(
     FloatCurve=DataInterface() {propertyValues {Curve {keys [{interpMode=RCIM_Cubic, tangentMode=RCTM_User, time=0, value=1}, {interpMode=RCIM_Cubic, tangentMode=RCTM_User, time=1, value=0}]}}}
     Scale Curve=0.3
    )
   ) {isSetParameters=true, module=SetVariables_#1}
   &ScaleColor(
    Color Value To Scale=&Multiply_LinearColorByFloat(
     Linear Color=<rgba 7.467 1.0769 20 1>
     Float=&FNTimeOfDayScalar(Night Scale=2, Morning Scale=1, Day Scale=1.5, Evening Scale=1.5)
    )
    Scale Alpha=&Multiply_Float(A=$Particles.AlphaByEmitterAge, B=$Particles.AlphaByNormalizedAge)
    Scale RGB=&MakeVectorFromLinearColorRGB(Color=&ColorFromCurve(ColorCurve=DataInterface() {propertyValues {
        AlphaCurve {keys []}
        BlueCurve {keys [<at 0 1 RCIM_Cubic>,<at 0.6466 1 RCIM_Cubic>]}
        GreenCurve {keys [<at 0 1 RCIM_Cubic>,<at 0.6466 0.0538 RCIM_Cubic>]}
        RedCurve {keys [<at 0 1 RCIM_Cubic>,<at 0.6466 0.3734 RCIM_Cubic>]}
       }}))
    ScaleA=true
    ScaleRGB=true
   ) {module=ScaleColor002}
   /SetVariables(Particles.Glow Color=&MakeLinearColorFromVectorAndFloat(
     "Vector (RGB)"=&MakeVectorFromLinearColorRGB(Color=&MultiplyLinearColor(
       Linear Color=<rgba 7.467 1.0769 20 0.5>
       Scale Factor=&MakeVector4_FromLinearColor(LinearColor=&ColorFromCurve(ColorCurve=DataInterface() {propertyValues {
           AlphaCurve {keys [<at 0 1 RCIM_Cubic>,<at 1 0 RCIM_Cubic>]}
           BlueCurve {keys [<at 0 1 RCIM_Cubic>,<at 0.5414 1.05 RCIM_Cubic>]}
           GreenCurve {keys [<at 0 1 RCIM_Cubic>,<at 0.5414 0.0565 RCIM_Cubic>]}
           RedCurve {keys [<at 0 1 RCIM_Cubic>,<at 0.5414 0.392 RCIM_Cubic>]}
          }}))
      ))
     "Float (Alpha)"=&Multiply_Float(A=$Particles.AlphaByEmitterAge, B=$Particles.AlphaByNormalizedAge)
    )) {isSetParameters=true, module=SetVariables_#2}
   /SetVariables(Particles.Glow Dyn Mat Param:Vector4f=&MakeVector4(
     X=0
     Y=&FloatFromCurve(
      FloatCurve=DataInterface() {propertyValues {Curve {keys [<at 0 0>,<at 0.1 0.6>,<at 0.3 0.075>,<at 0.5 0.825>,<at 0.7 0.3>,<at 1 0>]}}}
      Scale Curve=0.7
     )
     Z=0
     W=0
    )) {isSetParameters=true, module=SetVariables_#3}
   /SetVariables(Particles.Swoosh Color:LinearColor=&MakeLinearColorFromVectorAndFloat(
     "Vector (RGB)"=&MakeVectorFromLinearColorRGB(Color=&MultiplyLinearColor(
       Linear Color=<rgba 0.3285 0.0474 0.88 1>
       Scale Factor=&MakeVector4_FromLinearColor(LinearColor=&ColorFromCurve(ColorCurve=DataInterface() {propertyValues {
           AlphaCurve {keys []}
           BlueCurve {keys [<at 0 1.5 RCIM_Cubic>,<at 0.749 0.88>]}
           GreenCurve {keys [<at 0 1.5 RCIM_Cubic>,<at 0.749 0.0474>]}
           RedCurve {keys [<at 0 1.5 RCIM_Cubic>,<at 0.749 0.3285>]}
          }}))
      ))
     "Float (Alpha)"=&Multiply_Float(A=$Particles.AlphaByEmitterAge, B=$Particles.AlphaByNormalizedAge)
    )) {isSetParameters=true, module=SetVariables_#4}
   &DynamicMaterialParameters(
    Index 0 Param 2=&FloatFromCurve(FloatCurve=DataInterface() {propertyValues {Curve {keys [
         {arriveTangent=-0.0204, interpMode=RCIM_Cubic, leaveTangent=-0.0204, tangentMode=RCTM_User, time=0, value=0}
         {arriveTangent=-2.2131, interpMode=RCIM_Cubic, leaveTangent=-2.2131, tangentMode=RCTM_User, time=1, value=-1}
        ]}}})
    Module.Index0Float0Write=true
    Module.Index0Float1Write=true
    Module.Index0Float2Write=true
    Module.Index0Float3Write=true
   )
   &SolveForcesAndVelocity(Clamp Velocity=false, Limit Acceleration=false)
   /SetVariables(Particles.MOTION BLUR COLOR:LinearColor=<rgba 0.2582 0.0437 0.6867 0.5>, Particles.RibbonUVDistance=6.5185) {isSetParameters=true}
  ]
  rendererClasses [&NiagaraRibbonRendererProperties]
  renderers [
   {rendererClass=&NiagaraRibbonRendererProperties}
   {rendererClass=&NiagaraRibbonRendererProperties, rendererIndex=1}
   {rendererClass=&NiagaraRibbonRendererProperties, rendererIndex=2}
   {rendererClass=&NiagaraRibbonRendererProperties, rendererIndex=3}
  ]
 }
 Trail_1 {
  emitterUpdateScript [
   &EmitterState()
   &FN_SpawnAnimTrail()
  ]
  particleSpawnScript [
   &FN_InitializeAnimTrailBoneIndecies(First Bone or Socket Index=1, Second Bone or Socket Index=0)
   &FN_InitializeAnimTrail(Color=<rgba 6.3113 0.9102 16.9043 1>, Lifetime=0.1, SkeletalMesh=$User.Skeletal Mesh)
   &AddVelocity(Velocity=$User.NewVector3f, Velocity Speed Scale=307.5105) {enabled=false}
  ]
  particleUpdateScript [
   &FN_UpdateAnimTrail(SkeletalMesh=$User.Skeletal Mesh) {module=FN_UpdateAnimTrail001}
   /SetVariables(
    Particles.AlphaByEmitterAge:NiagaraFloat=&FloatFromCurve(
     FloatCurve=DataInterface() {propertyValues {Curve {
        defaultValue=1
        keys [{interpMode=RCIM_Cubic, tangentMode=RCTM_User, time=0, value=1}, {interpMode=RCIM_Cubic, tangentMode=RCTM_User, time=11, value=0}]
       }}}
     CurveIndex=$Emitter.Age
    )
    Particles.AlphaByNormalizedAge:NiagaraFloat=&FloatFromCurve(
     FloatCurve=DataInterface() {propertyValues {Curve {keys [{interpMode=RCIM_Cubic, tangentMode=RCTM_User, time=0, value=1}, {interpMode=RCIM_Cubic, tangentMode=RCTM_User, time=1, value=0}]}}}
     Scale Curve=0.3
    )
   ) {isSetParameters=true, module=SetVariables_#1}
   &ScaleColor(
    Color Value To Scale=&Multiply_LinearColorByFloat(
     Linear Color=<rgba 7.467 1.0769 20 1>
     Float=&FNTimeOfDayScalar(Night Scale=2, Morning Scale=1, Day Scale=1.5, Evening Scale=1.5)
    )
    Scale Alpha=&Multiply_Float(A=$Particles.AlphaByEmitterAge, B=$Particles.AlphaByNormalizedAge)
    Scale RGB=&MakeVectorFromLinearColorRGB(Color=&ColorFromCurve(ColorCurve=DataInterface() {propertyValues {
        AlphaCurve {keys []}
        BlueCurve {keys [<at 0 1 RCIM_Cubic>,<at 0.6466 1 RCIM_Cubic>]}
        GreenCurve {keys [<at 0 1 RCIM_Cubic>,<at 0.6466 0.0538 RCIM_Cubic>]}
        RedCurve {keys [<at 0 1 RCIM_Cubic>,<at 0.6466 0.3734 RCIM_Cubic>]}
       }}))
    ScaleA=true
    ScaleRGB=true
   ) {module=ScaleColor002}
   /SetVariables(Particles.Glow Color=&MakeLinearColorFromVectorAndFloat(
     "Vector (RGB)"=&MakeVectorFromLinearColorRGB(Color=&MultiplyLinearColor(
       Linear Color=<rgba 7.467 1.0769 20 0.5>
       Scale Factor=&MakeVector4_FromLinearColor(LinearColor=&ColorFromCurve(ColorCurve=DataInterface() {propertyValues {
           AlphaCurve {keys [<at 0 1 RCIM_Cubic>,<at 1 0 RCIM_Cubic>]}
           BlueCurve {keys [<at 0 1 RCIM_Cubic>,<at 0.5414 1.05 RCIM_Cubic>]}
           GreenCurve {keys [<at 0 1 RCIM_Cubic>,<at 0.5414 0.0565 RCIM_Cubic>]}
           RedCurve {keys [<at 0 1 RCIM_Cubic>,<at 0.5414 0.392 RCIM_Cubic>]}
          }}))
      ))
     "Float (Alpha)"=&Multiply_Float(A=$Particles.AlphaByEmitterAge, B=$Particles.AlphaByNormalizedAge)
    )) {isSetParameters=true, module=SetVariables_#2}
   /SetVariables(Particles.Glow Dyn Mat Param:Vector4f=&MakeVector4(
     X=0
     Y=&FloatFromCurve(
      FloatCurve=DataInterface() {propertyValues {Curve {keys [<at 0 0>,<at 0.1 0.6>,<at 0.3 0.075>,<at 0.5 0.825>,<at 0.7 0.3>,<at 1 0>]}}}
      Scale Curve=0.7
     )
     Z=0
     W=0
    )) {isSetParameters=true, module=SetVariables_#3}
   /SetVariables(Particles.Swoosh Color:LinearColor=&MakeLinearColorFromVectorAndFloat(
     "Vector (RGB)"=&MakeVectorFromLinearColorRGB(Color=&MultiplyLinearColor(
       Linear Color=<rgba 0.3285 0.0474 0.88 1>
       Scale Factor=&MakeVector4_FromLinearColor(LinearColor=&ColorFromCurve(ColorCurve=DataInterface() {propertyValues {
           AlphaCurve {keys []}
           BlueCurve {keys [<at 0 1.5 RCIM_Cubic>,<at 0.749 0.88>]}
           GreenCurve {keys [<at 0 1.5 RCIM_Cubic>,<at 0.749 0.0474>]}
           RedCurve {keys [<at 0 1.5 RCIM_Cubic>,<at 0.749 0.3285>]}
          }}))
      ))
     "Float (Alpha)"=&Multiply_Float(A=$Particles.AlphaByEmitterAge, B=$Particles.AlphaByNormalizedAge)
    )) {isSetParameters=true, module=SetVariables_#4}
   &DynamicMaterialParameters(
    Index 0 Param 2=&FloatFromCurve(FloatCurve=DataInterface() {propertyValues {Curve {keys [
         {arriveTangent=-0.0204, interpMode=RCIM_Cubic, leaveTangent=-0.0204, tangentMode=RCTM_User, time=0, value=0}
         {arriveTangent=-2.2131, interpMode=RCIM_Cubic, leaveTangent=-2.2131, tangentMode=RCTM_User, time=1, value=-1}
        ]}}})
    Module.Index0Float0Write=true
    Module.Index0Float1Write=true
    Module.Index0Float2Write=true
    Module.Index0Float3Write=true
   )
   &SolveForcesAndVelocity(Clamp Velocity=false, Limit Acceleration=false)
   /SetVariables(Particles.MOTION BLUR COLOR:LinearColor=<rgba 0.2582 0.0437 0.6867 0.5>, Particles.RibbonUVDistance=6.5185) {isSetParameters=true}
  ]
  rendererClasses [&NiagaraRibbonRendererProperties]
  renderers [
   {rendererClass=&NiagaraRibbonRendererProperties}
   {rendererClass=&NiagaraRibbonRendererProperties, rendererIndex=1}
   {rendererClass=&NiagaraRibbonRendererProperties, rendererIndex=2}
   {rendererClass=&NiagaraRibbonRendererProperties, rendererIndex=3}
  ]
 }
}
summary {
 systemName=NS_Emote_MuffinLadle_gas_Trail2
 emitters [
  {emitterName=Trail_1, rendererClasses [&NiagaraRibbonRendererProperties]}
  {emitterName=Trail001, rendererClasses [&NiagaraRibbonRendererProperties]}
 ]
 userVariables [
  User.bIsFrontEnd {type=NiagaraBool}
  User.bIsFrontEndPreview {type=NiagaraBool}
  User.NewVector3f {type=Vector3f, description="", defaultValue {struct=Vector3f, value=<10 0 0>}}
  User.RibbonIDOffset {type=NiagaraInt32}
  User.Skeletal Mesh {
   type=NiagaraDataInterfaceSkeletalMesh
   defaultValue {
    struct=NiagaraExt_VariableValue_DataInterface
    value {dataInterface=&self:NiagaraDataInterfaceSkeletalMesh_1}
   }
  }
  User.SkeletalMesh {
   type=NiagaraDataInterfaceSkeletalMesh
   defaultValue {
    struct=NiagaraExt_VariableValue_DataInterface
    value {dataInterface=&self:NiagaraDataInterfaceSkeletalMesh_2}
   }
  }
 ]
}
